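; Visual actor for the Athena safe zone storm, with storm quad, cloud ring and screen effect components.
; Construction configures or destroys components per storm-version flags and applies debug previews through the MPC_SafeZoneClouds parameter collection.
; BeginPlay swaps quads for cheaper rings on mobile and listens for game phases, revealing the screen effect at Aircraft, then unsubscribing.
; A Day Sequence interface event sets the volumetric storm flag to the inverse of mesh visibility.
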(event EventEndPlay
  (params (EndPlayReason :type EEndPlayReason%20Enum :default "Destroyed"))
  (bind _gamephaseupdatedasynctask (Variables|Default|GetGamePhaseUpdatedAsyncTask)
    (:then
      (GameplayEventRouter|StopListeningtoStatefulEvent _gamephaseupdatedasynctask))
    (:else)))

(event EventBeginPlay
  (Utilities|FlowControl|Sequence
    (:then_0
      (bind _returnvalue (GameplayEventRouter|GetEventRouter "/Script/FortniteGame.FortGameState" self))
      (bind (_asynctaskproxy _payload) (GameplayEventRouter|StartListeningToStatefulEvent _returnvalue "/Script/FortniteGame.FortGamePhaseUpdatedEvent")
        (:then
          (Variables|Default|SetGamePhaseUpdatedAsyncTask _asynctaskproxy))
        (:OnEventReceived
          _)
        (:OnSavedEventStateExists
          _))
      (CallFunction|HandleGamePhase (Utilities|Struct|BreakFortGamePhaseUpdatedEvent _payload)))
    (:then_1
      (if (Fortnite|IsMobilePlatform)
        (Rendering|SetVisibility (Variables|Default|GetStormQuadMain) false true)
        (if (Variables|Default|GetUseNewStorm)
          (Rendering|SetVisibility (Variables|Default|GetStormRings) false false)
          (Rendering|SetVisibility (Variables|Default|GetSM_CloudRing_Mobile) true false)
          (Rendering|SetVisibility (Variables|Default|GetSM_CloudRingTopBlend) false false)
          (else
            (Rendering|SetVisibility (Variables|Default|GetStormRings) true false)))))))

(event Collision|EventActorBeginOverlap
  (params (OtherActor :type Actor%20Object%20Reference)))

(event Custom|MobileClouds
  (params (Mobile :type Boolean))
  (if Mobile
    (Rendering|SetVisibility (Variables|Default|GetStormQuadMain) false true)
    (Rendering|SetVisibility (Variables|Default|GetStormRings) true false)
    (else
      (Rendering|SetVisibility (Variables|Default|GetStormQuadMain) true true)
      (Rendering|SetVisibility (Variables|Default|GetStormRings) false false))))

(event Custom|ShowClouds
  (params (Amount :type Float%20%28double-precision%29))
  (Rendering|Material|SetScalarParameterValue "/Game/Athena/SafeZone/MPC_SafeZoneClouds.MPC_SafeZoneClouds" "CloudStartingBias" :!class /Script/Engine.KismetMaterialLibrary)
  (Rendering|Material|SetScalarParameterValue "/Game/Athena/SafeZone/MPC_SafeZoneClouds.MPC_SafeZoneClouds" "SafeZoneScale" Amount :!class /Script/Engine.KismetMaterialLibrary))

(event Custom|ShowStormScreenEffect
  (params (Show :type Boolean))
  (Development|SetHiddeninGame (Variables|Default|GetSM_Storm_ScreenEffect) (not Show) false))

(event Storm/DaySequenceInterface|EventUpdateStormMeshVisibility
  (params (Visible :type Boolean))
  ;[0.1284 0.6408 1 0.2] Set from the Day Sequence
  (Variables|Default|SetUseVolumetricStorm (not Visible)))

(fn HandleGamePhase
  (params (GamePhase :type EAthenaGamePhase%20Enum))
  (switch Utilities|FlowControl|Switch|SwitchonEAthenaGamePhase GamePhase
    (:None
      _)
    (:Setup
      _)
    (:Warmup
      _)
    (:Aircraft
      (Utilities|FlowControl|Sequence
        (:then_0
          (Rendering|SetVisibility (Variables|Default|GetSM_Storm_ScreenEffect) true false))
        (:then_1
          (bind _gamephaseupdatedasynctask (Variables|Default|GetGamePhaseUpdatedAsyncTask)
            (:then
              (GameplayEventRouter|StopListeningtoStatefulEvent _gamephaseupdatedasynctask))
            (:else)))))
    (:SafeZones
      _)
    (:EndGame
      _)
    (:Count
      _))
  (return false))

(fn ConstructionScript
  (Utilities|FlowControl|Sequence
    (:then_0
      (if (Variables|Default|GetUseNewStorm)
        (Rendering|Material|SetMaterial (Variables|Default|GetStormQuadMain3) 0 "/Game/Athena/SafeZone/Materials/M_SafeZoneStorm_new2.M_SafeZoneStorm_new2")
        (Rendering|SetBoundsScale (Variables|Default|GetStormQuadMain15) 1.3)
        (Transformation|SetActorLocation "0, 0,28000.000000")
        (Components|DestroyComponent (Variables|Default|GetCloudsOuterRing))
        (else
          (Components|DestroyComponent (Variables|Default|GetPostProcess))
          (Components|DestroyComponent (Variables|Default|GetBox))
          (Components|DestroyComponent (Variables|Default|GetSM_Storm_ScreenEffect))
          (Components|DestroyComponent (Variables|Default|GetSM_CloudRingTopBlend))
          (Components|DestroyComponent (Variables|Default|GetSM_CloudRing_Mobile))
          (Components|DestroyComponent (Variables|Default|GetSM_CloudBottomBlend)))))
    (:then_1
      (if (Variables|Debug|GetMobileMode)
        (if (Variables|Default|GetUseNewStorm)
          (Rendering|SetVisibility (Variables|Default|GetStormQuadMain) false true)
          (Rendering|SetVisibility (Variables|Default|GetStormRings) false false)
          (else
            (Rendering|SetVisibility (Variables|Default|GetStormQuadMain) false true)
            (Rendering|SetVisibility (Variables|Default|GetStormRings) true false)))
        (else
          (Rendering|SetVisibility (Variables|Default|GetStormQuadMain) true true)
          (Rendering|SetVisibility (Variables|Default|GetStormRings) false false)
          (Rendering|SetVisibility (Variables|Default|GetSM_CloudRing_Mobile) false false))))
    (:then_2
      (if (Variables|Debug|GetShowStorm)
        (Rendering|Material|SetScalarParameterValue "/Game/Athena/SafeZone/MPC_SafeZoneClouds.MPC_SafeZoneClouds" "CloudStartingBias" (Variables|Debug|GetStormBias) :!class /Script/Engine.KismetMaterialLibrary)
        (Rendering|Material|SetScalarParameterValue "/Game/Athena/SafeZone/MPC_SafeZoneClouds.MPC_SafeZoneClouds" "SafeZoneScale" (Variables|Debug|GetStormSize) :!class /Script/Engine.KismetMaterialLibrary)
        (Rendering|Material|SetVectorParameterValue "/Game/Athena/SafeZone/MPC_SafeZoneClouds.MPC_SafeZoneClouds" "SafeZoneLoc" (Math|Conversions|ToLinearColor(Vector) (Variables|Debug|GetStormCenter)) :!class /Script/Engine.KismetMaterialLibrary)
        (else
          (Rendering|Material|SetScalarParameterValue "/Game/Athena/SafeZone/MPC_SafeZoneClouds.MPC_SafeZoneClouds" "CloudStartingBias" 1.0 :!class /Script/Engine.KismetMaterialLibrary))))
    (:then_3
      (if (Variables|Default|GetUseSeason11Storm)
        (Components|DestroyComponent (Variables|Default|GetStormQuadMain))))))
