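; BP_StormMeshRotator is the visual safe-zone storm mesh actor, with circle meshes and a screen post-process sphere.
; On cosmetic clients it hides the circle mesh for volumetric storms and adjusts visibility from playlist, storm-effect, day-sequence and mobile CVar signals.
; SetStormProgression drives scale and cloud bias in MPC_SafeZoneClouds, optionally applying color overrides through dynamic materials.
; Construction offers a test preview and destroys the rotating movement component.
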(event EventBeginPlay
  (Networking|CanExecuteCosmeticEvents
    (:True
      (CallFunction|InitializeMobileStormSphere)
      (bind _returnvalue (Actor|GetActorOfClass "/Game/Athena/SafeZone/BP_SafeZoneStorm.BP_SafeZoneStorm_C"))
      (Utilities|IsValid _returnvalue
        (:Is%20Valid
          (CallFunction|SetStormMeshVisibility (not (and (Variables|/Game/Athena/SafeZone/BP_SafeZoneStorm.BP_SafeZoneStorm_C|GetUseVolumetricStorm _returnvalue) (Class|BTechArtGlobalLibrary|UseVolumetricCloudsBR)))))
        (:Is%20Not%20Valid
          _))
      (bind (_gamestate _playlist _playlistcontexttags) (Class|FortAsyncActionCurrentPlaylistReady|PlaylistDataReadyAsync)
        (:OnReady
          (Utilities|FlowControl|Sequence
            (:then_0
              (if (GameplayTags|HasTag _playlistcontexttags "(TagName=\"Athena.Playlist.Mash\")" true)
                (Rendering|SetVisibility (Variables|Default|GetstormCircleLowerSheet) (not (Variables|Default|GetHide%20Storm%20in%20Horde%20Rush)) false)))
            (:then_1
              ;[1 1 1] When this HF is true, it indicates that we apply "in the storm" post process FX directly in the gameplay cue GC_Athena_OutsideSafeZone instead of using the material on this sphere component
              (if (ScalableFloat|GetValueasBool (Variables|Default|GetHF_ApplyScreenFXInGameplayCue))
                (Rendering|SetVisibility (Variables|Default|GetSphere) false false)))))))))

(event PostgameUI|EventOnStormEffectsEnabled
  (params (bEnabled :type Boolean))
  (Variables|Default|SetareStormEffectsEnabled bEnabled)
  (Rendering|SetVisibility (Variables|Default|GetSM_stormCircleNew) (Variables|Default|GetareStormEffectsEnabled) false)
  (Rendering|SetVisibility (Variables|Default|GetSphere) (CallFunction|GetSphereEnabled) false))

(event Custom|setStormMeshVisibility
  (params (bEnabled :type Boolean))
  (Rendering|SetVisibility (Variables|Default|GetSM_stormCircleNew) bEnabled false))

(event Custom|SetStormProgression
  (params (Progression :type Float%20%28double-precision%29))
  (bind _returnvalue (- 1.0 Progression))
  (if (> _returnvalue 1.0)
    (Rendering|Material|SetScalarParameterValue "/Game/Athena/SafeZone/MPC_SafeZoneClouds.MPC_SafeZoneClouds" "CloudStartingBias" (- _returnvalue 1.0) :!class /Script/Engine.KismetMaterialLibrary)
    (Rendering|Material|SetScalarParameterValue "/Game/Athena/SafeZone/MPC_SafeZoneClouds.MPC_SafeZoneClouds" "SafeZoneScale" 1.0 :!class /Script/Engine.KismetMaterialLibrary)
    (else
      (Rendering|Material|SetScalarParameterValue "/Game/Athena/SafeZone/MPC_SafeZoneClouds.MPC_SafeZoneClouds" "SafeZoneScale" _returnvalue :!class /Script/Engine.KismetMaterialLibrary)
      (Rendering|Material|SetScalarParameterValue "/Game/Athena/SafeZone/MPC_SafeZoneClouds.MPC_SafeZoneClouds" "CloudStartingBias" :!class /Script/Engine.KismetMaterialLibrary)))
  (if (Variables|MID|GetbOverrideMaterialParams)
    (CallFunction|OverrideMaterialParams)))

(event Storm/DaySequenceInterface|EventUpdateStormMeshVisibility
  (params (Visible :type Boolean))
  ;[0.1284 0.6408 1 0.2] Set from the Day Sequence
  (CallFunction|SetStormMeshVisibility Visible))

(fn getSphereEnabled
  (if (or (TouchPlatform|IsMobileGame) (TouchPlatform|IsNintendoSwitch))
    (return (and (Variables|Default|GetisMobileStormEnabled) (Variables|Default|GetareStormEffectsEnabled)))
    (else
      (return (Variables|Default|GetareStormEffectsEnabled)))))

(fn handleCvarChanged
  (params (CVarName :type String))
  (Utilities|FlowControl|Sequence
    (:then_0
      (if (Utilities|String|EqualExactly(String) CVarName "Test.EnableMobileStormEffects")
        (bind _returnvalue (Development|GetConsoleVariableBoolValue "Test.EnableMobileStormEffects"))
        (if (Math|Boolean|NotEqual(Boolean) _returnvalue (Variables|Default|GetisMobileStormEnabled))
          (Variables|Default|SetisMobileStormEnabled _returnvalue)
          (Rendering|Material|SetMaterial (Variables|Default|GetSphere) 0 (select (Variables|Default|GetisMobileStormEnabled) "/Game/GameplayCueNotifies/StWStorm/MI_HeavyRainPost_Mobile.MI_HeavyRainPost_Mobile" "/Game/GameplayCueNotifies/StWStorm/M_HeavyRainPost_Inst1.M_HeavyRainPost_Inst1"))
          (Rendering|SetVisibility (Variables|Default|GetSphere) (CallFunction|GetSphereEnabled) false))))
    (:then_1
      (bind _returnvalue_2 (Actor|GetActorOfClass "/Game/Athena/SafeZone/BP_SafeZoneStorm.BP_SafeZoneStorm_C"))
      (Utilities|IsValid _returnvalue_2
        (:Is%20Valid
          (CallFunction|SetStormMeshVisibility (not (and (Class|BTechArtGlobalLibrary|UseVolumetricCloudsBR) (Variables|/Game/Athena/SafeZone/BP_SafeZoneStorm.BP_SafeZoneStorm_C|GetUseVolumetricStorm _returnvalue_2)))))))))

(fn initializeMobileStormSphere
  (DataDrivenCVar|BindEventtoOnDataDrivenCVarDelegate (EngineSubsystems|GetDataDrivenCVarEngineSubsystem) (EventDispatchers|CreateEvent :!event handleCvarChanged))
  (CallFunction|HandleCvarChanged "Test.EnableMobileStormEffects"))

(fn OverrideMaterialParams
  (bind _returnvalue (Rendering|Material|CreateDynamicMaterialInstance (Variables|Default|GetSM_stormCircleNew) 0 (Rendering|Material|GetMaterial (Variables|Default|GetSM_stormCircleNew) 0) "None" :!class /Script/Engine.PrimitiveComponent))
  (Variables|MID|SetMIDStormMain _returnvalue)
  (bind _returnvalue_2 (Rendering|Material|CreateDynamicMaterialInstance (Variables|Default|GetSM_stormCircleNew) 1 (Rendering|Material|GetMaterial (Variables|Default|GetSM_stormCircleNew) 1) "None" :!class /Script/Engine.PrimitiveComponent))
  (Variables|MID|SetMIDStormLower _returnvalue_2)
  (bind _returnvalue_3 (Rendering|Material|CreateDynamicMaterialInstance (Variables|Default|GetSM_stormCircleTopRim) 0 (Rendering|Material|GetMaterial (Variables|Default|GetSM_stormCircleTopRim) 0) "None" :!class /Script/Engine.PrimitiveComponent))
  (Variables|MID|SetMIDStormTop _returnvalue_3)
  (Rendering|Material|SetVectorParameterValue (Variables|MID|GetMIDStormMain) "CloudColorDark" (Variables|MID|GetCloudColorDark) :!class /Script/Engine.MaterialInstanceDynamic)
  (Rendering|Material|SetVectorParameterValue (Variables|MID|GetMIDStormMain) "CloudColorLight" (Variables|MID|GetCloudColorLight) :!class /Script/Engine.MaterialInstanceDynamic)
  (Rendering|Material|SetVectorParameterValue (Variables|MID|GetMIDStormMain) "BackScatteringColor" (Variables|MID|GetBackScatteringColor) :!class /Script/Engine.MaterialInstanceDynamic)
  (Rendering|Material|SetVectorParameterValue (Variables|MID|GetMIDStormMain) "FresnelColor" (Variables|MID|GetFresnelColor) :!class /Script/Engine.MaterialInstanceDynamic)
  (Rendering|Material|SetVectorParameterValue (Variables|MID|GetMIDStormMain) "SunScatteringDay" (Variables|MID|GetSunScatteringDay) :!class /Script/Engine.MaterialInstanceDynamic)
  (Rendering|Material|SetVectorParameterValue (Variables|MID|GetMIDStormMain) "SunScatteringEvening" (Variables|MID|GetSunScatteringEvening) :!class /Script/Engine.MaterialInstanceDynamic)
  (Rendering|Material|SetVectorParameterValue (Variables|MID|GetMIDStormMain) "SunScatteringMorning" (Variables|MID|GetSunScatteringMorning) :!class /Script/Engine.MaterialInstanceDynamic)
  (Rendering|Material|SetVectorParameterValue (Variables|MID|GetMIDStormMain) "SunScatteringNight" (Variables|MID|GetSunScatteringNight) :!class /Script/Engine.MaterialInstanceDynamic)
  (Rendering|Material|SetVectorParameterValue (Variables|MID|GetMIDStormMain) "LightningColor" (Variables|MID|GetLightningColor) :!class /Script/Engine.MaterialInstanceDynamic)
  (Rendering|Material|SetVectorParameterValue (Variables|MID|GetMIDStormMain) "SecondaryClouds" (Variables|MID|GetSecondaryClouds) :!class /Script/Engine.MaterialInstanceDynamic))

(fn ConstructionScript
  (bind _sm_stormcirclenew (Variables|Default|GetSM_stormCircleNew))
  (bind _boundsscale (Variables|Default|GetBoundsScale))
  (bind _rotatingmovement (Variables|Default|GetRotatingMovement))
  (if (Variables|TestStorm|GetTestStorm)
    (CallFunction|SetStormProgression (Variables|TestStorm|GetStormProgression))
    (bind _returnvalue (Actor|GetActorOfClass "/Game/Athena/SafeZone/BP_SafeZoneStorm.BP_SafeZoneStorm_C"))
    (Utilities|IsValid _returnvalue
      (:Is%20Valid
        (if (Variables|/Game/Athena/SafeZone/BP_SafeZoneStorm.BP_SafeZoneStorm_C|GetUseCH6MS2Storm _returnvalue)
          (Utilities|FlowControl|Sequence
            (:then_0
              _)
            (:then_1
              (AddComponent|NODE_AddStaticMeshComponent-0)))))
      (:Is%20Not%20Valid
        (Utilities|FlowControl|Sequence
          (:then_0
            ;[1 1 1] Destroy Rotator on Mobile
            (bind _returnvalue_2 (Settings|GetGameUserSettings))
            (if (Math|Boolean|ORBoolean (Fortnite|IsMobilePlatform) (<= (Settings|GetVisualEffectQuality _returnvalue_2) 2) (Fortnite|IsConsolePlatform))
              (Utilities|IsValid _rotatingmovement
                (:Is%20Valid
                  (Components|DestroyComponent _rotatingmovement)))))
          (:then_1
            ;[1 1 1] Setup Materials
            (Rendering|SetBoundsScale _sm_stormcirclenew _boundsscale)
            (bind _returnvalue_3 (Rendering|Material|CreateDynamicMaterialInstance _sm_stormcirclenew 0 0 "BP_StormMeshRotator_SM_stormCircleNew_Mat0" :!class /Script/Engine.PrimitiveComponent))
            (bind _returnvalue_4 (Rendering|Material|CreateDynamicMaterialInstance _sm_stormcirclenew 1 0 "BP_StormMeshRotator_SM_stormCircleNew_Mat1" :!class /Script/Engine.PrimitiveComponent))
            (bind _returnvalue_5 (Rendering|Material|CreateDynamicMaterialInstance (Variables|Default|GetSM_stormCircleTopRim) 0 0 "BP_StormMeshRotator_SM_stormCircleTopRim_Mat0" :!class /Script/Engine.PrimitiveComponent))
            (bind _returnvalue_6 (Rendering|Material|CreateDynamicMaterialInstance (Variables|Default|GetstormCircleLowerSheet) 0 0 "BP_StormMeshRotator_stormCircleLowerSheet_Mat0" :!class /Script/Engine.PrimitiveComponent))
            (Rendering|Material|SetScalarParameterValue _returnvalue_5 "EnableWPO" 1.0 :!class /Script/Engine.MaterialInstanceDynamic)
            (Rendering|Material|SetScalarParameterValue _returnvalue_5 "BoundsScale" _boundsscale :!class /Script/Engine.MaterialInstanceDynamic)))))
    (else
      (bind _returnvalue_7 (Actor|GetActorOfClass "/Game/Athena/SafeZone/BP_SafeZoneStorm.BP_SafeZoneStorm_C"))
      (Utilities|IsValid _returnvalue_7
        (:Is%20Valid
          (if (Variables|/Game/Athena/SafeZone/BP_SafeZoneStorm.BP_SafeZoneStorm_C|GetUseCH6MS2Storm _returnvalue_7)
            _
            (elif (Variables|/Game/Athena/SafeZone/BP_SafeZoneStorm.BP_SafeZoneStorm_C|GetUseSeason11Storm _returnvalue_7)
              _)))
        (:Is%20Not%20Valid
          _))))
  ;[1 1 1] Destroy S11 components if !useSeason11Storm
  (Components|DestroyComponent (Variables|Default|GetRotatingMovement)))
