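# Unlit additive material that draws animated lightning strikes and impact glows on the Athena safe zone shield (JanusShield).
# Vertex color G blends panned, noise-textured lightning with radial impact flashes and waves, all on a looping, pivot-seeded timer that fades with camera distance.
# World position offset uses safe-zone center and radius parameters to randomly rotate, scale and wrap the mesh onto the ring.
# A quality switch collapses the mesh on Low, Medium and High.
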Material /Game/Athena/SafeZone/M_JanusShield_Impact
graph {
 properties {
  bUseMaterialAttributes=true
  bUsedWithNanite=true
  blendMode=BLEND_Additive
  pixelDepthOffsetMode=PDOM_Legacy
  shadingModel=MSM_Unlit
  twoSided=true
  usedShadingModels=Unlit
 }
 nodes [
  n1 = CollectionParameter() {collection=&FortniteMaterialParameters, parameterName=AthenaSafeZoneCenter}
  n2 = ComponentMask(None=@n1) {b=true, g=true, r=true}
  n3 = CollectionParameter() {collection=&FortniteMaterialParameters, parameterName=AthenaSafeZoneRadius}
  n4 = Multiply(A=ScalarParameter() {defaultValue=1, parameterName=SafeZone Radius}, B=ComponentMask(None=@n3) {r=true})
  n5 = Constant3Vector() {constant=<rgba 0 0 1 0>}
  n6 = ComponentMask(None=ObjectPositionWS()) {b=true}
  n7 = Multiply(A=Frac(None=Divide(A=@n6) {constB=1.33}), B=ScalarParameter() {defaultValue=321.2828, parameterName=Random2})
  n8 = Add(A=@n7, B=Time())
  n9 = Multiply(A=@n8, B=ScalarParameter() {defaultValue=0.35, parameterName=MasterSpeedLightning})
  n10 = Ceil(None=@n9)
  n11 = Ceil(None=Multiply(A=Add(A=@n10, B=@n7), B=ScalarParameter() {defaultValue=2.3418, parameterName=WaveSpeed}))
  n12 = Divide(A=@n11, B=ScalarParameter() {defaultValue=0.6086, parameterName=Randomb})
  n13 = LinearInterpolate(
   A=ScalarParameter() {defaultValue=-4.4493, parameterName=RotationOffset_Min}
   B=ScalarParameter() {defaultValue=2.8859, parameterName=RotationOffset_Max}
   Alpha=Frac(None=@n12)
  )
  n14 = Constant3Vector() {constant=<rgba 1 0 0 0>}
  n15 = Add(
   A=RotateAboutAxis(
    NormalizedRotationAxis=@n5
    RotationAngle=@n13
    PivotPoint=MaterialFunctionCall() {materialFunction=&ComponentPivotLocation}.Result
    Position=@n14
   )
   B=@n14
  )
  Add(A=@n2, B=Multiply(A=@n4, B=Normalize(VectorInput=@n15)))
  Constant()
  Frac(None=@n7)
  Multiply(B=ScalarParameter() {defaultValue=2.3418, parameterName=WaveSpeed})
  n16 = TextureCoordinate()
  n17 = Constant() {r=1}
  n18 = ScalarParameter() {defaultValue=0.5, parameterName=LightingTiling}
  n19 = Panner(
   Coordinate=Multiply(A=AppendVector(A=@n17, B=@n18), B=@n16)
   Time=@n8
   Speed=AppendVector(A=Constant(), B=ScalarParameter() {defaultValue=0.2, parameterName=PanLightingNOise})
  )
  n20 = Subtract(
   A=TextureSample(UVs=ComponentMask(None=@n19) {g=true}) {samplerType=SAMPLERTYPE_LinearColor, texture=/Game/Characters/Player/Male/Medium/Bodies/M_MED_SRDrift/Textures/T_ElectricNoise}.R
   B=ScalarParameter() {defaultValue=0.1238, parameterName=NoiseBias}
  ) {constB=0.5}
  n21 = ScalarParameter() {defaultValue=1.2622, parameterName=Noise}
  n22 = ComponentMask(None=Add(A=@n16, B=Multiply(A=@n20, B=@n21))) {r=true}
  Multiply(A=@n22, B=OneMinus(None=@n22))
  Multiply(A=@n6, B=@n12)
  ObjectPositionWS()
  ObjectPositionWS()
  n23 = Multiply(
   A=Power(
    Base=TextureSample(UVs=@n19) {texture=/Game/Athena/SafeZone/Textures/T_JanusShield_Lightning}.R
    Exp=ScalarParameter() {defaultValue=1.1, parameterName=FalloffLightning}
   )
   B=ScalarParameter() {defaultValue=1, parameterName=BrightnessLightning}
  ) {constB=2}
  n24 = ComponentMask(None=@n16) {g=true}
  n25 = Multiply(A=Frac(None=@n9), B=ScalarParameter() {defaultValue=2, parameterName=Pause})
  n26 = Saturate(None=@n25)
  n27 = Subtract(
   A=Multiply(A=@n26, B=ScalarParameter() {defaultValue=1.3, parameterName=SpeedStrike})
   B=ScalarParameter() {defaultValue=-0.4286, parameterName=Strikeoffset}
  )
  n28 = ComponentMask(None=TextureCoordinate()) {g=true}
  n29 = Add(
   A=LinearInterpolate(A=@n26, B=Saturate(None=@n27), Alpha=VertexColor().B)
   B=Subtract(A=Multiply(A=@n28) {constB=0.5}, B=ScalarParameter() {defaultValue=0.5, parameterName=PhaseOffset})
  )
  n30 = Multiply(
   A=Multiply(A=Multiply(A=@n23, B=@n24), B=VectorParameter() {parameterName=Color, defaultValue=<rgba 1.742 50 44.5161 1>}.RGB)
   B=Saturate(None=Sine(None=@n29))
  )
  n31 = TextureCoordinate()
  n32 = Subtract(A=ComponentMask(None=@n31) {r=true}) {constB=0.5}
  n33 = Transform(None=Constant3Vector() {constant=<rgba 1 0 0 0>})
  n34 = LinearInterpolate(
   A=ScalarParameter() {defaultValue=10, parameterName=Width}
   B=ScalarParameter() {defaultValue=100, parameterName=Width2}
   Alpha=ComponentMask(None=@n31) {g=true}
  )
  n35 = MaterialFunctionCall() {materialFunction=/Engine/Functions/Engine_MaterialFunctions02/WorldPositionOffset/ObjectScale}
  n36 = Multiply(A=@n35.Scale Z, B=ScalarParameter() {defaultValue=100, parameterName=WidthStrikeBottom})
  n37 = LinearInterpolate(
   A=Multiply(A=Multiply(A=@n32, B=@n33), B=Multiply(A=@n34, B=@n35.Scale Z))
   B=MaterialFunctionCall(WidthBase=@n36, WidthTip=@n36) {materialFunction=&SplineThicken}.WorldPositionOffset
   Alpha=VertexColor().B
  )
  n38 = OneMinus(None=@n26)
  n39 = Saturate(None=Multiply(A=@n25, B=ScalarParameter() {defaultValue=3, parameterName=FadeIn}))
  n40 = VectorParameter() {parameterName=Color3, defaultValue=<rgba 0 2.9075 10 1>}
  n41 = Distance(A=TextureCoordinate(), B=Constant2Vector() {g=0.5, r=0.5})
  n42 = ScalarParameter() {defaultValue=0.319, parameterName=ImpactRadius}
  n43 = Power(Base=OneMinus(None=Divide(A=@n41, B=@n42) {constB=1}), Exp=ScalarParameter() {defaultValue=8, parameterName=ImpactFalloff})
  n44 = ScalarParameter() {defaultValue=0.2571, parameterName=Delay1}
  n45 = ScalarParameter() {defaultValue=2, parameterName=RangeWave}
  n46 = Add(A=Saturate(None=Subtract(A=@n25, B=@n44)), B=OneMinus(None=Divide(A=@n41, B=@n45)))
  n47 = Divide(A=@n41, B=ScalarParameter() {defaultValue=0.5, parameterName=RangeFalloff}) {constB=1}
  n48 = Multiply(A=Saturate(None=Sine(None=@n46)), B=OneMinus(None=Saturate(None=@n47)))
  n49 = VectorParameter() {parameterName=Color2, defaultValue=<rgba 0 0.2908 1 1>}
  n50 = Add(A=Multiply(A=Multiply(A=@n38, B=@n39), B=Multiply(A=@n40.RGB, B=@n43)), B=Multiply(A=Multiply(A=@n48, B=@n49.RGB), B=@n49.RGB))
  n51 = VertexColor()
  n52 = GetMaterialAttributes(None=BlendMaterialAttributes(
    A=SetMaterialAttributes!(MaterialAttributes, Emissive Color=@n30, World Position Offset=@n37) {attributeSetTypes [#1, #2]}
    B=SetMaterialAttributes!(MaterialAttributes, Emissive Color=@n50) {attributeSetTypes [#1]}
    Alpha=@n51.G
   )) {attributeGetTypes [#2, #1]}
  n53 = Subtract(
   A=Add(A=WorldPosition() {worldPositionShaderOffset=WPT_ExcludeAllShaderOffsets}.XYZ, B=@n52.World Position Offset)
   B=MaterialFunctionCall() {materialFunction=&ComponentPivotLocation}.Result
  )
  n54 = Divide(A=@n11, B=ScalarParameter() {defaultValue=0.1705, parameterName=Randomb3})
  n55 = Divide(A=@n4, B=ScalarParameter() {defaultValue=100000, parameterName=maxRadius})
  n56 = Multiply(
   A=LinearInterpolate(
    A=ScalarParameter() {defaultValue=0.1, parameterName=RandScale_Min}
    B=ScalarParameter() {defaultValue=5, parameterName=RandScale_Max}
    Alpha=Frac(None=@n54)
   )
   B=LinearInterpolate(
    A=ScalarParameter() {defaultValue=0.2, parameterName=MinScale_Min}
    B=ScalarParameter() {defaultValue=1, parameterName=MaxScale_Max}
    Alpha=Saturate(None=@n55)
   )
  )
  n57 = Transform(None=Constant3Vector() {constant=<rgba 0 1 0 0>}) {transformSourceType=TRANSFORMSOURCE_Local}
  n58 = Add(A=Add(A=@n52.World Position Offset, B=Multiply(A=@n53, B=@n56)), B=Multiply(A=@n4, B=Normalize(VectorInput=@n57)))
  n59 = Constant3Vector() {constant=<rgba 1 1 0 0>}
  n60 = Multiply(A=Add(A=@n58, B=WorldPosition().XYZ), B=@n59)
  n61 = Multiply(
   A=@n59
   B=MaterialFunctionCall() {materialFunction=&ComponentPivotLocation}.Result
  )
  n62 = Subtract(A=Distance(A=@n60, B=@n61), B=@n4)
  n63 = Normalize(VectorInput=Subtract(A=@n61, B=@n60))
  n64 = Add(A=@n58, B=LinearInterpolate(A=Multiply(A=@n62, B=@n63), B=Constant3Vector(), Alpha=VertexColor().B))
  n65 = RotateAboutAxis(
   NormalizedRotationAxis=@n5
   RotationAngle=@n13
   PivotPoint=MaterialFunctionCall() {materialFunction=&ComponentPivotLocation}.Result
   Position=Add(A=@n64, B=WorldPosition().XYZ)
  )
  n66 = WorldPosition()
  n67 = CameraPositionWS()
  n68 = Saturate(None=Divide(A=Distance(A=@n66.XYZ, B=@n67), B=ScalarParameter() {defaultValue=54621.0312, parameterName=FadeDistance}))
  n69 = SetMaterialAttributes!(
   MaterialAttributes=@n52.MaterialAttributes
   World Position Offset=Add(A=@n2, B=Add(A=@n64, B=@n65))
   Emissive Color=Multiply(A=@n52.Emissive Color, B=OneMinus(None=@n68))
  ) {attributeSetTypes [#2, #1]}
  n70 = SetMaterialAttributes!(MaterialAttributes, World Position Offset=Subtract(A=ObjectPositionWS(), B=WorldPosition().XYZ)) {attributeSetTypes [#2]}
  QualitySwitch(Default=@n69, Low=@n70, High=@n70, Medium=@n70, Epic=@n69)
  Saturate(None=Add(A=@n51.G, B=@n51.G))
  ScalarParameter() {parameterName=TimeStatic}
  ScalarParameter() {defaultValue=1585.1021, parameterName=SafeZoneTest}
  ScalarParameter() {defaultValue=500, parameterName=WidthStrikeTip}
 ]
}
