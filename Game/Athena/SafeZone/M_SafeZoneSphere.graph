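# Translucent, two-sided unlit material for the safe zone storm wall, layering scanlines, caustics, grid and intersection glow with a storm-damage tint.
# An optional Janus shield layer adds quantized electric tiles and landscape-height RVT scanlines at Default/Epic quality, excluded on mobile.
# Opacity combines distance and height fades with camera position relative to the zone.
# Vertex offset aligns height to SafeZoneCapHeight and drops far vertices away, with cheaper mobile paths.
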Material /Game/Athena/SafeZone/M_SafeZoneSphere
graph {
 properties {
  bUsedWithNanite=true
  blendMode=BLEND_Translucent
  pixelDepthOffsetMode=PDOM_Legacy
  shadingModel=MSM_Unlit
  twoSided=true
  usedShadingModels=Unlit
 }
 nodes [
  n1 = Time() {bOverride_Period=true, period=1}
  n2 = Constant()
  n3 = Multiply(A=Add(A=WorldPosition().XYZ, B=Multiply(A=@n1, B=@n2)), B=Constant() {r=0.05})
  n4 = Add(A=PixelDepth().R, B=Constant() {r=-500})
  n5 = Constant() {r=3500}
  n6 = OneMinus(None=Saturate(None=Divide(A=@n4, B=@n5)))
  n7 = Multiply(
   A=TextureSample(UVs=ComponentMask(None=@n3) {b=true}) {samplerType=SAMPLERTYPE_LinearGrayscale, texture=&/ThinLines}.R
   B=@n6
  ) {desc=lines * distance fade}
  n8 = Add(
   A=Multiply(A=@n7, B=VectorParameter() {parameterName=scanlines Color, defaultValue=<rgba 0.0928 0.2583 0.8 1>}.RGB)
   B=VectorParameter() {parameterName=Day Color Far, defaultValue=<rgba 0.0125 0.2204 0.5 1>}.RGB
  )
  n9 = TextureObject() {samplerType=SAMPLERTYPE_Normal, texture=/Game/TheOutpost/Textures/TilingWeakPointNormal}
  n10 = VectorParameter() {parameterName=CausticWorldScale, defaultValue=<rgba 29400 36750 21000 0>}
  n11 = StaticBool()
  n12 = Constant() {r=1}
  n13 = Add(
   A=Multiply(
    A=MaterialFunctionCall!(TextureObject=@n9, TextureSize=@n10.RGB, WorldPosition, Export Float 4=@n11, In, World Space Normal, ProjectionTransitionContrast=@n12) {materialFunction=&/WorldAlignedTexture_ScalePan}.XY Texture
    B=VectorParameter() {parameterName=Caustic_TexScale, defaultValue=<rgba 1 1 1 1>}.RGB
   )
   B=VectorParameter() {parameterName=Caustic_TexBias, defaultValue=<rgba 0 0 0 1>}.RGB
  )
  n14 = Multiply(A=Time() {desc=15.015014, period=1}, B=ScalarParameter() {defaultValue=0.08, desc=0.0666, parameterName=Sine Speed A_1})
  n15 = ComponentMask(None=Frac(None=Add(A=@n13, B=@n14))) {g=true}
  n16 = Constant() {r=0.5}
  n17 = OneMinus(None=Multiply(A=Distance(A=@n15, B=@n16)) {constB=2})
  n18 = Multiply(
   A=LinearInterpolate(A=@n17, B=Round(None=@n17), Alpha=ScalarParameter() {defaultValue=0.25, parameterName=Caustic Round LERP})
   B=VectorParameter() {parameterName=Caustics Color, defaultValue=<rgba 0.016 0.1184 0.4 1>}.RGB
  )
  n19 = StaticSwitchParameter(True=@n17, False=ShadingPathSwitch(Default=@n17, Mobile=Constant() {r=1}) {desc=Migrated from FeatureLevelSwitch}) {parameterName=DisableMobileOptimizations}
  n20 = ScalarParameter() {defaultValue=0.2, parameterName=Ripple to Grid Scale}
  n21 = MaterialFunctionCall(
   TextureObject=TextureObject() {samplerType=SAMPLERTYPE_LinearColor, texture=/Game/Gadgets/Assets/Teleporter/Textures/T_Teleporter_GridSquares_01}
   TextureSize=Constant() {r=512}
   Export Float 4=StaticBool()
   ProjectionTransitionContrast=Constant() {r=5}
  ) {materialFunction=&WorldAlignedTexture_ScalePan}
  n22 = ComponentMask(None=@n21.XY Texture) {r=true}
  n23 = Multiply(A=Time() {bOverride_Period=true, period=3.8314}, B=ScalarParameter() {defaultValue=0.261, parameterName=Sine Speed Grid})
  n24 = Sine(None=Add(A=Multiply(A=@n19, B=@n20), B=Add(A=@n22, B=@n23)))
  n25 = ComponentMask(None=@n21.XY Texture) {g=true}
  n26 = Multiply(A=@n6, B=Power(Base=Multiply(A=@n24, B=@n25), Exp=ScalarParameter() {defaultValue=5, parameterName=Grid Power}))
  n27 = Multiply(A=@n26, B=VectorParameter() {parameterName=Grid Color, defaultValue=<rgba 0.0143 0.063 0.15 0>}.RGB)
  n28 = ScalarParameter() {defaultValue=350, parameterName=Intersection Edge Divisor}
  n29 = DepthFade(Opacity=Constant() {r=1}, FadeDistance=@n28)
  n30 = Multiply(
   A=Power(Base=OneMinus(None=@n29), Exp=ScalarParameter() {defaultValue=6, parameterName=Intersection Edge Exponent})
   B=VectorParameter() {parameterName=Intersection Edge Color, defaultValue=<rgba 0.25 1.5 15 1>}.RGB
  ) {constA=1}
  n31 = Add(A=Add(A=@n8, B=Add(A=@n18, B=@n27) {desc=+ Grid Blocks}), B=@n30) {desc=+ intersection edge}
  n32 = Multiply(
   A=LinearInterpolate(
    A=Constant() {r=1}
    B=VectorParameter() {parameterName=StormDamageColor, defaultValue=<rgba 4 0.2 0.2 1>}.RGB
    Alpha=ScalarParameter() {parameterName=StormDamage}
   ) {constAlpha=0}
   B=StaticSwitchParameter(True=@n31, False=ShadingPathSwitch(Default=@n31, Mobile=@n8) {desc=Migrated from FeatureLevelSwitch}) {parameterName=DisableMobileOptimizations}
  )
  n33 = Add(A=PixelDepth().R, B=Constant() {r=-100})
  n34 = ScalarParameter() {defaultValue=300, parameterName=NearFade}
  n35 = MaterialFunctionCall(Float3=WorldPosition().XYZ) {materialFunction=&BreakOutFloat3Components}
  n36 = Multiply(
   A=MaterialFunctionCall(Float3=CameraPositionWS()) {materialFunction=&BreakOutFloat3Components}.B
   B=ScalarParameter() {defaultValue=0.5, parameterName=CameraHeightInfluence}
  )
  n37 = Divide(
   A=MaterialFunctionCall(R=@n35.R, G=@n35.R, B=Subtract(A=@n35.B, B=@n36)) {materialFunction=&AppendMany}.RGB
   B=ScalarParameter() {defaultValue=23000, parameterName=HeightGradientScale}
  )
  n38 = OneMinus(None=Saturate(None=ComponentMask(None=@n37) {b=true}))
  n39 = Multiply(A=@n26, B=Constant() {r=0.2})
  n40 = ShadingPathSwitch(
   Default=ScalarParameter() {defaultValue=0.4, parameterName="Base Opacity Near (Outside OR Far from the shield)"}
   Mobile=ScalarParameter() {defaultValue=0.15, parameterName="Base Opacity Near (Outside OR Far from the shield) ES3.1"}
  ) {desc=Migrated from FeatureLevelSwitch}
  n41 = ScalarParameter() {defaultValue=0.2, parameterName="Base Opacity Near (Inside AND Close to the Shield)"}
  n42 = Distance(
   A=ComponentMask(None=CameraPositionWS()) {g=true, r=true}
   B=ComponentMask(None=CollectionParameter() {collection=&/MPC_SafeZoneClouds, parameterName=SafeZoneLoc}) {g=true, r=true}
  )
  n43 = MaterialFunctionCall() {materialFunction=&ObjectLocalBounds}
  n44 = Multiply(
   A=MaterialFunctionCall() {materialFunction=/Engine/Functions/Engine_MaterialFunctions02/WorldPositionOffset/ObjectScale}.Scale X
   B=Multiply(A=ComponentMask(None=@n43.Local Bounds Size) {r=true}) {constB=0.5}
  )
  n45 = Constant()
  n46 = Divide(A=Subtract(A=@n44, B=@n42), B=ScalarParameter() {defaultValue=10000, parameterName=FadeDist_Centimeters})
  n47 = If(A=@n42, B=@n44, "A > B"=@n45, "A == B"=@n45, "A < B"=OneMinus(None=Saturate(None=@n46)))
  n48 = ComponentMask(None=VertexInterpolator(VS=LinearInterpolate(A=@n40, B=@n41, Alpha=@n47)).PS) {r=true}
  n49 = ShadingPathSwitch(
   Default=ScalarParameter() {defaultValue=0.7, parameterName="Base Opacity Far (Outside OR Far from the shield)"}
   Mobile=ScalarParameter() {defaultValue=0.2, parameterName="Base Opacity Far (Outside OR Far from the shield) ES3.1"}
  ) {desc=Migrated from FeatureLevelSwitch}
  n50 = ScalarParameter() {defaultValue=0.2, parameterName="Base Opacity Far (Inside AND Close to the Shield)"}
  n51 = ComponentMask(None=VertexInterpolator(VS=LinearInterpolate(A=@n49, B=@n50, Alpha=@n47)).PS) {r=true}
  n52 = Constant() {r=10000}
  n53 = OneMinus(None=DepthFade(Opacity=Constant() {r=1}, FadeDistance=@n52))
  n54 = Add(A=@n30, B=Add(A=Add(A=@n39, B=@n7), B=LinearInterpolate(A=@n48, B=@n51, Alpha=@n53)))
  n55 = MaterialFunctionCall(Vector 2=ComponentMask(None=WorldPosition() {worldPositionShaderOffset=WPT_CameraRelative}.XYZ) {g=true, r=true}) {materialFunction=/Engine/Functions/Engine_MaterialFunctions02/Utility/VectorLength}
  n56 = Constant() {r=-100}
  n57 = QualitySwitch(
   Default=ScalarParameter() {defaultValue=44900, parameterName=FarFade}
   Epic=ScalarParameter() {defaultValue=84900, parameterName=FarFadeEpicQuality}
  )
  n58 = Saturate(None=Divide(A=Add(A=@n55.V2 Length, B=@n56), B=@n57))
  n59 = Multiply(A=Saturate(None=Divide(A=@n33, B=@n34)), B=Multiply(A=Multiply(A=@n38, B=@n54), B=OneMinus(None=@n58)))
  n60 = ScalarParameter() {defaultValue=0.328, parameterName=AddGlowToMobile}
  n61 = ShadingPathSwitch(Default=@n32, Mobile=Add(A=@n32, B=Multiply(A=@n59, B=@n60))) {desc=Migrated from FeatureLevelSwitch}
  n62 = Divide(
   A=Add(A=PixelDepth().R, B=ScalarParameter() {defaultValue=-2000, parameterName=OffsetNearFadeJanus})
   B=ScalarParameter() {defaultValue=155000, parameterName=NearFadeJanus}
  )
  n63 = ComponentMask(None=WorldPosition().XYZ) {r=true}
  n64 = ScalarParameter() {defaultValue=-21632.7676, parameterName=SizeQuantizedUVS}
  n65 = ScalarParameter() {defaultValue=32, parameterName=Quant}
  n66 = Floor(None=Multiply(A=Divide(A=@n63, B=@n64), B=@n65))
  n67 = RuntimeVirtualTextureSampleParameter(World Position=Multiply(A=Divide(A=@n66, B=@n65), B=@n64)) {
   materialType=WorldHeight
   parameterName=RVT Landscape Height Target
   virtualTexture=/Game/Athena/Apollo/Maps/Landscape/VT/RVT_Apollo_Landscape_Height
  }
  n68 = Subtract(A=WorldPosition().XYZ, B=Multiply(A=Constant3Vector() {constant=<rgba 0 0 1 0>}, B=@n67.WorldHeight))
  n69 = Multiply(A=ComponentMask(None=Divide(A=@n68, B=@n64)) {b=true, r=true})
  n70 = Panner(
   Coordinate=@n69
   Speed=AppendVector(
    A=ScalarParameter() {defaultValue=0.019, parameterName=SpeedPan_JanusNoise_x}
    B=ScalarParameter() {defaultValue=0.02, parameterName=SpeedPan_JanusNoise_y}
   )
  )
  n71 = ScalarParameter() {defaultValue=6.0141, parameterName=ScaleElectricNoise}
  n72 = ScalarParameter() {defaultValue=32, parameterName=Quant}
  n73 = Multiply(A=@n69, B=@n72)
  n74 = Divide(A=Multiply(A=Floor(None=@n73), B=ScalarParameter() {defaultValue=1.4333, parameterName=offsetRandom}), B=@n72)
  n75 = TextureSample(UVs=Add(A=Multiply(A=@n70, B=@n71), B=@n74)) {samplerType=SAMPLERTYPE_LinearColor, texture=/Game/Characters/Player/Male/Medium/Bodies/M_MED_SRDrift/Textures/T_ElectricNoise}
  n76 = ScalarParameter() {defaultValue=2.6805, parameterName=ScaleElectricNoise2}
  n77 = TextureSample(UVs=Add(A=Multiply(A=@n70, B=@n76), B=@n74)) {samplerType=SAMPLERTYPE_LinearColor, texture=&T_ElectricNoise}
  n78 = Power(
   Base=Multiply(A=ScalarParameter() {defaultValue=2, parameterName=BrightnesselectricityFalloff}, B=Add(A=@n75.R, B=@n77.R))
   Exp=ScalarParameter() {defaultValue=2, parameterName=ContrastElectricityFalloff}
  )
  n79 = ComponentMask(None=WorldPosition().XYZ) {g=true}
  n80 = ScalarParameter() {defaultValue=32, parameterName=Quant}
  n81 = Floor(None=Multiply(A=Divide(A=@n79, B=@n64), B=@n80))
  n82 = RuntimeVirtualTextureSampleParameter(World Position=Multiply(A=Divide(A=@n81, B=@n80), B=@n64)) {
   materialType=WorldHeight
   parameterName=RVT Landscape Height Target
   virtualTexture=&RVT_Apollo_Landscape_Height
  }
  n83 = Subtract(A=WorldPosition().XYZ, B=Multiply(A=Constant3Vector() {constant=<rgba 0 0 1 0>}, B=@n82.WorldHeight))
  n84 = Multiply(A=ComponentMask(None=Divide(A=@n83, B=@n64)) {b=true, g=true})
  n85 = Panner(
   Coordinate=@n84
   Speed=AppendVector(
    A=ScalarParameter() {defaultValue=0.019, parameterName=SpeedPan_JanusNoise_x}
    B=ScalarParameter() {defaultValue=0.02, parameterName=SpeedPan_JanusNoise_y}
   )
  )
  n86 = ScalarParameter() {defaultValue=6.0141, parameterName=ScaleElectricNoise}
  n87 = ScalarParameter() {defaultValue=32, parameterName=Quant}
  n88 = Multiply(A=@n84, B=@n87)
  n89 = Divide(A=Multiply(A=Floor(None=@n88), B=ScalarParameter() {defaultValue=1.4333, parameterName=offsetRandom}), B=@n87)
  n90 = TextureSample(UVs=Add(A=Multiply(A=@n85, B=@n86), B=@n89)) {samplerType=SAMPLERTYPE_LinearColor, texture=&T_ElectricNoise}
  n91 = ScalarParameter() {defaultValue=2.6805, parameterName=ScaleElectricNoise2}
  n92 = TextureSample(UVs=Add(A=Multiply(A=@n85, B=@n91), B=@n89)) {samplerType=SAMPLERTYPE_LinearColor, texture=&T_ElectricNoise}
  n93 = Power(
   Base=Multiply(A=ScalarParameter() {defaultValue=2, parameterName=BrightnesselectricityFalloff}, B=Add(A=@n90.R, B=@n92.R))
   Exp=ScalarParameter() {defaultValue=2, parameterName=ContrastElectricityFalloff}
  )
  n94 = ComponentMask(None=VertexNormalWS()) {r=true}
  n95 = Saturate(None=MaterialFunctionCall(In=Abs(None=@n94), Contrast=ScalarParameter() {defaultValue=8.8349, parameterName=TransitionBlendQuantizedNoise}) {materialFunction=&CheapContrast}.Result)
  n96 = Multiply(A=@n69, B=ScalarParameter() {defaultValue=32, parameterName=NumTiles})
  n97 = ScalarParameter() {defaultValue=2, parameterName=Scale Janus}
  n98 = Panner(
   Coordinate=@n69
   Speed=AppendVector(
    A=ScalarParameter() {defaultValue=0.022, parameterName=SpeedPan_Janus3_x}
    B=ScalarParameter() {defaultValue=0.033, parameterName=SpeedPan_Janus3_y}
   )
  )
  n99 = ScalarParameter() {defaultValue=32, parameterName=Quant}
  n100 = Multiply(A=Subtract(A=@n98, B=@n69), B=@n99)
  n101 = Subtract(A=Multiply(A=@n98, B=@n99), B=@n100)
  n102 = Divide(A=Add(A=Floor(None=@n101), B=@n100), B=@n99)
  n103 = TextureSample(UVs=@n102) {
   samplerSource=SSM_Wrap_WorldGroupSettings
   samplerType=SAMPLERTYPE_LinearColor
   texture=&T_ElectricNoise
  }
  n104 = TextureSample(UVs=@n102) {samplerSource=SSM_Wrap_WorldGroupSettings, samplerType=SAMPLERTYPE_Masks, texture=/Game/Athena/SafeZone/Textures/T_JanusShield5_M}
  n105 = Max(A=@n103.R, B=@n104.R)
  n106 = TextureCoordinate()
  n107 = TextureSample!(
   UVs=MaterialFunctionCall(UVs=Frac(None=@n96), Texture Scale=Multiply(A=@n97, B=@n105)) {materialFunction=&ScaleUVsByCenter}.UVs
   Tex
   "DDX(UVs)"=DDX(Value=@n106)
   "DDY(UVs)"=DDY(Value=@n106)
   Apply View MipBias
  ) {
   mipValueMode=TMVM_Derivative
   samplerSource=SSM_Clamp_WorldGroupSettings
   samplerType=SAMPLERTYPE_Masks
   texture=/Game/Athena/SafeZone/Textures/T_JanusShield_M
  }
  n108 = Add(A=@n103.R, B=Multiply(A=@n104.R, B=ScalarParameter() {defaultValue=0.5, parameterName=BrightnessJanusLine}))
  n109 = Multiply(A=@n84, B=ScalarParameter() {defaultValue=32, parameterName=NumTiles})
  n110 = ScalarParameter() {defaultValue=2, parameterName=Scale Janus}
  n111 = Panner(
   Coordinate=@n84
   Speed=AppendVector(
    A=ScalarParameter() {defaultValue=0.022, parameterName=SpeedPan_Janus3_x}
    B=ScalarParameter() {defaultValue=0.033, parameterName=SpeedPan_Janus3_y}
   )
  )
  n112 = ScalarParameter() {defaultValue=32, parameterName=Quant}
  n113 = Multiply(A=Subtract(A=@n111, B=@n84), B=@n112)
  n114 = Subtract(A=Multiply(A=@n111, B=@n112), B=@n113)
  n115 = Divide(A=Add(A=Floor(None=@n114), B=@n113), B=@n112)
  n116 = TextureSample!(UVs=@n115, Tex, Apply View MipBias) {
   samplerSource=SSM_Wrap_WorldGroupSettings
   samplerType=SAMPLERTYPE_LinearColor
   texture=&T_ElectricNoise
  }
  n117 = TextureSample(UVs=@n115) {samplerSource=SSM_Wrap_WorldGroupSettings, samplerType=SAMPLERTYPE_Masks, texture=&T_JanusShield5_M}
  n118 = Max(A=@n116.R, B=@n117.R)
  n119 = TextureCoordinate()
  n120 = TextureSample!(
   UVs=MaterialFunctionCall(UVs=Frac(None=@n109), Texture Scale=Multiply(A=@n110, B=@n118)) {materialFunction=&ScaleUVsByCenter}.UVs
   Tex
   "DDX(UVs)"=DDX(Value=@n119)
   "DDY(UVs)"=DDY(Value=@n119)
   Apply View MipBias
  ) {
   mipValueMode=TMVM_Derivative
   samplerSource=SSM_Clamp_WorldGroupSettings
   samplerType=SAMPLERTYPE_Masks
   texture=&T_JanusShield_M
  }
  n121 = Add(A=@n116.R, B=Multiply(A=@n117.R, B=ScalarParameter() {defaultValue=0.5, parameterName=BrightnessJanusLine}))
  n122 = Power(
   Base=LinearInterpolate(A=Multiply(A=@n107.R, B=@n108), B=Multiply(A=@n120.R, B=@n121), Alpha=@n95)
   Exp=ScalarParameter() {defaultValue=1.2, parameterName=Contrast}
  )
  n123 = Add(A=Multiply(A=LinearInterpolate(A=@n78, B=@n93, Alpha=@n95), B=@n122), B=@n122)
  n124 = VectorParameter() {parameterName=ColorGrid, defaultValue=<rgba 0 7.1076 10 1>}
  n125 = WorldPosition()
  n126 = ObjectPositionWS()
  n127 = Normalize(VectorInput=Multiply(A=Constant3Vector() {constant=<rgba 1 1 0 0>}, B=Subtract(A=@n125.XYZ, B=@n126)))
  n128 = MaterialFunctionCall(Multiplier=Constant() {r=2}) {materialFunction=&Pi}
  n129 = Multiply(A=MaterialFunctionCall() {materialFunction=&ObjectScale}.Scale X) {constB=50}
  n130 = Multiply(
   A=Subtract(A=MaterialFunctionCall(Vector or UVs=@n127) {materialFunction=&VectorToRadialValue}.Vector Converted to Angle) {constB=0.5}
   B=Divide(A=Multiply(A=@n128.Result, B=@n129), B=ScalarParameter() {defaultValue=22397.4902, parameterName=RadialScaleX})
  )
  n131 = ComponentMask(None=WorldPosition().XYZ) {b=true}
  n132 = ScalarParameter() {defaultValue=-21175.7793, parameterName=ZScale}
  n133 = Multiply(A=AppendVector(A=Add(A=@n130) {constB=0.5}, B=Divide(A=@n131, B=@n132)), B=ScalarParameter() {defaultValue=1.1842, parameterName=ScaleJanus})
  n134 = Panner(
   Coordinate=@n133
   Speed=AppendVector(
    A=ScalarParameter() {defaultValue=0.019, parameterName=SpeedPan_JanusNoise_x}
    B=ScalarParameter() {defaultValue=0.02, parameterName=SpeedPan_JanusNoise_y}
   )
  )
  n135 = ScalarParameter() {defaultValue=6.0141, parameterName=ScaleElectricNoise}
  n136 = ScalarParameter() {defaultValue=32, parameterName=Quant}
  n137 = Multiply(A=@n133, B=@n136)
  n138 = Divide(A=Multiply(A=Floor(None=@n137), B=ScalarParameter() {defaultValue=1.4333, parameterName=offsetRandom}), B=@n136)
  n139 = TextureSample!(UVs=Add(A=Multiply(A=@n134, B=@n135), B=@n138), Tex, Apply View MipBias) {samplerType=SAMPLERTYPE_LinearColor, texture=&T_ElectricNoise}
  n140 = ScalarParameter() {defaultValue=2.6805, parameterName=ScaleElectricNoise2}
  n141 = TextureSample(UVs=Add(A=Multiply(A=@n134, B=@n140), B=@n138)) {samplerType=SAMPLERTYPE_LinearColor, texture=&T_ElectricNoise}
  n142 = Power(
   Base=Multiply(A=ScalarParameter() {defaultValue=2, parameterName=BrightnesselectricityFalloff}, B=Add(A=@n139.R, B=@n141.R))
   Exp=ScalarParameter() {defaultValue=2, parameterName=ContrastElectricityFalloff}
  )
  n143 = Multiply(A=@n133, B=ScalarParameter() {defaultValue=32, parameterName=NumTiles})
  n144 = ScalarParameter() {parameterName=Scale Janu2s}
  n145 = ScalarParameter() {defaultValue=2, parameterName=Scale Janus}
  n146 = WorldPosition()
  n147 = RuntimeVirtualTextureSampleParameter(World Position=@n146.XYZ) {
   materialType=WorldHeight
   parameterName=RVT Landscape Height Target
   virtualTexture=&RVT_Apollo_Landscape_Height
  }
  n148 = Distance(A=@n147.WorldHeight, B=ComponentMask(None=@n146.XYZ) {b=true})
  n149 = Saturate(None=Divide(A=@n148, B=ScalarParameter() {defaultValue=25000, parameterName=JanusScanlinesFadeHeight}))
  n150 = OneMinus(None=@n149)
  n151 = Divide(A=@n148, B=ScalarParameter() {defaultValue=5000, parameterName=JanusScanlinesFrequency})
  n152 = Frac(None=Multiply(A=Time(), B=ScalarParameter() {defaultValue=-0.2, parameterName=JanusScanlinesSpeed}))
  n153 = TextureSample(UVs=AppendVector(A=@n149, B=Add(A=@n151, B=@n152))) {samplerSource=SSM_Wrap_WorldGroupSettings, samplerType=SAMPLERTYPE_LinearGrayscale, texture=&/T_StormShield_Scanline}
  n154 = Multiply(A=@n150, B=@n153.R)
  n155 = ScalarParameter() {defaultValue=4, parameterName=JanusScanlineExponent}
  n156 = Multiply(A=Power(Base=@n150) {constExponent=10}) {constB=2}
  n157 = Clamp(None=Add(A=Power(Base=@n154, Exp=@n155), B=@n156))
  n158 = TextureCoordinate()
  n159 = TextureSample!(
   UVs=MaterialFunctionCall(UVs=Frac(None=@n143), Texture Scale=LinearInterpolate(A=@n144, B=@n145, Alpha=@n157)) {materialFunction=&ScaleUVsByCenter}.UVs
   Tex
   "DDX(UVs)"=DDX(Value=@n158)
   "DDY(UVs)"=DDY(Value=@n158)
   Apply View MipBias
  ) {
   mipValueMode=TMVM_Derivative
   samplerSource=SSM_Clamp_WorldGroupSettings
   samplerType=SAMPLERTYPE_Masks
   texture=&T_JanusShield_M
  }
  n160 = Power(Base=Multiply(A=@n159.R, B=@n157), Exp=ScalarParameter() {defaultValue=1.2, parameterName=Contrast})
  n161 = Add(A=Multiply(A=LinearInterpolate(Alpha=@n142) {constA=0.5}, B=@n160), B=@n160)
  n162 = VectorParameter() {parameterName=ColorGrid, defaultValue=<rgba 0 7.1076 10 1>}
  n163 = Multiply(
   A=StaticSwitchParameter(True=Multiply(A=@n123, B=@n124.RGB), False=Multiply(A=@n161, B=@n162.RGB)) {defaultValue=true, parameterName=WorldSpace}
   B=Clamp(None=Add(A=@n154, B=@n156))
  )
  n164 = Multiply(A=VectorParameter() {parameterName=JanusScanlineColor, defaultValue=<rgba 0.0949 1.8207 2 1>}.RGB, B=@n157)
  n165 = Add(A=Multiply(A=Saturate(None=@n62), B=Add(A=@n163, B=@n164)), B=@n61)
  n166 = ShadingPathSwitch(Default=QualitySwitch(Default=@n165, Low=@n61, High=@n61, Medium=@n61, Epic=@n165), Mobile=@n61) {desc=Migrated from FeatureLevelSwitch}
  n167 = ShadingPathSwitch(
   Default=ScalarParameter() {defaultValue=1, parameterName=BrightnessMorning}
   Mobile=ScalarParameter() {defaultValue=1.4, parameterName=BrightnessMorning_Mobile}
  ) {desc=Migrated from FeatureLevelSwitch}
  n168 = ShadingPathSwitch(
   Default=ScalarParameter() {defaultValue=1, parameterName=BrightnessDay}
   Mobile=ScalarParameter() {defaultValue=1.8, parameterName=BrightnessDay_Mobile}
  ) {desc=Migrated from FeatureLevelSwitch}
  n169 = ShadingPathSwitch(
   Default=ScalarParameter() {defaultValue=1, parameterName=BrightnessEvening}
   Mobile=ScalarParameter() {defaultValue=1.5, parameterName=BrightnessEvening_Mobile}
  ) {desc=Migrated from FeatureLevelSwitch}
  n170 = ShadingPathSwitch(
   Default=ScalarParameter() {defaultValue=1, parameterName=BrightnessNight}
   Mobile=ScalarParameter() {defaultValue=1.5, parameterName=BrightnessNight_Mobile}
  ) {desc=Migrated from FeatureLevelSwitch}
  EyeAdaptationInverse(LightValueInput=Multiply(
    A=StaticSwitchParameter(True=@n166, False=@n61) {defaultValue=true, parameterName=IsJanusShield}
    B=MaterialFunctionCall(Value_Morning=@n167, Value_Day=@n168, Value_Evening=@n169, Value_Night=@n170) {materialFunction=&Scalar_Time_of_Day}.Result
   ))
  MaterialFunctionCall!(Distance Bias=Constant(), Distance Divisor=@n52, Scene Depth, Pixel Depth) {materialFunction=/Game/Effects/Fort_Effects/Materials/_Functions/MF_DistanceBehindTranslucency_01}
  MaterialFunctionCall(Distance Bias=Constant(), Distance Divisor=@n28) {materialFunction=&MF_DistanceBehindTranslucency_01}
  n171 = Multiply(A=@n59, B=ScalarParameter() {defaultValue=1.5, parameterName=OpacityMobileMult})
  Reroute(None=Multiply(
    A=ShadingPathSwitch(Default=@n59, Mobile=@n171) {desc=Migrated from FeatureLevelSwitch}
    B=ScalarParameter() {defaultValue=1, parameterName=OpacityOverride, sliderMax=1}
   ))
  ScalarParameter() {defaultValue=10, parameterName=GroundHighlightExponent}
  ScalarParameter() {defaultValue=2, parameterName=GroundHighlightIntensity}
  n172 = Constant()
  n173 = If(
   A=@n55.V2 Length
   B=Multiply(A=@n57, B=ScalarParameter() {defaultValue=1.1, parameterName=Buffer}) {constB=1.36, desc=add some buffer}
   "A > B"=Constant3Vector() {constant=<rgba 0 0 -999999 0>}
   "A == B"=@n172
   "A < B"=@n172
  )
  n174 = ComponentMask(None=CollectionParameter() {collection=&MPC_SafeZoneClouds, parameterName=SafeZoneCapHeight}) {r=true}
  n175 = Add(
   A=Multiply(A=MaterialFunctionCall() {materialFunction=&ObjectScale}.Scale Z) {constB=50}
   B=ComponentMask(None=ObjectPositionWS()) {b=true}
  )
  n176 = Add(A=AppendVector(A=Constant2Vector(), B=Subtract(A=@n174, B=@n175)), B=@n173)
  StaticSwitchParameter(True=@n176, False=ShadingPathSwitch(Default=@n176, Mobile=@n173) {desc=Migrated from FeatureLevelSwitch}) {parameterName=DisableMobileOptimizations}
 ]
}
