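# Masked, two-sided Athena storm cloud material that opens and dissipates around the safe zone.
# MPC_SafeZoneClouds location and scale drive radial gradients that offset clouds vertically and height-lerp an eroded opacity mask with burn-out color.
# World-projected, counter-scrolling cloud height and normal textures add waves and detail, with flowmap distortion and bump offset.
# Color blends light bleed-through with camera-distance-masked Fresnel rims, while a sphere-masked vertical offset bends the horizon.
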Material /Game/Athena/SafeZone/M_SafeZoneStorm
graph {
 properties {
  bFullyRough=true
  bTangentSpaceNormal=false
  bUseLightmapDirectionality=false
  bUseTranslucencyVertexFog=false
  bUsedWithNanite=true
  blendMode=BLEND_Masked
  pixelDepthOffsetMode=PDOM_Legacy
  translucencyLightingMode=TLM_SurfacePerPixelLighting
  translucencyPass=MTP_BeforeDOF
  twoSided=true
  usedShadingModels=Default Lit
 }
 nodes [
  n1 = ScalarParameter() {defaultValue=300000, group=StormDissipate, parameterName=LargestShieldRadius_1}
  n2 = ComponentMask(None=CollectionParameter() {collection=&/MPC_SafeZoneClouds, parameterName=SafeZoneLoc}) {g=true, r=true}
  n3 = Add(
   A=LinearInterpolate(
    A=CollectionParameter() {collection=&MPC_SafeZoneClouds, parameterName=SafeZoneScale}
    Alpha=CollectionParameter() {collection=&MPC_SafeZoneClouds, parameterName=PreventStormShrinking}
   )
   B=Multiply(A=CollectionParameter() {collection=&MPC_SafeZoneClouds, parameterName=CloudStartingBias}) {constB=4}
  )
  n4 = MaterialFunctionCall(
   UVs=Divide(A=WorldPosition().XYZ, B=@n1)
   CenterPosition=Divide(A=@n2, B=@n1)
   Radius=LinearInterpolate(A=Constant(), B=Constant() {r=1}, Alpha=@n3)
   Density=LinearInterpolate(
    A=ScalarParameter() {defaultValue=2.33, group=StormDissipate, parameterName=FalloffDensitySmall}
    B=ScalarParameter() {defaultValue=2.33, group=StormDissipate, parameterName=FalloffDensityLarge}
    Alpha=@n3
   )
  ) {materialFunction=&RadialGradientExponential}
  n5 = Multiply(A=@n4.RadialGradientExponential, B=ScalarParameter() {defaultValue=45000, parameterName=CloudSafeZoneWPOffset})
  n6 = SphereMask(
   A=ScalarParameter() {defaultValue=128, parameterName=LFCH-D}
   B=WorldPosition().XYZ
   Radius=ScalarParameter() {defaultValue=450000, parameterName=LFCH-Radius}
   Hardness=Constant()
  ) {hardnessPercent=0}
  n7 = Multiply(
   A=LinearInterpolate(
    A=ScalarParameter() {defaultValue=-44000, parameterName=CloudHorizonWPOffsetEnd}
    B=ScalarParameter() {defaultValue=-35000, parameterName=CloudHorizonWPOffsetStart}
    Alpha=CollectionParameter() {collection=&MPC_SafeZoneClouds, parameterName=CloudStartingBias}
   )
   B=Power(Base=OneMinus(None=@n6), Exp=ScalarParameter() {defaultValue=5, parameterName=LFCH-Power})
  )
  n8 = ScalarParameter() {defaultValue=0.633, parameterName=PHTTile_6_Offset_1}
  n9 = WorldPosition()
  n10 = Divide(A=@n9.XYZ, B=ScalarParameter() {defaultValue=120000, parameterName=PHTTile_7})
  n11 = ComponentMask(None=@n10) {r=true}
  n12 = Constant() {r=0.5}
  n13 = Time()
  n14 = Multiply(A=ScalarParameter() {defaultValue=0.012, parameterName=PSG_4}, B=@n13)
  n15 = AppendVector(A=Add(A=Add(A=@n8, B=@n11), B=Multiply(A=@n12, B=@n14)), B=Add(A=ComponentMask(None=@n10) {g=true}, B=@n14))
  n16 = Constant() {r=0.5}
  n17 = Multiply(A=@n13, B=ScalarParameter() {defaultValue=-0.01, parameterName=PSG_5})
  n18 = Divide(A=@n9.XYZ, B=ScalarParameter() {defaultValue=100000, parameterName=PHTTile_2})
  n19 = AppendVector(A=Add(A=Multiply(A=@n16, B=@n17), B=ComponentMask(None=@n18) {r=true}), B=Add(A=ComponentMask(None=@n18) {g=true}, B=@n17))
  n20 = Multiply(
   A=TextureSample(UVs=@n15) {samplerType=SAMPLERTYPE_LinearGrayscale, texture=/Game/Global/Textures/Default/Tileable_Patterns/T_StormCloudDetail01_M}.R
   B=TextureSample(UVs=@n19) {samplerType=SAMPLERTYPE_LinearGrayscale, texture=&T_StormCloudDetail01_M}.R
  )
  n21 = Power(Base=@n20, Exp=ScalarParameter() {defaultValue=0.5, parameterName=WPOAmtNew1_2})
  n22 = VectorParameter() {parameterName=BaseColorB_1, defaultValue=<rgba 0 0 -1 1>}
  n23 = Multiply(
   A=Add(
    A=TextureSample(UVs=@n15) {samplerType=SAMPLERTYPE_Normal, texture=/Game/Global/Textures/Default/Tileable_Patterns/T_StormCloudDetail01_N}.RGB
    B=TextureSample(UVs=@n19) {samplerType=SAMPLERTYPE_Normal, texture=&T_StormCloudDetail01_N}.RGB
   )
   B=Constant3Vector() {constant=<rgba 1 1 0.5 0>}
  )
  n24 = Transform(None=Multiply(A=@n23, B=ComponentMask(None=@n22.RGB) {b=true, g=true, r=true}))
  Add(
   A=Multiply(A=Add(A=@n5, B=@n7), B=Constant3Vector() {constant=<rgba 0 0 1 0>})
   B=Multiply(A=Multiply(A=@n21, B=@n24), B=ScalarParameter() {defaultValue=-12000, parameterName=CloudWavesWPOffset})
  )
  n25 = Constant3Vector() {constant=<rgba 0.5 0 0 0>}
  n26 = Constant3Vector() {constant=<rgba 0 0.5 0 0>}
  Add(
   A=Multiply(A=Multiply(A=@n25), B=ScalarParameter() {parameterName=SafeZoneRadiusDensity_1})
   B=Multiply(A=Multiply(A=@n26), B=ScalarParameter() {parameterName=SafeZoneRadiusDensity_2})
  )
  n27 = VectorParameter() {parameterName=FresnelRimLightColor_1, defaultValue=<rgba 0.0127 0.0167 0.05 1>}
  n28 = ScalarParameter() {defaultValue=0.633, parameterName=PHTTile_6_Offset}
  n29 = WorldPosition()
  n30 = Divide(A=@n29.XYZ, B=ScalarParameter() {defaultValue=300000, parameterName=PHTTile_6})
  n31 = ComponentMask(None=@n30) {r=true}
  n32 = Constant() {r=0.5}
  n33 = Time()
  n34 = Multiply(A=@n33, B=ScalarParameter() {defaultValue=0.012, parameterName=PSG_2})
  n35 = AppendVector(A=Add(A=Add(A=@n28, B=@n31), B=Multiply(A=@n32, B=@n34)), B=Add(A=ComponentMask(None=@n30) {g=true}, B=@n34))
  n36 = Constant() {r=0.5}
  n37 = Multiply(A=@n33, B=ScalarParameter() {defaultValue=-0.01, parameterName=PSG_3})
  n38 = Divide(A=@n29.XYZ, B=ScalarParameter() {defaultValue=200000, parameterName=PHTTile_4})
  n39 = AppendVector(A=Add(A=Multiply(A=@n36, B=@n37), B=ComponentMask(None=@n38) {r=true}), B=Add(A=ComponentMask(None=@n38) {g=true}, B=@n37))
  n40 = Multiply(
   A=Add(
    A=TextureSample(UVs=@n35) {samplerType=SAMPLERTYPE_Normal, texture=&T_StormCloudDetail01_N}.RGB
    B=TextureSample(UVs=@n39) {samplerType=SAMPLERTYPE_Normal, texture=&T_StormCloudDetail01_N}.RGB
   )
   B=Constant3Vector() {constant=<rgba 1 1 0.5 0>}
  )
  n41 = VectorParameter() {parameterName=BaseColorB_4, defaultValue=<rgba 1 1 -1 1>}
  n42 = Fresnel(
   ExponentIn=ScalarParameter() {defaultValue=1.5, parameterName=FrnExp_1}
   BaseReflectFractionIn=ScalarParameter() {defaultValue=0.04, parameterName=FrnBseRft_1}
   Normal=Multiply(A=@n40, B=ComponentMask(None=@n41.RGB) {b=true, g=true, r=true})
  ) {exponent=0.02}
  n43 = Clamp(None=@n42)
  n44 = SphereMask(A=CameraPositionWS(), B=WorldPosition().XYZ, Radius=ScalarParameter() {defaultValue=900000, parameterName=DB-BR}) {hardnessPercent=0}
  n45 = ScalarParameter() {defaultValue=10, parameterName=DB-SO}
  n46 = TextureSample(UVs=@n35) {samplerType=SAMPLERTYPE_LinearGrayscale, texture=&T_StormCloudDetail01_M}
  n47 = TextureSample(UVs=@n39) {samplerType=SAMPLERTYPE_LinearGrayscale, texture=&T_StormCloudDetail01_M}
  n48 = Power(Base=SquareRoot(None=Multiply(A=@n46.R, B=@n47.R)), Exp=ScalarParameter() {defaultValue=1, parameterName=WPOAmtNew1_1})
  n49 = ScalarParameter() {defaultValue=4.5, parameterName=BeersThick}
  Add(
   A=Multiply(A=Multiply(A=@n27.RGB, B=@n43), B=Power(Base=@n44, Exp=@n45))
   B=Multiply(
    A=VectorParameter() {parameterName=BaseColorB_2, defaultValue=<rgba 0.1722 0.0452 0.5 1>}.RGB
    B=MaterialFunctionCall!(Thickness=@n48, Depth Scale=@n49) {materialFunction=/Engine/Functions/Engine_MaterialFunctions01/Volumetrics/BeersLaw}.Result
   )
  )
  n50 = Multiply(A=Time(), B=ScalarParameter() {defaultValue=0.008, parameterName=n990082222_1})
  n51 = Divide(A=@n9.XYZ, B=ScalarParameter() {defaultValue=45000, parameterName=bb5ds3_1})
  n52 = ComponentMask(None=@n51) {r=true}
  n53 = TextureSample(UVs=AppendVector(A=Add(A=@n50, B=@n52), B=ComponentMask(None=@n51) {g=true})) {samplerType=SAMPLERTYPE_LinearColor, texture=/Game/VisualThreatManager/StormVisuals/LargeClouds/T_Swirl_Flowmap}
  n54 = Multiply(A=Subtract(A=ComponentMask(None=@n53.RGB) {g=true, r=true}) {constB=0.5}, B=ScalarParameter() {parameterName=bb43wdqww22_1})
  Add(A=@n54, B=@n15)
  Add(A=@n54, B=@n19)
  n55 = Divide(A=@n9.XYZ, B=ScalarParameter() {defaultValue=12000, parameterName=PHTTile_3})
  n56 = ComponentMask(None=@n55) {g=true}
  n57 = Multiply(A=Time(), B=ScalarParameter() {defaultValue=-0.01, parameterName=PSG_1})
  n58 = Divide(A=@n9.XYZ, B=ScalarParameter() {defaultValue=20000, parameterName=PHTTile_5})
  n59 = ComponentMask(None=@n58) {r=true}
  n60 = Add(A=ComponentMask(None=@n58) {g=true}, B=Multiply(A=Time(), B=ScalarParameter() {defaultValue=0.02, parameterName=gf34533wrwwe4_1}))
  n61 = ComponentMask(None=TextureSample(UVs=AppendVector(A=@n59, B=@n60)) {samplerType=SAMPLERTYPE_LinearColor, texture=&T_Swirl_Flowmap}.RGB) {g=true, r=true}
  n62 = Add(
   A=AppendVector(A=ComponentMask(None=@n55) {r=true}, B=Add(A=@n56, B=@n57))
   B=Multiply(A=Subtract(A=@n61) {constB=0.5}, B=ScalarParameter() {defaultValue=0.4, parameterName=vv44ew222222_2})
  )
  Add(A=@n54, B=@n62)
  n63 = ScalarParameter() {defaultValue=0.1, parameterName=BOHeightRatio_4}
  n64 = Reroute()
  n65 = Reroute()
  n66 = MaterialFunctionCall(Transition Phase=@n64, Height Texture=@n65) {materialFunction=&HeightLerp}
  Clamp(None=LinearInterpolate(
    A=ScalarParameter() {parameterName=TopOpacityMaskHigh_1}
    B=ScalarParameter() {defaultValue=1, parameterName=TopOpacityMaskLow_1}
    Alpha=Power(Base=@n66.Alpha, Exp=@n63)
   ))
  Constant()
  Constant() {r=1}
  Constant() {r=150000}
  Constant() {r=100000}
  Constant() {r=0.1}
  Constant() {r=60000}
  Constant() {r=12000}
  Constant() {r=20000}
  Constant() {r=-0.5}
  Constant() {r=12000}
  Constant() {r=20000}
  Constant() {r=-0.5}
  Constant() {r=170000}
  Constant() {r=195000}
  n67 = Time()
  LinearInterpolate(Alpha=ConstantBiasScale(None=Sine(None=@n67) {period=25})) {constB=3.5}
  LinearInterpolate(A=Constant3Vector() {constant=<rgba 1 0.016 0 1>}, B=Constant3Vector() {constant=<rgba 0 0.991 1 1>})
  n68 = Power(Base=@n64, Exp=ScalarParameter() {defaultValue=2, parameterName=BOHeightRatio_7})
  MaterialFunctionCall(Transition Phase=@n68, Height Texture=@n65) {materialFunction=&HeightLerp}
  n69 = MaterialFunctionCall!(
   HeightTexture=TextureSample(UVs=@n62) {samplerType=SAMPLERTYPE_LinearGrayscale, texture=&T_StormCloudDetail01_M}.R
   TextureCoordinates=@n62
   ReferencePlane=ScalarParameter() {defaultValue=1, parameterName=BORefPlane_1}
   HeightRatioInput=ScalarParameter() {defaultValue=-0.3, parameterName=BOHeightRatio_1}
   CameraVector
  ) {materialFunction=/Engine/Functions/Engine_MaterialFunctions02/Utility/BumpOffset_Advanced.BumpOffset_advanced}
  n70 = TextureSample(UVs=@n69.UVs) {samplerType=SAMPLERTYPE_Normal, texture=&T_StormCloudDetail01_N}
  n71 = Divide(A=@n9.XYZ, B=ScalarParameter() {defaultValue=14000, parameterName=PHTTile_8})
  n72 = Time()
  n73 = ScalarParameter() {defaultValue=-0.01, parameterName=PSG_6}
  n74 = AppendVector(A=ComponentMask(None=@n71) {r=true}, B=Add(A=ComponentMask(None=@n71) {g=true}, B=Multiply(A=@n72, B=@n73)))
  n75 = Divide(A=@n9.XYZ, B=ScalarParameter() {defaultValue=50000, parameterName=PHTTile_9})
  n76 = Time()
  n77 = ScalarParameter() {defaultValue=0.02, parameterName=gf34533wrwwe4_2}
  n78 = AppendVector(A=ComponentMask(None=@n75) {r=true}, B=Add(A=ComponentMask(None=@n75) {g=true}, B=Multiply(A=@n76, B=@n77)))
  n79 = Subtract(A=ComponentMask(None=TextureSample(UVs=@n78) {samplerType=SAMPLERTYPE_LinearColor, texture=&T_Swirl_Flowmap}.RGB) {g=true, r=true}) {constB=0.5}
  n80 = ScalarParameter() {defaultValue=0.5, parameterName=vv44ew222222_1}
  n81 = MaterialFunctionCall(
   HeightTexture=TextureSample(UVs=Add(A=@n54, B=@n74)) {samplerType=SAMPLERTYPE_LinearGrayscale, texture=&T_StormCloudDetail01_M}.R
   TextureCoordinates=Add(A=@n74, B=Multiply(A=@n79, B=@n80))
   ReferencePlane=ScalarParameter() {defaultValue=1, parameterName=BORefPlane_2}
   HeightRatioInput=ScalarParameter() {defaultValue=-0.3, parameterName=BOHeightRatio_3}
  ) {materialFunction=&BumpOffset_advanced}
  n82 = TextureSample(UVs=@n69.UVs) {samplerType=SAMPLERTYPE_LinearGrayscale, texture=&T_StormCloudDetail01_M}
  MaterialFunctionCall!(
   Diffuse 1=@n70.RGB
   Diffuse 2=TextureSample(UVs=@n81.UVs) {samplerType=SAMPLERTYPE_Normal, texture=&T_StormCloudDetail01_N}.RGB
   Transistion Phase=@n21
   Height Texture 1=@n82.R
   Height Texture 2=TextureSample(UVs=@n81.UVs) {samplerType=SAMPLERTYPE_LinearGrayscale, texture=&T_StormCloudDetail01_M}.R
   Contrast
  ) {materialFunction=/Engine/Functions/Engine_MaterialFunctions02/Texturing/HeightLerpWithTwoHeightMaps}
  Multiply(B=VectorParameter() {parameterName=2SidedSide_1, defaultValue=<rgba 1 1 1 1>}.RGB)
  n83 = Multiply(A=@n40, B=ComponentMask(None=VectorParameter() {parameterName=BaseColorB_3, defaultValue=<rgba 0 0 1 1>}.RGB) {b=true, g=true, r=true})
  Multiply(A=Multiply(A=@n48, B=Transform(None=@n83)), B=ScalarParameter() {defaultValue=15000, parameterName=WPOAmtNew1})
  n84 = Power(Base=@n66.Alpha, Exp=ScalarParameter() {defaultValue=2, parameterName=OpacityPower-SS}) {constExponent=5}
  OneMinus(None=@n84)
  OneMinus(None=@n42)
  n85 = SphereMask(A=CameraPositionWS(), B=WorldPosition().XYZ, Radius=ScalarParameter() {defaultValue=900000, parameterName=DB-BR_1}) {hardnessPercent=0}
  Power(Base=@n85, Exp=ScalarParameter() {defaultValue=1, parameterName=DB-SO_1})
  n86 = OneMinus(None=@n68)
  n87 = ScalarParameter() {defaultValue=4, parameterName=SSColorMaskPower}
  Reroute(None=Multiply(A=Power(Base=@n86, Exp=@n87) {constExponent=5}, B=VectorParameter() {parameterName=SSColor, defaultValue=<rgba 0 0 0 1>}.RGB))
  n88 = Power(Base=@n86, Exp=ScalarParameter() {defaultValue=2, parameterName=OpacityPower-SS_3}) {constExponent=5}
  n89 = Multiply(
   A=Add(A=@n84, B=ScalarParameter() {parameterName=OpacityPower-SS_1})
   B=Add(A=OneMinus(None=@n88), B=ScalarParameter() {defaultValue=1, parameterName=OpacityPower-SS_2})
  )
  Reroute(None=LinearInterpolate(A=Clamp(None=@n89), Alpha=ScalarParameter() {defaultValue=1, parameterName=OpacityPower-SS_4}))
  n90 = LinearInterpolate(A=Constant3Vector() {constant=<rgba 0 0 1 0>}, B=@n70.RGB, Alpha=@n20)
  n91 = Constant3Vector() {constant=<rgba 1 1 0.5 0>}
  n92 = Multiply(A=Add(A=@n23, B=Multiply(A=@n90, B=@n91)), B=Constant3Vector() {constant=<rgba 1 1 0.5 0>})
  n93 = Multiply(A=@n92, B=VectorParameter() {parameterName=2SidedSide, defaultValue=<rgba 1 1 -1.5 1>}.RGB)
  Reroute(None=Transform(None=@n93)) {desc=Normal}
  n94 = ScalarParameter() {defaultValue=250000, group=StormDissipate, parameterName=LargestShieldRadius}
  n95 = Divide(A=WorldPosition().XYZ, B=@n94)
  n96 = Divide(A=@n2, B=@n94)
  n97 = LinearInterpolate(A=Constant(), B=Constant() {r=1}, Alpha=@n3)
  n98 = LinearInterpolate(
   A=ScalarParameter() {defaultValue=2.33, group=StormDissipate, parameterName=ShieldDensitySmall}
   B=ScalarParameter() {defaultValue=2.33, group=StormDissipate, parameterName=ShieldDensityLarge}
   Alpha=@n3
  )
  n99 = LinearInterpolate(
   A=Power(Base=@n82.R, Exp=ScalarParameter() {defaultValue=5, parameterName=WPOAmtNew1_7})
   Alpha=OneMinus(None=MaterialFunctionCall(UVs=@n95, CenterPosition=@n96, Radius=@n97, Density=@n98) {materialFunction=&RadialGradientExponential}.RadialGradientExponential)
  )
  n100 = MaterialFunctionCall(Transition Phase=@n21, Height Texture=@n82.R) {materialFunction=&HeightLerp}
  n101 = Power(Base=@n100.Lerp Alpha No Contrast, Exp=ScalarParameter() {defaultValue=1, parameterName=WPOAmtNew1_6})
  n102 = Power(
   Base=Power(
    Base=MaterialFunctionCall(Transition Phase=@n20, Height Texture=@n101) {materialFunction=&HeightLerp}.Lerp Alpha No Contrast
    Exp=ScalarParameter() {defaultValue=0.5, parameterName=WPOAmtNew1_4}
   )
   Exp=ScalarParameter() {defaultValue=3, parameterName=BOHeightRatio_13}
  )
  n103 = ScalarParameter() {defaultValue=1, parameterName=BOHeightRatio_5}
  n104 = LinearInterpolate(A=Clamp(None=Multiply(A=@n102, B=@n103)), Alpha=Power(Base=@n20, Exp=ScalarParameter() {defaultValue=1, parameterName=BOHeightRatio_2}))
  Reroute(None=MaterialFunctionCall(Transition Phase=OneMinus(None=@n4.RadialGradientExponential), Height Texture=Multiply(A=@n99, B=@n104)) {materialFunction=&HeightLerp}.Lerp Alpha No Contrast) {desc=OpacityMask}
  n105 = ScalarParameter() {parameterName=BOHeightRatio_11}
  n106 = ScalarParameter() {defaultValue=1.5, parameterName=BOHeightRatio_9}
  n107 = ScalarParameter() {defaultValue=-0.2, parameterName=WPOAmtNew1_8}
  n108 = MaterialFunctionCall(
   Transition Phase=Multiply(A=LinearInterpolate(A=@n105, B=@n106, Alpha=@n4.RadialGradientExponential), B=Power(Base=@n82.R, Exp=@n107))
   Height Texture=OneMinus(None=@n104)
  ) {materialFunction=&HeightLerp}
  n109 = Power(Base=@n108.Lerp Alpha No Contrast, Exp=ScalarParameter() {defaultValue=4, parameterName=tyu568ttg_1})
  n110 = ScalarParameter() {defaultValue=2, parameterName=tyu568ttg_2}
  n111 = Multiply(
   A=Add(A=@n108.Lerp Alpha No Contrast, B=Multiply(A=@n109, B=@n110))
   B=VectorParameter() {group=StormDissipate, parameterName=BurnOutColor, defaultValue=<rgba 0.0075 0.0259 0.2 0>}.RGB
  )
  n112 = ScalarParameter() {defaultValue=0.2, parameterName=WPOAmtNew1_3}
  n113 = Power(
   Base=MaterialFunctionCall(Transition Phase=@n20, Height Texture=Power(Base=@n100.Lerp Alpha No Contrast, Exp=@n112)) {materialFunction=&HeightLerp}.Lerp Alpha No Contrast
   Exp=ScalarParameter() {defaultValue=0.5, parameterName=WPOAmtNew1_5}
  )
  n114 = Multiply(
   A=VectorParameter() {group=StormDissipate, parameterName=BurnOutColor_1, defaultValue=<rgba 0.038 0.0068 0.08 0>}.RGB
   B=Power(Base=OneMinus(None=@n113), Exp=ScalarParameter() {defaultValue=1.5, parameterName=SSColorMaskPower_1}) {constExponent=5}
  )
  n115 = VectorParameter() {group=StormDissipate, parameterName=BurnOutColor_2, defaultValue=<rgba 0.1691 0.06 1 0>}
  n116 = Add(A=@n82.R, B=@n20)
  n117 = Power(Base=OneMinus(None=Clamp(None=@n116)), Exp=ScalarParameter() {defaultValue=4, parameterName=SSColorMaskPower_2}) {constExponent=5}
  Reroute(None=Add(A=Add(A=@n111, B=@n114), B=Multiply(A=@n115.RGB, B=@n117)))
  n118 = VectorParameter() {parameterName=2SidedSide_3, defaultValue=<rgba 1 1 0 1>}
  n119 = Power(Base=@n20, Exp=ScalarParameter() {defaultValue=3, parameterName=BOHeightRatio_10})
  n120 = Multiply(
   A=Fresnel(
    ExponentIn=ScalarParameter() {defaultValue=3, parameterName=FrnExp_2}
    BaseReflectFractionIn=ScalarParameter() {defaultValue=0.04, parameterName=FrnBseRft_2}
    Normal=Multiply(A=@n118.RGB, B=@n92)
   ) {exponent=0.02}
   B=@n119
  )
  n121 = VectorParameter() {parameterName=FresnelRimLightColor_2, defaultValue=<rgba 0.0617 0.0562 0.865 1>}
  n122 = VectorParameter() {parameterName=FresnelRimLightColor, defaultValue=<rgba 0.1002 0.06 2 1>}
  n123 = Multiply(
   A=Fresnel(
    ExponentIn=ScalarParameter() {defaultValue=8, parameterName=FrnExp}
    BaseReflectFractionIn=ScalarParameter() {defaultValue=0.04, parameterName=FrnBseRft}
    Normal=@n93
   ) {exponent=0.02}
   B=@n119
  )
  n124 = Multiply(A=@n85, B=Add(A=Multiply(A=@n120, B=@n121.RGB), B=Multiply(A=@n122.RGB, B=@n123)))
  n125 = ScalarParameter() {defaultValue=1, parameterName=tg34tgwfwe_1}
  n126 = LinearInterpolate(
   A=VectorParameter() {parameterName=LightBleedThroughB, defaultValue=<rgba 0.0115 0.0028 0.05 1>}.RGB
   B=VectorParameter() {parameterName=BaseColorB, defaultValue=<rgba 0.0117 0.0023 0.08 1>}.RGB
   Alpha=@n113
  )
  n127 = ScalarParameter() {defaultValue=1, parameterName=tg34tgwfwe_3}
  Reroute(None=Add(A=Multiply(A=@n124, B=@n125), B=Multiply(A=@n126, B=@n127))) {desc=Color}
  SquareRoot(None=@n20)
  Transform(None=@n40)
 ]
}
