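# Translucent unlit material for the storm cloud extension around the safe zone, driven by safe-zone parameter collection values.
# Vertex offset math uses the U coordinate to pull geometry onto the zone's radius, with a huge Z push above a cloud-bias threshold.
# Time-panned flowmap distorts world-projected cloud detail, which is gradient-mapped, skylight-tinted and blended to fog color beyond the radius.
# A U-based lip falloff multiplies a scene-depth fade as a translucency mask.
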Material /Game/Athena/SafeZone/M_StormExtend
graph {
 properties {bUsedWithNanite=true, blendMode=BLEND_Translucent, pixelDepthOffsetMode=PDOM_Legacy, shadingModel=MSM_Unlit, usedShadingModels=Unlit}
 nodes [
  n1 = CollectionParameter() {collection=&/MPC_SafeZoneClouds, parameterName=SafeZoneScale}
  Add(
   A=@n1
   B=Multiply(
    A=CollectionParameter() {collection=&MPC_SafeZoneClouds, parameterName=CloudStartingBias}
    B=ScalarParameter() {defaultValue=4, parameterName=fudge multiply}
   )
  )
  n2 = TextureCoordinate()
  n3 = WorldPosition() {worldPositionShaderOffset=WPT_ExcludeAllShaderOffsets}
  n4 = ObjectPositionWS()
  n5 = ScalarParameter() {defaultValue=210886.9688, parameterName=SafeZoneMulti}
  n6 = Multiply(A=Normalize(VectorInput=Subtract(A=@n3.XYZ, B=@n4)), B=Multiply(A=@n5, B=@n1))
  n7 = ComponentMask(None=CollectionParameter() {collection=&MPC_SafeZoneClouds, parameterName=SafeZoneLoc}) {g=true, r=true}
  n8 = ComponentMask(None=ObjectPositionWS()) {b=true}
  n9 = Multiply(A=Subtract(A=WorldPosition() {worldPositionShaderOffset=WPT_ExcludeAllShaderOffsets}.XYZ, B=AppendVector(A=@n7, B=@n8))) {constB=-1}
  n10 = Constant()
  Add(
   A=Multiply(A=ComponentMask(None=@n2) {r=true}, B=Add(A=@n6, B=@n9))
   B=If(
    A=CollectionParameter() {collection=&MPC_SafeZoneClouds, parameterName=CloudStartingBias}
    B=Constant() {r=0.3}
    "A > B"=Constant3Vector() {constant=<rgba 0 0 10000000 0>}
    "A == B"=@n10
    "A < B"=@n10
   )
  )
  CollectionParameter() {collection=&MPC_SafeZoneClouds, parameterName=CloudStartingBias}
  Constant() {r=1}
  n11 = WorldPosition()
  n12 = Constant() {r=634212.8125}
  n13 = Add(A=ComponentMask(None=Divide(A=@n11.XYZ, B=@n12)) {g=true, r=true}, B=Multiply(A=Time(), B=Constant() {r=0.002}))
  n14 = Subtract(A=ComponentMask(None=TextureSample(UVs=@n13) {samplerType=SAMPLERTYPE_LinearColor, texture=/Game/VisualThreatManager/StormVisuals/LargeClouds/T_Swirl_Flowmap}.RGB) {g=true, r=true}) {constB=0.5}
  n15 = Constant() {r=0.8}
  n16 = Divide(A=@n11.XYZ, B=Constant() {r=313571.9375})
  n17 = TextureSample(UVs=Add(A=Multiply(A=@n14, B=@n15), B=ComponentMask(None=@n16) {g=true, r=true})) {samplerType=SAMPLERTYPE_LinearGrayscale, texture=/Game/Global/Textures/Default/Tileable_Patterns/T_StormCloudDetail01_M}
  n18 = Constant() {r=0.842}
  n19 = MaterialFunctionCall!(
   Greyscale Value To Gradient Map=Power(Base=Multiply(A=@n17.R, B=@n18), Exp=Constant() {r=1.3})
   "Index (0 Based)"
   Gradient Texture=TextureObject() {texture=&/Gradient}
   Number Of Gradients In the Gradient Map=Constant() {r=1}
  ) {materialFunction=/Engine/Functions/Engine_MaterialFunctions02/Gradients/GradientMap_Multi}
  n20 = Constant() {r=1}
  n21 = CollectionParameter() {
   collection=&FortniteMaterialParameters
   parameterName=SkyLightColorAndIntensity
  }
  n22 = ComponentMask(None=@n21) {b=true, g=true, r=true}
  n23 = Constant() {r=1}
  n24 = CollectionParameter() {
   collection=&FortniteMaterialParameters
   parameterName=FogDirectionalInscatteringColor
  }
  n25 = Constant() {r=0.3}
  n26 = ComponentMask(None=CollectionParameter() {collection=&MPC_SafeZoneClouds, parameterName=SafeZoneLoc}) {g=true, r=true}
  n27 = ComponentMask(None=WorldPosition().XYZ) {g=true, r=true}
  n28 = Subtract(
   A=MaterialFunctionCall(Vector 2=Subtract(A=@n26, B=@n27)) {materialFunction=/Engine/Functions/Engine_MaterialFunctions02/Utility/VectorLength}.V2 Length
   B=Multiply(A=@n5, B=CollectionParameter() {collection=&MPC_SafeZoneClouds, parameterName=SafeZoneScale})
  )
  n29 = Constant() {r=2381250.75}
  LinearInterpolate(
   A=Multiply(A=Multiply(A=@n19.RGB, B=@n20), B=Multiply(A=@n22, B=@n23))
   B=ComponentMask(None=Desaturation(None=@n24, Fraction=@n25)) {b=true, g=true, r=true}
   Alpha=Clamp(None=Divide(A=@n28, B=@n29))
  )
  n30 = ComponentMask(None=TextureCoordinate()) {r=true}
  n31 = ScalarParameter() {defaultValue=2083.7856, parameterName=LipPow}
  Multiply(
   A=OneMinus(None=Power(Base=@n30, Exp=@n31))
   B=MaterialFunctionCall!(Distance Bias=Constant(), Distance Divisor=ScalarParameter() {defaultValue=13000, parameterName=DepthFalloff}, Scene Depth, Pixel Depth) {materialFunction=/Game/Effects/Fort_Effects/Materials/_Functions/MF_DistanceBehindTranslucency_01}.Normalized
  )
 ]
}
