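# Mobile unlit, masked, two-sided storm cloud material for the Athena safe zone, driven by MPC_SafeZoneClouds.
# Vertex-color-weighted offsets fit the mesh to zone center and scale, shape an inner lip and drop distant clouds.
# Panned cloud samples plus flowmap-distorted detail build normals for fresnel rim colors, time-of-day tint and blue glow near the edge.
# A vertex-shader cloud sample computes displacement, and a HeightLerp derives an erosion mask from cloud alphas and vertex color.
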Material /Game/Athena/SafeZone/M_StormMobile
graph {
 properties {
  bUsedWithNanite=true
  blendMode=BLEND_Masked
  opacityMaskClipValue=0.1
  pixelDepthOffsetMode=PDOM_Legacy
  shadingModel=MSM_Unlit
  twoSided=true
  usedShadingModels=Unlit
 }
 nodes [
  n1 = VertexColor()
  n2 = ScalarParameter()
  n3 = Constant() {r=0.633}
  n4 = VertexColor()
  n5 = Normalize(VectorInput=Subtract(A=WorldPosition() {worldPositionShaderOffset=WPT_ExcludeAllShaderOffsets}.XYZ, B=ObjectPositionWS()))
  n6 = CollectionParameter() {collection=&/MPC_SafeZoneClouds, parameterName=CloudStartingBias}
  n7 = Add(
   A=LinearInterpolate(
    A=CollectionParameter() {collection=&MPC_SafeZoneClouds, parameterName=SafeZoneScale}
    Alpha=CollectionParameter() {collection=&MPC_SafeZoneClouds, parameterName=PreventStormShrinking}
   )
   B=@n6
  )
  n8 = Multiply(A=ScalarParameter() {defaultValue=199906.8438, parameterName=NewRadius}, B=@n7)
  n9 = Multiply(A=@n5, B=@n8)
  n10 = ComponentMask(None=CollectionParameter() {collection=&MPC_SafeZoneClouds, parameterName=SafeZoneLoc}) {g=true, r=true}
  n11 = ComponentMask(None=ObjectPositionWS()) {b=true}
  n12 = Multiply(A=Subtract(A=WorldPosition() {worldPositionShaderOffset=WPT_ExcludeAllShaderOffsets}.XYZ, B=AppendVector(A=@n10, B=@n11))) {constB=-1}
  n13 = Multiply(A=Clamp(None=@n7) {maxDefault=0.5}) {constB=2}
  n14 = LinearInterpolate(
   A=ScalarParameter() {defaultValue=-5515.9438, parameterName=InnerlipSmall}
   B=ScalarParameter() {defaultValue=-21886.5215, parameterName=InnerLipBig}
   Alpha=@n13
  )
  n15 = Add(A=Multiply(A=@n4.B, B=Add(A=@n9, B=@n12)), B=Multiply(A=VertexColor().G, B=Multiply(A=@n5, B=@n14)))
  n16 = Divide(A=Add(A=WorldPosition().XYZ, B=@n15), B=Constant() {r=101637.1484})
  n17 = ComponentMask(None=@n16) {r=true}
  n18 = Constant() {r=0.5}
  n19 = Time()
  n20 = Multiply(A=Constant() {r=0.012}, B=@n19)
  n21 = AppendVector(A=Add(A=Add(A=@n3, B=@n17), B=Multiply(A=@n18, B=@n20)), B=Add(A=ComponentMask(None=@n16) {g=true}, B=@n20))
  n22 = VertexInterpolator(VS=@n21)
  n23 = ScalarParameter() {defaultValue=0.8}
  n24 = AppendVector(A=Multiply(A=@n19, B=Constant() {r=-0.2}), B=Multiply(A=@n19, B=Constant() {r=-0.1}))
  n25 = Constant() {r=0.1}
  n26 = TextureSample(UVs=Add(A=Multiply(A=@n22.PS, B=@n23), B=Multiply(A=@n24, B=@n25))) {samplerType=SAMPLERTYPE_LinearColor, texture=/Game/VisualThreatManager/StormVisuals/LargeClouds/T_StormCloudMobile}
  n27 = TextureSample(UVs=@n22.PS) {samplerType=SAMPLERTYPE_LinearColor, texture=&T_StormCloudMobile}
  n28 = Add(
   A=ConstantBiasScale(None=@n27.RGB) {bias=-0.5, scale=2}
   B=Multiply(A=ConstantBiasScale(None=@n26.RGB) {bias=-0.5, scale=2}, B=Constant3Vector() {constant=<rgba 1 1 0 0>})
  )
  n29 = WorldPosition()
  n30 = Divide(A=@n29.XYZ, B=Constant() {r=11058.1074})
  n31 = ComponentMask(None=@n30) {r=true}
  n32 = Add(A=ComponentMask(None=@n30) {g=true}, B=Multiply(A=@n19, B=Constant() {r=-0.01}))
  n33 = Divide(A=@n29.XYZ, B=Constant() {r=34547.1133})
  n34 = ComponentMask(None=@n33) {r=true}
  n35 = Add(A=ComponentMask(None=@n33) {g=true}, B=Multiply(A=@n19, B=Constant() {r=0.01}))
  n36 = ComponentMask(None=TextureSample(UVs=AppendVector(A=@n34, B=@n35)) {samplerType=SAMPLERTYPE_LinearColor, texture=/Game/VisualThreatManager/StormVisuals/LargeClouds/T_Swirl_Flowmap}.RGB) {g=true, r=true}
  n37 = TextureSample(UVs=Add(A=AppendVector(A=@n31, B=@n32), B=ConstantBiasScale(None=@n36) {bias=-0.5, scale=0.8912})) {samplerType=SAMPLERTYPE_LinearColor, texture=&T_StormCloudMobile}
  n38 = Multiply(A=ConstantBiasScale(None=@n37.RGB) {bias=-0.5, scale=2}, B=Constant3Vector() {constant=<rgba 0.25 0.25 0 0>})
  n39 = Multiply(A=Normalize(VectorInput=Add(A=@n28, B=@n38)), B=TwoSidedSign())
  n40 = ConstantBiasScale(None=DotProduct(A=Transform(None=@n39), B=CameraVectorWS()))
  n41 = OneMinus(None=@n40)
  n42 = Constant() {r=4}
  n43 = VectorParameter() {parameterName=LightBleedThroughB, defaultValue=<rgba 0.006 0.0014 0.026 1>}
  n44 = VectorParameter() {parameterName=FresnelRimLightColor_2, defaultValue=<rgba 0.1923 0.13 2 1>}
  n45 = Power(Base=@n40, Exp=Constant() {r=6})
  n46 = Multiply(A=@n27.A, B=@n26.A)
  n47 = Multiply(A=@n46, B=@n46)
  n48 = Clamp(None=Add(A=@n45, B=Multiply(A=@n47, B=@n47)))
  n49 = MaterialFunctionCall(Color_Morning=Constant3Vector() {constant=<rgba 0.7292 0.3174 0.1861 1>}) {materialFunction=&Color_Time_of_Day}
  n50 = Add(
   A=Multiply(A=Constant3Vector() {constant=<rgba 0.0712 0.0156 0.2 0>}, B=Power(Base=@n41, Exp=@n42))
   B=Multiply(A=LinearInterpolate(A=@n43.RGB, B=@n44.RGB, Alpha=@n48), B=@n49.Result)
  )
  Add(B=Multiply(A=Multiply(A=@n1.G, B=@n2), B=@n50))
  n51 = Multiply(A=@n46, B=Constant() {r=1})
  n52 = Constant() {r=0.4368}
  n53 = Multiply(A=Add(A=@n51, B=Multiply(A=@n37.A, B=@n52)), B=Constant() {r=-1.8077})
  n54 = Constant() {r=-1.5364}
  n55 = ComponentMask(None=CollectionParameter() {collection=&MPC_SafeZoneClouds, parameterName=SafeZoneLoc}) {g=true, r=true}
  n56 = Add(A=@n15, B=WorldPosition() {worldPositionShaderOffset=WPT_ExcludeAllShaderOffsets}.XYZ)
  n57 = ComponentMask(None=@n56) {g=true, r=true}
  n58 = Add(A=@n14, B=ScalarParameter() {defaultValue=-2592.0859, parameterName=InnerRadiusBias})
  n59 = Clamp(None=Subtract(A=Distance(A=@n55, B=@n57), B=Add(A=@n8, B=@n58))) {clampMode=CMODE_ClampMin}
  n60 = LinearInterpolate(
   A=ScalarParameter() {defaultValue=54426.8203, parameterName=FalloffRadiusSmall}
   B=ScalarParameter() {defaultValue=120512.8594, parameterName=FalloffRadiusBig}
   Alpha=@n13
  )
  n61 = OneMinus(None=Clamp(None=Divide(A=@n59, B=@n60)))
  n62 = VertexInterpolator(VS=@n61)
  n63 = Multiply(A=@n62.PS, B=@n62.PS)
  n64 = Constant() {r=4.6579}
  n65 = Clamp(None=Add(A=Add(A=@n53, B=@n54), B=Multiply(A=@n63, B=@n64))) {clampMode=CMODE_ClampMin, maxDefault=2}
  Add(A=Multiply(A=ComponentMask(None=@n65) {r=true}, B=Constant3Vector() {constant=<rgba 0.0261 0.0907 0.7 0>}), B=@n50)
  n66 = TextureSample!(UVs=@n21, Tex, Level=Constant() {r=2}, Apply View MipBias) {
   mipValueMode=TMVM_MipLevel
   samplerType=SAMPLERTYPE_LinearColor
   texture=&T_StormCloudMobile
  }
  n67 = Multiply(A=@n66.A, B=@n66.A)
  n68 = Multiply(A=@n67, B=@n67)
  n69 = ScalarParameter() {defaultValue=-0.0952, parameterName=bias}
  Add(A=Multiply(A=Add(A=@n68, B=@n69), B=ScalarParameter() {defaultValue=-6000, parameterName=VertexMovementAmount}))
  n70 = VertexColor()
  Add(A=@n63, B=Multiply(A=Multiply(A=@n70.B, B=@n70.B), B=Constant() {r=0.1}))
  n71 = Multiply(A=ScalarParameter() {defaultValue=33899.4062, parameterName=InnerLipVertHeight}, B=Power(Base=@n61, Exp=Constant() {r=2.5}))
  n72 = ScalarParameter() {defaultValue=-50000, parameterName=CloudHorizonWPOffsetEnd}
  n73 = ScalarParameter() {defaultValue=-40000, parameterName=CloudHorizonWPOffsetStart}
  n74 = CollectionParameter() {collection=&MPC_SafeZoneClouds, parameterName=CloudStartingBias}
  n75 = OneMinus(None=SphereMask(
    A=ScalarParameter() {defaultValue=128, parameterName=LFCH-D}
    B=@n56
    Radius=ScalarParameter() {defaultValue=483516.9688, parameterName=LFCH-Radius}
    Hardness=Constant()
   ) {hardnessPercent=0})
  n76 = Constant() {r=4}
  n77 = Add(
   A=Multiply(A=LinearInterpolate(A=@n72, B=@n73, Alpha=@n74), B=Power(Base=@n75, Exp=@n76))
   B=Multiply(A=@n74, B=ScalarParameter() {defaultValue=-55722.7109, parameterName=FadeOutVertDrop})
  )
  Add(A=@n15, B=Multiply(A=Add(A=@n71, B=@n77), B=Constant3Vector() {constant=<rgba 0 0 1 0>}))
  Constant3Vector() {constant=<rgba 0.006 0.0014 0.026 1>}
  Constant3Vector() {constant=<rgba 0.0832 0.0562 0.865 1>}
  Constant() {r=3.5314}
  Constant() {r=159772.7344}
  Constant() {r=-1.0909}
  Constant()
  Constant()
  Multiply(B=@n6)
  Multiply(A=@n1.G, B=@n1.G)
  Multiply(A=@n45, B=Constant() {r=1})
  Multiply(A=Subtract(A=@n36) {constB=0.5}, B=ScalarParameter() {defaultValue=0.8912})
  Power(Base=@n46) {constExponent=3}
  Power(Base=@n4.B)
  n78 = VertexColor()
  n79 = ScalarParameter() {defaultValue=1, parameterName=VertColorMulti_1}
  n80 = Constant() {r=4}
  n81 = Add(A=@n51, B=Multiply(A=Power(Base=@n37.A, Exp=@n80), B=Constant() {r=0.4686}))
  n82 = Constant() {r=2.35}
  n83 = ScalarParameter() {defaultValue=-1.0909, parameterName=HightBias}
  n84 = ScalarParameter() {defaultValue=-3.2444, parameterName=HightBiasStart}
  n85 = MaterialFunctionCall(
   Input=CollectionParameter() {collection=&MPC_SafeZoneClouds, parameterName=CloudStartingBias}
   Input Low=Constant() {r=0.96}
   Input High=Constant() {r=1}
  ) {materialFunction=&RemapValue_0-1}
  n86 = MaterialFunctionCall(
   Transition Phase=OneMinus(None=Multiply(A=@n78.G, B=@n79))
   Height Texture=Add(A=Multiply(A=@n81, B=@n82) {constB=2}, B=LinearInterpolate(A=@n83, B=@n84, Alpha=@n85.Result))
  ) {materialFunction=&HeightLerp}
  Reroute(None=@n86.Lerp Alpha No Contrast)
  ScalarParameter() {defaultValue=-1.8077, parameterName=GlowHeightMultiply}
  ScalarParameter() {defaultValue=-1.2392, parameterName=GlowHeightBias}
  ScalarParameter() {defaultValue=1, parameterName=VertColorMulti}
  ScalarParameter() {defaultValue=-1.0909, parameterName=HightBias}
  Subtract(A=@n86.Lerp Alpha No Contrast, B=ScalarParameter() {parameterName=ErodeIn})
  VectorParameter() {group=StormDissipate, parameterName=BurnOutColor, defaultValue=<rgba 0.0224 0.0778 0.6 0>}
  VertexInterpolator(VS=@n49.Result)
 ]
}
