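# Unlit translucent material for the blended top of the safe-zone storm cloud ring.
# Scrolling bump-offset cloud textures and blended normals drive colors that shift with time of day, sun inscattering and a local-color sphere mask.
# World position offset math resizes the ring from safe-zone collection parameters, adding camera-relative extension, twist and sine waves.
# Opacity math combines vertex-color HeightLerp with depth fade, plus camera fade on the bottom-side switch.
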Material /Game/Athena/SafeZone/Materials/M_CloudRingTopBlend
graph {
 properties {bUsedWithNanite=true, blendMode=BLEND_Translucent, pixelDepthOffsetMode=PDOM_Legacy, shadingModel=MSM_Unlit, usedShadingModels=Unlit}
 nodes [
  n1 = VectorParameter() {parameterName=SineWaveDirection, defaultValue=<rgba 0 0 1 1>}
  Add(A=VertexNormalWS(), B=@n1.RGB)
  n2 = CollectionParameter() {
   collection=&FortniteMaterialParameters
   parameterName=FogDirectionalInscatteringColor
  }
  n3 = Multiply(
   A=Desaturation(None=ComponentMask(None=@n2) {b=true, g=true, r=true}, Fraction=ScalarParameter() {defaultValue=-0.4667, parameterName=SunSaturation})
   B=ScalarParameter() {defaultValue=0.5905, parameterName=SunBrightness}
  )
  n4 = ScalarParameter() {defaultValue=0.0035, parameterName=Speed_X}
  n5 = ScalarParameter() {parameterName=Speed_Y}
  n6 = Add(A=TextureCoordinate(), B=Multiply(A=Time(), B=AppendVector(A=@n4, B=@n5)))
  n7 = ScalarParameter() {defaultValue=5, parameterName=UVTiling_X}
  n8 = ScalarParameter() {defaultValue=0.7851, parameterName=UVTiling_Y}
  n9 = ComponentMask(None=TextureCoordinate()) {r=true}
  n10 = ScalarParameter() {defaultValue=1, parameterName=UVShift}
  n11 = Add(A=Multiply(A=Frac(None=@n6), B=AppendVector(A=@n7, B=@n8)), B=AppendVector(A=Constant(), B=Multiply(A=@n9, B=@n10)))
  n12 = TextureCoordinate()
  n13 = DDX(Value=@n12)
  n14 = DDY(Value=@n12)
  n15 = BumpOffset(
   Coordinate=@n11
   Height=TextureSample!(UVs=@n11, Tex, "DDX(UVs)"=@n13, "DDY(UVs)"=@n14, Apply View MipBias) {mipValueMode=TMVM_Derivative, samplerType=SAMPLERTYPE_Masks, texture=/Game/Athena/SafeZone/Textures/StormCloudtubes_M}.G
   HeightRatioInput=ScalarParameter() {defaultValue=-0.0476, parameterName=BumpOffset}
  )
  n16 = TextureSample(UVs=@n15, "DDX(UVs)"=@n13, "DDY(UVs)"=@n14) {mipValueMode=TMVM_Derivative, samplerType=SAMPLERTYPE_Masks, texture=&StormCloudtubes_M}
  n17 = Power(Base=@n16.G, Exp=ScalarParameter() {defaultValue=0.2475, parameterName=SkyBlendContrast})
  n18 = VectorParameter() {group=CoLors, parameterName=Color_BaseMult_Night, sortPriority=1, defaultValue=<rgba 0.3397 0.272 0.5677 1>}
  n19 = Constant3Vector() {constant=<rgba 0.224 0.1534 0.1306 1>}
  n20 = CollectionParameter() {collection=/Game/Athena/SafeZone/MPC_SafeZoneClouds, parameterName=LocalColorPosition}
  n21 = Power(Base=SphereMask(
    A=ComponentMask(None=@n20) {b=true, g=true, r=true}
    B=WorldPosition().XYZ
    Radius=CollectionParameter() {collection=&MPC_SafeZoneClouds, parameterName=LocalColorRadius}
   ) {attenuationRadius=250000, hardnessPercent=65}) {constExponent=0.7}
  n22 = VectorParameter() {group=CoLors, parameterName=Color_BaseMult_Day, sortPriority=2, defaultValue=<rgba 0.6919 0.5987 1.2 1>}
  n23 = Constant3Vector() {constant=<rgba 0.2656 0.2211 0.202 1>}
  n24 = Sine(None=Divide(
    A=CollectionParameter() {collection=&FortniteMaterialParameters, parameterName=TimeOfDay_0-1}
    B=Constant() {r=2}
   ))
  n25 = VectorParameter() {group=CoLors, parameterName=FresnelColor_Dark_Day, sortPriority=8, defaultValue=<rgba 0.0875 0.0452 0.0885 1>}
  n26 = LinearInterpolate(
   A=VectorParameter() {group=CoLors, parameterName=FresnelColor_Dar_Night, sortPriority=7, defaultValue=<rgba 0 0 0 1>}.RGB
   B=LinearInterpolate(A=@n25.RGB, B=Constant3Vector() {constant=<rgba 0.051 0.0298 0.0175 1>}, Alpha=@n21)
   Alpha=@n24
  )
  n27 = VectorParameter() {group=CoLors, parameterName=FresnelColor_Bright_Night, sortPriority=9, defaultValue=<rgba 0.2021 0.2184 0.3333 1>}
  n28 = VectorParameter() {group=CoLors, parameterName=FresnelColor_Bright_Day, sortPriority=10, defaultValue=<rgba 0.2717 0.1363 0.3177 1>}
  n29 = LinearInterpolate(
   A=LinearInterpolate(A=@n27.RGB, B=Constant3Vector() {constant=<rgba 0.2917 0.1486 0.1868 1>}, Alpha=@n21)
   B=LinearInterpolate(A=@n28.RGB, B=Constant3Vector() {constant=<rgba 0.4062 0.2728 0.201 1>}, Alpha=@n21)
   Alpha=@n24
  )
  n30 = TextureSample(UVs=Multiply(A=@n11, B=ScalarParameter() {defaultValue=5, parameterName=UVTiling_Detail}), "DDX(UVs)"=@n13, "DDY(UVs)"=@n14) {mipValueMode=TMVM_Derivative, samplerType=SAMPLERTYPE_Normal, texture=/Game/Athena/SafeZone/Textures/StormCloudtubes_N_2}
  n31 = ScalarParameter() {defaultValue=0.7995, parameterName=DetailNormal}
  n32 = Transform(None=MaterialFunctionCall(
    BaseNormal=TextureSample(UVs=@n15, "DDX(UVs)"=@n13, "DDY(UVs)"=@n14) {mipValueMode=TMVM_Derivative, samplerType=SAMPLERTYPE_Normal, texture=/Game/Athena/SafeZone/Textures/StormCloudtubes_N}.RGB
    AdditionalNormal=MaterialFunctionCall(Normal=@n30.RGB, Flatness=@n31) {materialFunction=&FlattenNormal}.Result
   ) {materialFunction=&BlendAngleCorrectedNormals}.Result)
  n33 = DotProduct(A=Constant3Vector() {constant=<rgba 0 0 1 0>}, B=@n32)
  n34 = ScalarParameter() {defaultValue=1.2015, parameterName=Fresnel_Exponent}
  n35 = Power(Base=@n33, Exp=@n34)
  n36 = VectorParameter() {group=CoLors, parameterName=Color_Height_Low_Night, sortPriority=3, defaultValue=<rgba 0.5814 0.3479 0.6042 1>}
  n37 = VectorParameter() {group=CoLors, parameterName=Color_Height_Low_Day, sortPriority=4, defaultValue=<rgba 0.5709 0.3043 0.7396 1>}
  n38 = LinearInterpolate(
   A=LinearInterpolate(A=@n36.RGB, B=Constant3Vector() {constant=<rgba 0.5 0.3991 0.3365 1>}, Alpha=@n21)
   B=LinearInterpolate(A=@n37.RGB, B=Constant3Vector() {constant=<rgba 0.7865 0.6636 0.6226 1>}, Alpha=@n21)
   Alpha=@n24
  )
  n39 = VectorParameter() {group=CoLors, parameterName=Color_Height_High_Night, sortPriority=5, defaultValue=<rgba 0.1856 0.0595 0.1979 1>}
  n40 = VectorParameter() {group=CoLors, parameterName=Color_Height_High_Day, sortPriority=6, defaultValue=<rgba 0.0698 0.0329 0.099 1>}
  n41 = LinearInterpolate(
   A=LinearInterpolate(A=@n39.RGB, B=Constant3Vector() {constant=<rgba 0.1667 0.1158 0.0842 1>}, Alpha=@n21)
   B=LinearInterpolate(A=@n40.RGB, B=Constant3Vector() {constant=<rgba 0.3333 0.2581 0.2205 1>}, Alpha=@n21)
   Alpha=@n24
  )
  n42 = Multiply(
   A=LinearInterpolate(A=LinearInterpolate(A=@n18.RGB, B=@n19, Alpha=@n21), B=LinearInterpolate(A=@n22.RGB, B=@n23, Alpha=@n21), Alpha=@n24)
   B=Add(A=LinearInterpolate(A=@n26, B=@n29, Alpha=@n35), B=LinearInterpolate(A=@n38, B=@n41, Alpha=@n16.G))
  )
  Add(A=Multiply(A=@n3, B=OneMinus(None=@n17)), B=@n42)
  n43 = CollectionParameter() {collection=&MPC_SafeZoneClouds, parameterName=SafeZoneLoc}
  n44 = ScalarParameter() {defaultValue=16902.3105, parameterName=PushRadius_Min}
  n45 = ScalarParameter() {defaultValue=224707.5312, parameterName=PushRadius}
  n46 = StaticSwitchParameter(
   True=Clamp(None=ScalarParameter() {defaultValue=-0.0166, parameterName=Debug_ZoneScale}) {clampMode=CMODE_ClampMin}
   False=CollectionParameter() {collection=&MPC_SafeZoneClouds, parameterName=SafeZoneScale}
  ) {parameterName=Debug}
  n47 = CollectionParameter() {collection=&MPC_SafeZoneClouds, parameterName=PreventStormShrinking}
  n48 = StaticSwitchParameter(
   True=ScalarParameter() {parameterName=Debug_CloudStartingBias}
   False=Max(A=@n47, B=CollectionParameter() {collection=&MPC_SafeZoneClouds, parameterName=CloudStartingBias})
  ) {parameterName=Debug_1}
  n49 = Add(A=@n46, B=@n48)
  n50 = ComponentMask(None=CollectionParameter() {collection=&MPC_SafeZoneClouds, parameterName=SafeZoneLoc}) {g=true, r=true}
  n51 = ComponentMask(None=CameraPositionWS()) {g=true, r=true}
  n52 = Add(A=Distance(A=@n50, B=@n51), B=ScalarParameter() {defaultValue=47245.6758, parameterName=FalloffExtendAdd})
  n53 = LinearInterpolate(A=@n45, B=@n45, Alpha=@n46)
  n54 = ComponentMask(None=WorldPosition().XYZ) {g=true, r=true}
  n55 = DotProduct(A=Normalize(VectorInput=Subtract(A=@n50, B=@n51)), B=Normalize(VectorInput=Subtract(A=@n54, B=@n50)))
  n56 = Multiply(
   A=Clamp(None=Subtract(A=@n52, B=@n53)) {clampMode=CMODE_ClampMin}
   B=Power(Base=Clamp(None=@n55), Exp=ScalarParameter() {defaultValue=4.6347, parameterName=FalloffExtendToCamera})
  )
  n57 = Multiply(A=OneMinus(None=VertexColor().R), B=@n56)
  n58 = LinearInterpolate(
   A=Add(A=LinearInterpolate(A=@n44, B=@n45, Alpha=@n49), B=StaticSwitchParameter(True=@n56, False=@n57) {parameterName=IsBottomSide})
   B=ScalarParameter() {defaultValue=2025478.25, parameterName=PushRadius_Start}
   Alpha=@n48
  )
  n59 = ScalarParameter() {defaultValue=0.0806, parameterName=PushRadius_TopMult_Min}
  n60 = ScalarParameter() {defaultValue=0.5759, parameterName=PushRadius_TopMult}
  n61 = Multiply(
   A=@n58
   B=LinearInterpolate(A=LinearInterpolate(A=@n59, B=@n60, Alpha=@n49), B=ScalarParameter() {defaultValue=1, parameterName=PushRadius_ToMult_Start}, Alpha=@n48)
  )
  n62 = VertexColor()
  n63 = ScalarParameter() {defaultValue=2.0548, parameterName=PowerAnimation}
  n64 = LinearInterpolate(
   A=LinearInterpolate(
    A=ScalarParameter() {defaultValue=25422.9375, parameterName=Height_BottomOffset_min}
    B=ScalarParameter() {defaultValue=48707.1055, parameterName=Height_BottomOffset_max}
    Alpha=@n49
   )
   B=ScalarParameter() {defaultValue=-110591.3359, parameterName=Height_BottomOffset_Start}
   Alpha=Clamp(None=Multiply(A=@n48, B=@n63))
  )
  n65 = LinearInterpolate(
   A=LinearInterpolate(
    A=ScalarParameter() {defaultValue=27333.3535, parameterName=HeightTopOffse_Min}
    B=ScalarParameter() {defaultValue=182099.9531, parameterName=HeightTopOffse_Max}
    Alpha=@n49
   )
   B=ScalarParameter() {defaultValue=-223687.6094, parameterName=HeightTopOffse_Start}
   Alpha=@n48
  )
  n66 = Add(
   A=Multiply(A=VertexNormalWS(), B=LinearInterpolate(A=@n58, B=@n61, Alpha=@n62.R))
   B=Multiply(A=Constant3Vector() {constant=<rgba 0 0 1 0>}, B=LinearInterpolate(A=@n64, B=@n65, Alpha=@n62.R))
  )
  n67 = Constant3Vector() {constant=<rgba 0 0 1 0>}
  n68 = Multiply(A=@n62.R, B=ScalarParameter() {defaultValue=0.1212, parameterName=Twist})
  n69 = ActorPositionWS()
  n70 = Add(A=@n66, B=WorldPosition().XYZ)
  n71 = Add(
   A=AppendVector(A=ComponentMask(None=@n43) {g=true, r=true}, B=Constant())
   B=Add(A=@n66, B=RotateAboutAxis(NormalizedRotationAxis=@n67, RotationAngle=@n68, PivotPoint=@n69, Position=@n70))
  )
  n72 = Add(A=@n71, B=WorldPosition() {worldPositionShaderOffset=WPT_ExcludeAllShaderOffsets}.XYZ)
  n73 = ScalarParameter() {defaultValue=166932.625, parameterName=WaveFequency_1}
  n74 = ScalarParameter() {defaultValue=146539.0156, parameterName=WaveFequency}
  n75 = Sine(None=Divide(A=ComponentMask(None=@n72) {g=true, r=true}, B=AppendVector(A=@n73, B=@n74)))
  n76 = ScalarParameter() {defaultValue=6618.9858, parameterName=WaveAmplitude}
  n77 = Multiply(A=Add(A=ComponentMask(None=@n75) {r=true}, B=ComponentMask(None=@n75) {g=true}), B=@n76)
  n78 = Multiply(A=@n48, B=ScalarParameter() {defaultValue=9.6897, parameterName=MultAnimation_SineWaveIntensity})
  Add(A=@n71, B=LinearInterpolate(A=Multiply(A=@n1.RGB, B=@n77), B=Constant3Vector(), Alpha=Clamp(None=@n78)))
  CollectionParameter() {collection=&MPC_SafeZoneClouds, parameterName=PreventStormShrinking}
  ComponentMask(None=MaterialFunctionCall(Number=@n24, UVs=TextureCoordinate()) {materialFunction=&DebugScalarValues}.Result) {b=true, g=true, r=true}
  ComponentMask() {g=true}
  Fresnel(ExponentIn=@n34, Normal=@n32) {baseReflectFraction=0}
  LinearInterpolate(A=@n3, B=@n42, Alpha=@n17)
  LinearInterpolate(A=@n46, Alpha=@n47)
  Multiply()
  Multiply(A=@n35, B=Sign(None=@n33))
  n79 = ScalarParameter() {defaultValue=0.3952, parameterName=WaveFade}
  Multiply(A=@n76, B=OneMinus(None=Multiply(B=@n79)))
  n80 = VertexColor()
  n81 = DepthFade(
   Opacity=MaterialFunctionCall(Transition Phase=OneMinus(None=@n80.G), Height Texture=@n16.G, Contrast=ScalarParameter() {defaultValue=0.4443, parameterName=Contrast}) {materialFunction=&HeightLerp}.Alpha
   FadeDistance=LinearInterpolate(A=ScalarParameter() {defaultValue=32253.8008, parameterName=DepthFade_1}, B=ScalarParameter() {parameterName=DepthFade_2}, Alpha=@n16.G)
  )
  n82 = LinearInterpolate(
   A=ScalarParameter() {defaultValue=143000.3125, parameterName=CameraFadeLength_Min}
   B=ScalarParameter() {defaultValue=143000.3125, parameterName=CameraFadeLength}
   Alpha=@n49
  )
  n83 = LinearInterpolate(
   A=ScalarParameter() {defaultValue=84900.1172, parameterName=CameraFadeOffset_Min}
   B=ScalarParameter() {defaultValue=84900.1172, parameterName=CameraFadeOffset}
   Alpha=@n49
  )
  StaticSwitchParameter(
   True=Multiply(
    A=@n81
    B=MaterialFunctionCall(Fade Length=@n82, Fade Offset=@n83) {materialFunction=&CameraDepthFade}.Result
   )
   False=@n81
  ) {parameterName=IsBottomSide}
  VertexColor()
 ]
}
