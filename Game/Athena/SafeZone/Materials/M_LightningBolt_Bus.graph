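# Translucent unlit mesh-particle material for a lightning bolt, named for the bus and stored with the Athena safe zone assets.
# An HDR purple color parameter is multiplied by a constant boost and fed into the Emissive input of the shared MF_FX_EmissiveAdjustments function.
# A Vertex Color node is also present with no visible connection, and the graph samples no textures.
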Material /Game/Athena/SafeZone/Materials/M_LightningBolt_Bus
graph {
 properties {bUsedWithMeshParticles=true, blendMode=BLEND_Translucent, pixelDepthOffsetMode=PDOM_Legacy, shadingModel=MSM_Unlit, usedShadingModels=Unlit}
 nodes [
  MaterialFunctionCall(Emissive=Multiply(A=VectorParameter() {parameterName=v, defaultValue=<rgba 22.4709 10.1544 30 0>}.RGB, B=Constant() {r=2.5})) {materialFunction=&MF_FX_EmissiveAdjustments}
  VertexColor()
 ]
}
