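# Unlit two-sided mobile material for the Athena safe zone storm wall, driven by the MPC_SafeZoneClouds parameter collection with debug switches.
# Two panning tilings of cloud-tube mask and normal textures, sampled with explicit derivatives to survive Frac wrapping, blend by zone scale and vertex red.
# Vertex-color tints, Fresnel and a panning swirl mask build the color.
# Vertex offset math centers and resizes the ring around the zone, pushing it outward toward high, distant cameras.
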Material /Game/Athena/SafeZone/Materials/M_SafeZoneStorm_Mobile
graph {
 properties {bUsedWithNanite=true, pixelDepthOffsetMode=PDOM_Legacy, shadingModel=MSM_Unlit, twoSided=true, usedShadingModels=Unlit}
 nodes [
  n1 = TextureCoordinate()
  n2 = ScalarParameter() {defaultValue=4, parameterName=UVShift_1}
  n3 = TextureSample(UVs=Panner(Coordinate=Multiply(A=@n1, B=@n2), Speed=ScalarParameter() {defaultValue=0.02, parameterName=SpeedPan_1}) {bFractionalPart=true}) {samplerType=SAMPLERTYPE_LinearGrayscale, texture=/Game/Global/Textures/Default/Flowmaps/T_StormSwirlMask}
  n4 = Multiply(
   A=Time()
   B=AppendVector(A=ScalarParameter() {defaultValue=-0.005, parameterName=Speed_X}, B=ScalarParameter() {defaultValue=-0.005, parameterName=Speed_Y})
  )
  n5 = Add(A=TextureCoordinate(), B=@n4)
  n6 = ScalarParameter() {defaultValue=7, parameterName=UVTiling_X}
  n7 = ScalarParameter() {defaultValue=4, parameterName=UVTiling_Y}
  n8 = ComponentMask(None=TextureCoordinate()) {r=true}
  n9 = ScalarParameter() {defaultValue=3, parameterName=UVShift}
  n10 = Add(A=Multiply(A=Frac(None=@n5), B=AppendVector(A=@n6, B=@n7)), B=AppendVector(A=Constant(), B=Multiply(A=@n8, B=@n9)))
  n11 = TextureCoordinate()
  n12 = DDX(Value=@n11)
  n13 = DDY(Value=@n11)
  n14 = TextureCoordinate()
  n15 = Multiply(
   A=Frac(None=Add(A=@n14, B=@n4))
   B=AppendVector(A=ScalarParameter() {defaultValue=1, parameterName=UVTiling_X_1}, B=ScalarParameter() {defaultValue=4, parameterName=UVTiling_Y_1})
  )
  n16 = TextureCoordinate()
  n17 = DDX(Value=@n16)
  n18 = DDY(Value=@n16)
  n19 = ScalarParameter() {defaultValue=1.4741, parameterName=low}
  n20 = ScalarParameter() {defaultValue=-3.0248, parameterName=high}
  n21 = ScalarParameter() {defaultValue=-0.0166, parameterName=Debug_ZoneScale}
  n22 = LinearInterpolate(
   A=StaticSwitchParameter(
    True=Clamp(None=@n21) {clampMode=CMODE_ClampMin}
    False=CollectionParameter() {collection=/Game/Athena/SafeZone/MPC_SafeZoneClouds, parameterName=SafeZoneScale}
   ) {parameterName=Debug}
   Alpha=CollectionParameter() {collection=&MPC_SafeZoneClouds, parameterName=PreventStormShrinking}
  )
  n23 = Multiply(A=@n22, B=VertexColor().R)
  n24 = VertexInterpolator(VS=Clamp(None=LinearInterpolate(A=@n19, B=@n20, Alpha=@n23)))
  n25 = LinearInterpolate(
   A=TextureSample!(UVs=@n10, Tex, "DDX(UVs)"=@n12, "DDY(UVs)"=@n13, Apply View MipBias) {mipValueMode=TMVM_Derivative, samplerType=SAMPLERTYPE_Masks, texture=/Game/Athena/SafeZone/Textures/StormCloudtubes_M}.G
   B=TextureSample(UVs=@n15, "DDX(UVs)"=@n17, "DDY(UVs)"=@n18) {mipValueMode=TMVM_Derivative, samplerType=SAMPLERTYPE_Masks, texture=&StormCloudtubes_M}.G
   Alpha=@n24.PS
  )
  n26 = Power(Base=OneMinus(None=@n25), Exp=ScalarParameter() {defaultValue=4.669, parameterName=Contrast})
  n27 = VertexColor()
  n28 = Multiply(A=@n27.G, B=VectorParameter() {parameterName=Color_3, defaultValue=<rgba 0.1432 0.1093 0.1615 1>}.RGB)
  n29 = TextureSample(UVs=@n10, "DDX(UVs)"=@n12, "DDY(UVs)"=@n13) {mipValueMode=TMVM_Derivative, samplerType=SAMPLERTYPE_Normal, texture=/Game/Athena/SafeZone/Textures/StormCloudtubes_N}
  n30 = TextureSample(UVs=@n15, "DDX(UVs)"=@n17, "DDY(UVs)"=@n18) {mipValueMode=TMVM_Derivative, samplerType=SAMPLERTYPE_Normal, texture=&StormCloudtubes_N}
  n31 = Fresnel(
   ExponentIn=ScalarParameter() {defaultValue=1.3824, parameterName=Fresnel}
   Normal=Transform(None=LinearInterpolate(A=@n29.RGB, B=@n30.RGB, Alpha=@n24.PS))
  )
  n32 = VectorParameter() {parameterName=Color_1, defaultValue=<rgba -0.03 -0.03 -0.03 1>}
  n33 = LinearInterpolate(
   A=VectorParameter() {parameterName=Color_2, defaultValue=<rgba 0.0286 0.0103 0.099 1>}.RGB
   B=VectorParameter() {parameterName=Color_5, defaultValue=<rgba 0.1704 0.059 0.3333 1>}.RGB
   Alpha=@n27.R
  )
  Add(
   A=Multiply(A=VectorParameter() {parameterName=Color_4, defaultValue=<rgba 0.2973 0.231 1 1>}.RGB, B=Multiply(A=@n3.R, B=@n26))
   B=Add(A=Multiply(A=@n28, B=@n31), B=LinearInterpolate(A=@n32.RGB, B=@n33, Alpha=@n25))
  )
  n34 = ComponentMask(None=CollectionParameter() {collection=&MPC_SafeZoneClouds, parameterName=SafeZoneLoc}) {g=true, r=true}
  n35 = StaticSwitchParameter(
   True=ScalarParameter() {parameterName=Debug_CloudStartingBias}
   False=CollectionParameter() {collection=&MPC_SafeZoneClouds, parameterName=CloudStartingBias}
  ) {parameterName=Debug_1}
  n36 = AppendVector(A=@n34, B=LinearInterpolate(A=Constant(), B=ScalarParameter() {defaultValue=-116941.0781, parameterName=StartPushDown}, Alpha=@n35))
  n37 = WorldPosition()
  n38 = ScalarParameter() {defaultValue=58148.8789, parameterName=ShrinkingCenterHeight}
  n39 = Multiply(
   A=Subtract(A=Add(A=@n36, B=@n37.XYZ), B=AppendVector(A=@n34, B=@n38))
   B=VectorParameter() {parameterName=PinchScale, defaultValue=<rgba 1 1 0.7238 1>}.RGB
  )
  n40 = ScalarParameter() {defaultValue=-192979.7344, parameterName=RingSize_Min}
  n41 = ScalarParameter() {parameterName=RingSize_Max}
  n42 = Add(A=@n22, B=Multiply(A=@n35) {constB=10})
  n43 = LinearInterpolate(A=@n40, B=@n41, Alpha=@n42)
  n44 = Subtract(A=ComponentMask(None=CameraPositionWS()) {b=true}, B=ScalarParameter() {defaultValue=35142.25, parameterName=Heightoffset})
  n45 = ScalarParameter() {defaultValue=5000, parameterName=Heightoffset_1}
  n46 = ComponentMask(None=CollectionParameter() {collection=&MPC_SafeZoneClouds, parameterName=SafeZoneLoc}) {g=true, r=true}
  n47 = ComponentMask(None=CameraPositionWS()) {g=true, r=true}
  n48 = Subtract(
   A=Add(A=Distance(A=@n46, B=@n47), B=ScalarParameter() {defaultValue=47917.5156, parameterName=FalloffExtendAdd})
   B=LinearInterpolate(A=@n41, B=Abs(None=@n40), Alpha=@n42)
  )
  n49 = Subtract(A=@n47, B=@n46)
  n50 = Subtract(A=ComponentMask(None=WorldPosition().XYZ) {g=true, r=true}, B=@n46)
  n51 = Clamp(None=DotProduct(A=Normalize(VectorInput=@n49), B=Normalize(VectorInput=@n50)))
  n52 = ScalarParameter() {defaultValue=14.3983, parameterName=FalloffExtendToCamera}
  n53 = Multiply(A=Clamp(None=Divide(A=@n44, B=@n45)), B=Multiply(A=Clamp(None=@n48) {clampMode=CMODE_ClampMin}, B=Power(Base=@n51, Exp=@n52)))
  Add(A=@n36, B=Multiply(A=Normalize(VectorInput=@n39), B=Add(A=@n43, B=@n53) {constB=0}))
  LinearInterpolate()
 ]
}
