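# Material function for safe-zone storm cloud ring tornadoes, outputting material attributes plus emissive, normal, WPO, blend mask and opacity mask.
# Panned, warped UVs sample cloud-tube mask and normal textures for color and view-dependent glow.
# WPO pulls vertices downward by a UV falloff, adds pivot-phased dual sine sway and pushes toward camera.
# AnimPhase and Rotation fade drop, sway and blend mask; Rotation also cuts the temporally dithered opacity.
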MaterialFunction /Game/Athena/SafeZone/Materials/MF_CloudRingTornado
graph {nodes [
  ActorPositionWS()
  ActorPositionWS()
  n1 = MaterialFunctionCall() {materialFunction=&ComponentPivotLocation}
  n2 = Normalize(VectorInput=Subtract(A=WorldPosition() {worldPositionShaderOffset=WPT_ExcludeAllShaderOffsets}.XYZ, B=@n1.Result))
  n3 = ComponentMask(None=TextureCoordinate()) {g=true}
  n4 = OneMinus(None=@n3)
  n5 = Power(Base=@n4, Exp=ScalarParameter() {defaultValue=2.8873, parameterName=FalloffDropDown})
  n6 = Multiply(A=ScalarParameter() {parameterName=Offset}, B=@n5)
  n7 = TextureCoordinate()
  n8 = AppendVector(A=Constant(), B=Multiply(A=ComponentMask(None=@n7) {r=true}, B=ScalarParameter() {defaultValue=2, parameterName=T_UVShift}))
  n9 = ScalarParameter() {parameterName=T_Tiling_Min}
  n10 = ComponentMask(None=@n7) {g=true}
  n11 = Multiply(
   A=Power(Base=OneMinus(None=@n10), Exp=ScalarParameter() {defaultValue=2.9309, parameterName=UVFalloff})
   B=ScalarParameter() {defaultValue=4, parameterName=T_Tiling_Max}
  )
  n12 = Panner(
   Coordinate=Add(A=Add(A=@n7, B=@n8), B=AppendVector(A=@n9, B=@n11))
   Time=Time()
   Speed=AppendVector(A=ScalarParameter() {defaultValue=-0.0862, parameterName=T_SpeedX}, B=ScalarParameter() {defaultValue=0.1429, parameterName=T_SpeedY})
  )
  n13 = TextureSample(UVs=@n12) {samplerType=SAMPLERTYPE_Masks, texture=/Game/Athena/SafeZone/Textures/StormCloudtubes_M}
  Add(A=Multiply(A=Multiply(A=@n2, B=@n6), B=Subtract(A=@n13.B) {constB=0.5}))
  n14 = LinearInterpolate(
   A=VectorParameter() {parameterName=TornadoColor, defaultValue=<rgba 0 0 0 1>}.RGB
   B=LinearInterpolate(
    A=VectorParameter() {parameterName=TornadoColor_1, defaultValue=<rgba 0.0958 0.0392 0.3906 1>}.RGB
    B=VectorParameter() {parameterName=TornadoColor_3, defaultValue=<rgba 0 0 0 1>}.RGB
    Alpha=@n5
   )
   Alpha=@n13.B
  )
  n15 = ScalarParameter() {defaultValue=1.0706, parameterName=Fresnel}
  n16 = TextureSample(UVs=@n12) {samplerType=SAMPLERTYPE_Normal, texture=/Game/Athena/SafeZone/Textures/StormCloudtubes_N_2}
  n17 = MaterialFunctionCall(BaseNormal=@n2, AdditionalNormal=@n16.RGB) {materialFunction=&BlendAngleCorrectedNormals}
  Add(
   A=@n14
   B=LinearInterpolate(
    A=VectorParameter() {parameterName=TornadoColor_2, defaultValue=<rgba 0 0 0 1>}.RGB
    B=VectorParameter() {parameterName=TornadoColor_4, defaultValue=<rgba 0.2596 0.151 1 1>}.RGB
    Alpha=Fresnel(ExponentIn=@n15, Normal=@n17.Result) {baseReflectFraction=0}
   )
  )
  Constant3Vector() {constant=<rgba 0 0 -1 0>}
  n18 = OneMinus(None=@n13.B)
  n19 = DotProduct(A=@n2, B=CameraVectorWS())
  n20 = Clamp(None=LinearInterpolate(
    A=ScalarParameter() {defaultValue=-0.0416, parameterName=Fresnel_1}
    B=ScalarParameter() {defaultValue=0.0794, parameterName=Fresnel_2}
    Alpha=Multiply(A=@n18, B=@n19)
   ))
  n21 = Add(
   A=@n14
   B=LinearInterpolate(
    A=VectorParameter() {parameterName=TornadoColor_5, defaultValue=<rgba 0 0 0 1>}.RGB
    B=VectorParameter() {parameterName=TornadoColor_6, defaultValue=<rgba 2.659 0 20 1>}.RGB
    Alpha=@n20
   )
  )
  n22 = FunctionInput() {inputName=Normal}
  n23 = Constant3Vector() {constant=<rgba 0 0 -1 0>}
  n24 = ScalarParameter() {defaultValue=14171.3779, parameterName=DropDownDistance}
  n25 = FunctionInput() {inputName=Rotation, inputType=FunctionInput_Scalar, sortPriority=5}
  n26 = ScalarParameter() {defaultValue=6.2599, parameterName=AnimationFadeOnRotation}
  n27 = OneMinus(None=Clamp(None=Multiply(A=@n25, B=@n26)))
  n28 = Multiply(A=ScalarParameter() {defaultValue=1, parameterName=AnimPhase}, B=@n27)
  n29 = Multiply(A=@n5, B=Multiply(A=LinearInterpolate(A=@n22, B=@n23, Alpha=@n5), B=LinearInterpolate(B=@n24, Alpha=@n28)))
  n30 = Multiply(A=Time(), B=ScalarParameter() {defaultValue=0.03, parameterName=Tornado_speed_SineWave})
  n31 = ComponentMask(None=@n1.Result) {r=true}
  n32 = ComponentMask(None=@n1.Result) {g=true}
  n33 = Add(
   A=Multiply(A=Add(A=@n3, B=@n30), B=ScalarParameter() {defaultValue=1.3201, parameterName=Tornado_WaveLength})
   B=Frac(None=Add(A=@n31, B=@n32))
  )
  n34 = Sine(None=@n33)
  n35 = Multiply(A=@n5, B=LinearInterpolate(B=ScalarParameter() {defaultValue=2482.6853, parameterName=Tornado_Amplitude}, Alpha=@n28))
  n36 = Sine(None=Add(A=@n33) {constB=45.3425})
  n37 = Add(
   A=Multiply(A=Multiply(A=@n34, B=@n35), B=VectorParameter() {parameterName=WaveDirection, defaultValue=<rgba 1 0.2 0 1>}.RGB)
   B=Multiply(A=Multiply(A=@n36, B=@n35), B=VectorParameter() {parameterName=WaveDirection_2, defaultValue=<rgba 0 1 0 1>}.RGB)
  )
  n38 = Add(
   A=Multiply(
    A=Add(A=@n29, B=@n37)
    B=MaterialFunctionCall() {materialFunction=/Engine/Functions/Engine_MaterialFunctions02/WorldPositionOffset/ObjectScale}.Scale XYZ
   )
   B=Multiply(
    A=MaterialFunctionCall() {materialFunction=&CameraDirectionVector}.Result
    B=ScalarParameter() {defaultValue=-500, parameterName=Push To Camera}
   )
  )
  FunctionOutput(None=MakeMaterialAttributes!(
    BaseColor=@n14
    Metallic
    Specular=Constant()
    Roughness=Constant() {r=1}
    Anisotropy
    EmissiveColor=@n21
    Opacity
    OpacityMask
    Normal=@n16.RGB
    Tangent
    WorldPositionOffset=@n38
    SubsurfaceColor
    ClearCoat
    ClearCoatRoughness
    AmbientOcclusion
    Refraction
    CustomizedUVs_0
    CustomizedUVs_1
    CustomizedUVs_2
    CustomizedUVs_3
    CustomizedUVs_4
    CustomizedUVs_5
    CustomizedUVs_6
    CustomizedUVs_7
    PixelDepthOffset
    ShadingModel
    Displacement
   ))
  FunctionOutput(None=@n21) {outputName=Emissive, sortPriority=2}
  FunctionOutput(None=@n17.Result) {outputName=Normal, sortPriority=3}
  FunctionOutput(None=@n38) {outputName=WPO, sortPriority=4}
  n39 = Power(
   Base=Add(A=@n4, B=ScalarParameter() {defaultValue=0.2446, parameterName=FalloffDropDown_2})
   Exp=ScalarParameter() {defaultValue=3.3274, parameterName=FalloffDropDown_1}
  )
  FunctionOutput(None=LinearInterpolate(B=Clamp(None=@n39), Alpha=@n28)) {outputName=BlendMask, sortPriority=4}
  n40 = Clamp(None=LinearInterpolate(
    A=ScalarParameter() {parameterName=TornadoBlendOpacityMaks_1}
    B=ScalarParameter() {defaultValue=2.0816, parameterName=TornadoBlendOpacityMaks_2}
    Alpha=@n4
   ))
  n41 = Ceil(None=@n27)
  FunctionOutput(None=LinearInterpolate(
    B=MaterialFunctionCall(Alpha Threshold=@n40) {materialFunction=&DitherTemporalAA}.Result
    Alpha=@n41
   )) {outputName=OpacityMask, sortPriority=6}
  LinearInterpolate(A=VectorParameter() {parameterName=OffsetOverdrawOptimization, defaultValue=<rgba 0 0 -55555555328 1>}.RGB, B=@n38, Alpha=@n41)
  Multiply(A=@n18, B=@n20)
  Transform(None=@n17.Result)
  Transform(None=@n2) {transformSourceType=TRANSFORMSOURCE_World, transformType=TRANSFORM_Tangent}
 ]}
