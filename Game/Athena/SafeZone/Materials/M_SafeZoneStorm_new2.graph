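# Masked two-sided subsurface material for the Athena safe zone storm clouds, following zone location and scale from MPC_SafeZoneClouds.
# World position offset bends and pushes clouds around the zone's opening, stretched toward the camera, adding waves and texture displacement.
# Twisted radial UVs blend into planar center mapping, feeding color, normals and flickering lightning, with local-color and super-storm tints.
# A distance-based opacity mask cuts the center; the IsTornade switch layers in MF_CloudRingTornado.
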Material /Game/Athena/SafeZone/Materials/M_SafeZoneStorm_new2
graph {
 properties {
  bFullyRough=true
  bTangentSpaceNormal=false
  bUseLightmapDirectionality=false
  bUseMaterialAttributes=true
  bUseTranslucencyVertexFog=false
  bUsedWithNanite=true
  blendMode=BLEND_Masked
  pixelDepthOffsetMode=PDOM_Legacy
  shadingModel=MSM_Subsurface
  translucencyLightingMode=TLM_SurfacePerPixelLighting
  translucencyPass=MTP_BeforeDOF
  twoSided=true
  usedShadingModels=Subsurface
 }
 nodes [
  n1 = CollectionParameter() {collection=/Game/Athena/SafeZone/MPC_SafeZoneClouds, parameterName=LocalColorPosition}
  n2 = Power(Base=SphereMask(
    A=ComponentMask(None=@n1) {b=true, g=true, r=true}
    B=WorldPosition().XYZ
    Radius=CollectionParameter() {collection=&MPC_SafeZoneClouds, parameterName=LocalColorRadius}
   ) {attenuationRadius=250000, hardnessPercent=65}) {constExponent=0.7}
  n3 = VectorParameter() {parameterName=Color2, defaultValue=<rgba 0.0084 0.0009 0.015 1>}
  n4 = Divide(
   A=ComponentMask(None=WorldPosition() {worldPositionShaderOffset=WPT_ExcludeAllShaderOffsets}.XYZ) {g=true, r=true}
   B=ScalarParameter() {defaultValue=55918.9258, parameterName=CenterblendScale}
  )
  n5 = ComponentMask(None=WorldPosition() {worldPositionShaderOffset=WPT_ExcludeAllShaderOffsets}.XYZ) {g=true, r=true}
  n6 = ComponentMask(None=CollectionParameter() {collection=&MPC_SafeZoneClouds, parameterName=SafeZoneLoc}) {g=true, r=true}
  n7 = StaticSwitchParameter(
   True=ComponentMask(None=VectorParameter() {parameterName=DebugPosition, defaultValue=<rgba 222.1989 61093.6641 0 1>}.RGB) {g=true, r=true}
   False=@n6
  ) {parameterName=Debug}
  n8 = MaterialFunctionCall(Vector or UVs=Divide(A=Subtract(A=@n5, B=@n7), B=ScalarParameter() {defaultValue=2504.9761, parameterName=Scale_1})) {materialFunction=&VectorToRadialValue}
  n9 = Constant() {r=1}
  n10 = Power(Base=ComponentMask(None=@n8.Radial Coordinates) {g=true}, Exp=ScalarParameter() {defaultValue=0.6162, parameterName=FallofTwist})
  n11 = Multiply(
   A=Add(A=@n10, B=ScalarParameter() {defaultValue=1.237, parameterName=Scale_Twist_Bias})
   B=ScalarParameter() {defaultValue=1, parameterName=Scale_Twist_Y}
  )
  n12 = Multiply(A=@n8.Radial Coordinates, B=AppendVector(A=ScalarParameter() {defaultValue=5, parameterName=Scale_Twist_X}, B=Divide(A=@n9, B=@n11)))
  n13 = Add(A=@n12, B=AppendVector(A=Constant(), B=ComponentMask(None=@n12) {r=true}))
  n14 = Time()
  n15 = AppendVector(A=ScalarParameter() {defaultValue=-0.01, parameterName=Speed_x}, B=ScalarParameter() {defaultValue=-0.025, parameterName=Speed_y})
  n16 = Add(A=@n13, B=Frac(None=Multiply(A=@n14, B=@n15)))
  n17 = TextureCoordinate()
  n18 = DDX(Value=@n17)
  n19 = DDY(Value=@n17)
  n20 = TextureSample!(UVs=@n16, Tex, "DDX(UVs)"=@n18, "DDY(UVs)"=@n19, Apply View MipBias) {mipValueMode=TMVM_Derivative, samplerType=SAMPLERTYPE_Masks, texture=/Game/Athena/SafeZone/Textures/StormCloudtubes_M}
  n21 = ComponentMask(None=WorldPosition() {worldPositionShaderOffset=WPT_ExcludeAllShaderOffsets}.XYZ) {g=true, r=true}
  n22 = Distance(A=@n7, B=@n21)
  n23 = ScalarParameter() {defaultValue=36556.4023, parameterName=CenterblendRadius}
  n24 = Clamp(None=Power(Base=Divide(A=@n22, B=@n23), Exp=ScalarParameter() {defaultValue=3.7308, parameterName=CenterblendContrast}))
  n25 = MaterialFunctionCall(Float3=LinearInterpolate(
    A=TextureSample!(UVs=@n4, Tex, Apply View MipBias) {samplerType=SAMPLERTYPE_Masks, texture=&StormCloudtubes_M}.RGB
    B=@n20.RGB
    Alpha=@n24
   )) {materialFunction=&BreakOutFloat3Components}
  n26 = LinearInterpolate(
   A=LinearInterpolate(
    A=VectorParameter() {parameterName=Color1, defaultValue=<rgba 0.0118 0.0021 0.04 1>}.RGB
    B=Constant3Vector() {constant=<rgba 0.0392 0.0263 0.0039 1>}
    Alpha=@n2
   )
   B=@n3.RGB
   Alpha=@n25.R
  )
  n27 = VectorParameter() {parameterName=Color_3, defaultValue=<rgba 0.0791 0.0467 0.1094 1>}
  n28 = Divide(A=@n22, B=ScalarParameter() {defaultValue=413360.1562, parameterName=GradientDistance})
  n29 = Power(Base=OneMinus(None=Clamp(None=@n28)), Exp=ScalarParameter() {defaultValue=7.5796, parameterName=GradientDistance_Falloff})
  Add(A=@n26, B=Multiply(A=@n27.RGB, B=@n29))
  n30 = CameraPositionWS()
  n31 = WorldPosition()
  Add(A=@n30, B=@n31.XYZ)
  n32 = StaticSwitchParameter(
   True=Clamp(None=ScalarParameter() {defaultValue=-0.0166, parameterName=Debug_ZoneScale}) {clampMode=CMODE_ClampMin}
   False=CollectionParameter() {collection=&MPC_SafeZoneClouds, parameterName=SafeZoneScale}
  ) {parameterName=Debug}
  n33 = ScalarParameter() {parameterName=Debug_CloudStartingBias}
  n34 = CollectionParameter() {collection=&MPC_SafeZoneClouds, parameterName=PreventStormShrinking}
  n35 = Max(A=@n34, B=CollectionParameter() {collection=&MPC_SafeZoneClouds, parameterName=CloudStartingBias})
  n36 = Add(A=@n32, B=Multiply(A=StaticSwitchParameter(True=@n33, False=@n35) {parameterName=Debug}) {constB=5.6})
  n37 = LinearInterpolate(
   A=ScalarParameter() {desc=Remove?, parameterName=Ring_Offset_Min}
   B=ScalarParameter() {defaultValue=25375.584, desc=Remove?, parameterName=Ring_Offset_2}
   Alpha=@n36
  )
  n38 = LinearInterpolate(
   A=ScalarParameter() {parameterName=Ring_Radius_Min}
   B=ScalarParameter() {defaultValue=212768.9688, parameterName=Ring_Radius_Max}
   Alpha=@n36
  )
  n39 = CameraPositionWS()
  n40 = Divide(
   A=Subtract(A=ComponentMask(None=@n39) {b=true}, B=ScalarParameter() {defaultValue=35142.25, parameterName=Heightoffset})
   B=ScalarParameter() {defaultValue=5000, parameterName=Heightoffset_1}
  )
  n41 = ComponentMask(None=CameraPositionWS()) {g=true, r=true}
  n42 = Distance(A=@n6, B=@n41)
  n43 = ScalarParameter() {defaultValue=47917.5156, parameterName=FalloffExtendAdd}
  n44 = Clamp(None=Subtract(A=Add(A=@n42, B=@n43), B=@n38)) {clampMode=CMODE_ClampMin}
  n45 = Normalize(VectorInput=Subtract(A=@n41, B=@n6))
  n46 = WorldPosition()
  n47 = Normalize(VectorInput=Subtract(A=ComponentMask(None=@n46.XYZ) {g=true, r=true}, B=@n6))
  n48 = Power(Base=Clamp(None=DotProduct(A=@n45, B=@n47)), Exp=ScalarParameter() {defaultValue=14.3983, parameterName=FalloffExtendToCamera})
  n49 = Add(A=@n38, B=Multiply(A=Clamp(None=@n40), B=Multiply(A=@n44, B=@n48)))
  n50 = Clamp(None=Divide(A=Subtract(A=@n22, B=@n37), B=@n49))
  n51 = OneMinus(None=@n50)
  n52 = Constant3Vector() {constant=<rgba 0 0 1 0>}
  n53 = Subtract(A=@n21, B=@n7)
  n54 = Constant()
  n55 = Normalize(VectorInput=CrossProduct(A=Constant3Vector() {constant=<rgba 0 0 1 0>}, B=AppendVector(A=@n53, B=@n54)))
  n56 = Multiply(A=@n51, B=ScalarParameter() {defaultValue=0.6, desc=should be 0.6 to avoid artifact with self inserection, parameterName=Rot})
  n57 = WorldPosition() {worldPositionShaderOffset=WPT_ExcludeAllShaderOffsets}
  n58 = ComponentMask(None=WorldPosition() {worldPositionShaderOffset=WPT_ExcludeAllShaderOffsets}.XYZ) {b=true}
  n59 = ScalarParameter() {defaultValue=1.2044, parameterName=BiasBendRaidus}
  n60 = LinearInterpolate(
   A=ScalarParameter() {defaultValue=10.9401, parameterName=BendRadius_Min}
   B=ScalarParameter() {defaultValue=49224.6836, parameterName=BendRadius}
   Alpha=Clamp(None=Multiply(A=@n36, B=@n59))
  )
  n61 = WorldPosition() {worldPositionShaderOffset=WPT_ExcludeAllShaderOffsets}
  n62 = RotateAboutAxis(
   NormalizedRotationAxis=@n55
   RotationAngle=@n56
   PivotPoint=AppendVector(A=ComponentMask(None=@n57.XYZ) {g=true, r=true}, B=Add(A=@n58, B=@n60))
   Position=@n61.XYZ
  )
  Add(A=Multiply(A=Multiply(A=@n51, B=@n52), B=ScalarParameter() {defaultValue=104, parameterName=PreventClipPushUp}), B=@n62)
  n63 = VectorParameter() {parameterName=SineWaveDirection, defaultValue=<rgba 0.2 0.2 1 1>}
  n64 = WorldPosition() {worldPositionShaderOffset=WPT_ExcludeAllShaderOffsets}
  n65 = ScalarParameter() {defaultValue=166932.625, parameterName=WaveFequency_1}
  n66 = ScalarParameter() {defaultValue=146539.0156, parameterName=WaveFequency}
  n67 = Sine(None=Divide(A=ComponentMask(None=@n64.XYZ) {g=true, r=true}, B=AppendVector(A=@n65, B=@n66)))
  n68 = ComponentMask(None=WorldPosition() {worldPositionShaderOffset=WPT_ExcludeAllShaderOffsets}.XYZ) {g=true, r=true}
  n69 = Constant2Vector()
  n70 = Power(
   Base=Divide(A=Distance(A=@n68, B=@n69), B=ScalarParameter() {defaultValue=424703.9375, parameterName=Radius})
   Exp=ScalarParameter() {defaultValue=2.1788, parameterName=CloudHorizonWPO_Falloff}
  )
  n71 = Clamp(None=@n70)
  n72 = Multiply(A=@n71, B=ScalarParameter() {defaultValue=0.3952, parameterName=WaveFade})
  n73 = Multiply(
   A=Add(A=ComponentMask(None=@n67) {r=true}, B=ComponentMask(None=@n67) {g=true})
   B=Multiply(A=ScalarParameter() {defaultValue=6618.9858, parameterName=WaveAmplitude}, B=OneMinus(None=@n72))
  )
  n74 = MaterialFunctionCall(Rotation Axis=@n55, Rotation Angle=@n56) {materialFunction=&FixRotateAboutAxisNormals}
  n75 = ScalarParameter() {defaultValue=-23657.0508, parameterName=WPO_CloudTubes}
  n76 = Divide(A=@n22, B=ScalarParameter() {defaultValue=216239.4844, parameterName=WPO_CloudTubes_Falloff})
  n77 = Multiply(
   A=TextureSample!(UVs=@n16, Tex, Level=ScalarParameter() {defaultValue=5, parameterName=BlurCloudOffset}, Apply View MipBias) {mipValueMode=TMVM_MipLevel, samplerType=SAMPLERTYPE_Masks, texture=&StormCloudtubes_M}.G
   B=Multiply(A=Multiply(A=@n74.New World Space Vertex Normal, B=@n75), B=Clamp(None=@n76))
  )
  n78 = Add(A=@n77, B=ScalarParameter() {defaultValue=3, parameterName=Transluceny_WPO_Offset})
  n79 = ComponentMask(None=Add(A=Multiply(A=@n63.RGB, B=@n73), B=StaticSwitchParameter(True=@n78, False=@n77) {parameterName=IsTranslucentLayer})) {b=true, desc=Clouds WPO, g=true, r=true}
  Add(A=@n79, B=@n61.XYZ)
  Clamp(None=OneMinus(None=@n10))
  n80 = WorldPosition()
  n81 = ActorPositionWS()
  n82 = OneMinus(None=ComponentMask(None=Subtract(A=@n80.XYZ, B=@n81)) {b=true})
  n83 = ScalarParameter() {defaultValue=138656.875, parameterName=FalloffMask}
  n84 = ScalarParameter() {defaultValue=1.2272, parameterName=FalloffMask_Offset_M}
  n85 = ScalarParameter() {defaultValue=1.5908, parameterName=FalloffMask_Offset}
  n86 = Clamp(None=Add(A=Divide(A=@n82, B=@n83), B=LinearInterpolate(A=@n84, B=@n85, Alpha=@n36)))
  n87 = Floor(None=@n86)
  n88 = LinearInterpolate(
   A=ScalarParameter() {defaultValue=53.5965, parameterName=maskRing_Min}
   B=ScalarParameter() {defaultValue=333069.125, parameterName=maskRing_1}
   Alpha=@n36
  )
  n89 = Clamp(None=OneMinus(None=Divide(A=@n22, B=@n88)))
  Clamp(None=Multiply(A=Ceil(None=@n86), B=Add(A=@n87, B=@n89)))
  ComponentMask(None=ActorPositionWS()) {g=true, r=true}
  ComponentMask() {g=true}
  Constant3Vector() {constant=<rgba 0 0 1 0>}
  Constant()
  Constant() {r=170000}
  Constant() {r=195000}
  LinearInterpolate(A=@n32, Alpha=@n34)
  LinearInterpolate(A=@n3.RGB, B=@n27.RGB, Alpha=@n29)
  n90 = Time()
  LinearInterpolate(Alpha=ConstantBiasScale(None=Sine(None=@n90) {period=25})) {constB=3.5}
  LinearInterpolate()
  MaterialFunctionCall() {materialFunction=&CameraDirectionVector}
  n91 = ScalarParameter() {defaultValue=250000, group=StormDissipate, parameterName=LargestShieldRadius}
  n92 = ComponentMask(None=CollectionParameter() {collection=&MPC_SafeZoneClouds, parameterName=SafeZoneLoc}) {g=true, r=true}
  n93 = Add(
   A=LinearInterpolate(
    A=CollectionParameter() {collection=&MPC_SafeZoneClouds, parameterName=SafeZoneScale}
    Alpha=CollectionParameter() {collection=&MPC_SafeZoneClouds, parameterName=PreventStormShrinking}
   )
   B=Multiply(A=CollectionParameter() {collection=&MPC_SafeZoneClouds, parameterName=CloudStartingBias}) {constB=4}
  )
  MaterialFunctionCall(
   UVs=Divide(A=WorldPosition() {worldPositionShaderOffset=WPT_ExcludeAllShaderOffsets}.XYZ, B=@n91)
   CenterPosition=Divide(A=@n92, B=@n91)
   Radius=LinearInterpolate(A=Constant(), B=Constant() {r=1}, Alpha=@n93)
   Density=LinearInterpolate(
    A=ScalarParameter() {defaultValue=2.1091, group=StormDissipate, parameterName=ShieldDensitySmall}
    B=ScalarParameter() {defaultValue=1.6272, group=StormDissipate, parameterName=ShieldDensityLarge}
    Alpha=@n93
   )
  ) {materialFunction=&RadialGradientExponential}
  n94 = ScalarParameter() {defaultValue=300000, group=StormDissipate, parameterName=LargestShieldRadius_1}
  MaterialFunctionCall(
   UVs=Divide(A=WorldPosition() {worldPositionShaderOffset=WPT_ExcludeAllShaderOffsets}.XYZ, B=@n94)
   CenterPosition=Divide(A=@n92, B=@n94)
   Radius=LinearInterpolate(A=Constant(), B=Constant() {r=1}, Alpha=@n93)
   Density=LinearInterpolate(
    A=ScalarParameter() {defaultValue=1.8675, group=StormDissipate, parameterName=FalloffDensitySmall}
    B=ScalarParameter() {defaultValue=0.4979, group=StormDissipate, parameterName=FalloffDensityLarge}
    Alpha=@n93
   )
  ) {materialFunction=&RadialGradientExponential}
  n95 = CameraPositionWS()
  n96 = WorldPosition()
  n97 = Clamp(None=Divide(A=Distance(A=@n95, B=@n96.XYZ), B=ScalarParameter() {defaultValue=206364.8594, parameterName=NearCameraDistance}))
  Multiply(
   A=Multiply(A=Constant3Vector() {constant=<rgba 0 0 1 0>}, B=OneMinus(None=@n97))
   B=ScalarParameter() {defaultValue=473976.75, parameterName=PushAroundCamera}
  )
  Multiply(A=@n70, B=@n50)
  n98 = ScalarParameter() {defaultValue=-0.5284, parameterName=MoutainCloudFade}
  Multiply(A=@n98)
  n99 = ScalarParameter() {defaultValue=16568113152, parameterName=TestRadius_1}
  n100 = ScalarParameter() {defaultValue=7.7304, parameterName=TestRadius}
  Multiply(
   A=Clamp(None=LinearInterpolate(A=@n99, B=@n100, Alpha=@n50))
   B=VectorParameter() {parameterName=ZOffestCenter, defaultValue=<rgba 0 0 0 -359624064>}.RGB
  )
  OneMinus(None=@n86)
  n101 = Multiply(A=Time(), B=ScalarParameter() {defaultValue=1.0842, parameterName=TimeShift})
  n102 = Frac(None=@n101)
  n103 = OneMinus(None=@n102)
  n104 = Clamp(None=Multiply(A=Multiply(A=@n102, B=@n103), B=ScalarParameter() {defaultValue=8.4141, parameterName=FadeSpeedbetweenShit}))
  n105 = Divide(
   A=ComponentMask(None=WorldPosition() {worldPositionShaderOffset=WPT_ExcludeAllShaderOffsets}.XYZ) {g=true, r=true}
   B=ScalarParameter() {defaultValue=603352.375, parameterName=Scale_2}
  )
  n106 = ScalarParameter() {defaultValue=3.3322, parameterName=UVOffsetLayer2}
  n107 = Ceil(None=@n101)
  n108 = ScalarParameter() {defaultValue=0.0262, parameterName=TimeShiftOffset}
  n109 = TextureSample!(UVs=AppendVector(A=Multiply(A=@n107, B=@n108), B=Constant()), Tex, Apply View MipBias) {samplerType=SAMPLERTYPE_Masks, texture=/Game/Gadgets/Assets/VinderTech_GliderChute/Textures/T_LightingTimeline_M}
  n110 = Multiply(A=@n106, B=@n109.G)
  n111 = Multiply(A=@n106, B=@n109.B)
  n112 = TextureSample(UVs=Add(A=@n105, B=AppendVector(A=@n110, B=@n111))) {samplerType=SAMPLERTYPE_LinearColor, texture=/Game/Athena/SafeZone/Textures/T_Clouds_Lightning}
  n113 = Multiply(
   A=Add(A=Time(), B=ScalarParameter() {defaultValue=1.6322, parameterName=LightingOffset_1})
   B=ScalarParameter() {defaultValue=0.2803, parameterName=LightingSpeed_1}
  )
  n114 = CurveAtlasRowParameter(CurveTime=Frac(None=@n113)) {
   atlas=&AthenaCurveAtlas01
   curve=/Game/Athena/SafeZone/Textures/StormLighting
   defaultValue=0.0723
   parameterName=Curve1
  }
  n115 = Multiply(A=Time(), B=ScalarParameter() {defaultValue=2.8429, parameterName=TimeShift_1})
  n116 = Frac(None=@n115)
  n117 = Multiply(A=Multiply(A=@n116, B=OneMinus(None=@n116)), B=ScalarParameter() {defaultValue=1, parameterName=FadeSpeedbetweenShit_1})
  n118 = ScalarParameter() {defaultValue=5.0061, parameterName=UVOffsetLayer2_1}
  n119 = Ceil(None=@n115)
  n120 = ScalarParameter() {defaultValue=0.1062, parameterName=TimeShiftOffset_1}
  n121 = TextureSample(UVs=AppendVector(A=Multiply(A=@n119, B=@n120), B=Constant())) {samplerType=SAMPLERTYPE_Masks, texture=&T_LightingTimeline_M}
  n122 = Add(A=@n105, B=AppendVector(A=Multiply(A=@n118, B=@n121.G), B=Multiply(A=@n118, B=@n121.B)))
  n123 = Multiply(
   A=Add(A=Time(), B=ScalarParameter() {defaultValue=1.0611, parameterName=LightingOffset_2})
   B=ScalarParameter() {defaultValue=0.2238, parameterName=LightingSpeed_2}
  )
  n124 = CurveAtlasRowParameter(CurveTime=Frac(None=@n123)) {
   atlas=&AthenaCurveAtlas01
   curve=&StormLighting
   defaultValue=18
   parameterName=Curve1_1
  }
  n125 = Multiply(
   A=Multiply(
    A=Clamp(None=@n117)
    B=TextureSample(UVs=@n122) {samplerType=SAMPLERTYPE_LinearColor, texture=&T_Clouds_Lightning}.G
   )
   B=@n124.R
  )
  n126 = Add(A=Multiply(A=Multiply(A=@n104, B=@n112.G), B=@n114.R), B=@n125)
  QualitySwitch(Default=@n126, Low=@n125, High=@n126, Medium=@n125) {desc=removed some of the lighting due to Design feedback}
  n127 = Multiply(
   A=Power(Exp=ScalarParameter() {defaultValue=3, parameterName=BOHeightRatio_13})
   B=ScalarParameter() {defaultValue=1, parameterName=BOHeightRatio_5}
  )
  n128 = Reroute()
  n129 = ScalarParameter() {defaultValue=1, parameterName=BOHeightRatio_2}
  Reroute(None=LinearInterpolate(A=Clamp(None=@n127), Alpha=Power(Base=@n128, Exp=@n129)))
  ScalarParameter() {parameterName=asdasd}
  ScalarParameter() {defaultValue=1, parameterName=BlurMountain}
  ScalarParameter() {parameterName=SuperStorm}
  n130 = ComponentMask(None=WorldPosition() {worldPositionShaderOffset=WPT_CameraRelative}.XYZ) {g=true, r=true}
  n131 = Add(
   A=ComponentMask(None=MaterialFunctionCall(Vector or UVs=@n130) {materialFunction=&VectorToRadialValue}.Radial Coordinates) {r=true}
   B=ScalarParameter() {defaultValue=0.397, parameterName=FakeMountain_offset_X}
  )
  n132 = ScalarParameter() {defaultValue=1, parameterName=FakeMountain_Scale_X}
  n133 = ScalarParameter() {defaultValue=221246.7969, parameterName=ProjectionDistance}
  n134 = CameraPositionWS()
  n135 = Distance(A=@n31.XYZ, B=@n134)
  n136 = Subtract(A=@n31.XYZ, B=@n134)
  n137 = Add(A=@n31.XYZ, B=Multiply(A=Subtract(A=@n133, B=@n135), B=Normalize(VectorInput=@n136)))
  n138 = Add(A=ComponentMask(None=Subtract(A=@n137, B=@n30)) {b=true}, B=ScalarParameter() {defaultValue=-1237704.5, parameterName=FakeMountain_offset_Y})
  n139 = ScalarParameter() {defaultValue=-317329.5625, parameterName=FakeMountain_Scale_Y}
  n140 = TextureSample(UVs=AppendVector(A=Divide(A=@n131, B=@n132), B=Divide(A=@n138, B=@n139))) {samplerType=SAMPLERTYPE_Masks, texture=/Game/Athena/SafeZone/Textures/T_CloudsFakeMountain}
  n141 = Clamp(None=LinearInterpolate(A=@n98, B=ScalarParameter() {defaultValue=1.0419, parameterName=MoutainCloudFade_1}, Alpha=@n25.G))
  n142 = ComponentMask(None=@n51) {desc=Gradient where the Rotation is, r=true}
  n143 = VectorParameter() {parameterName=Color4, defaultValue=<rgba 0.2046 0.0682 0.2292 1>}
  n144 = MaterialFunctionCall(Value_Morning=Constant() {r=1}, Value_Day=Constant() {r=1}, Value_Evening=Constant() {r=0.65}, Value_Night=Constant() {r=0.35}) {materialFunction=&Scalar_Time_of_Day}
  n145 = Multiply(A=Constant3Vector() {constant=<rgba 0.2275 0.1229 0.0667 1>}, B=@n144.Result)
  n146 = Add(A=Multiply(A=Add(A=@n142, B=@n29), B=LinearInterpolate(A=@n143.RGB, B=@n145, Alpha=@n2)), B=@n26)
  n147 = Constant3Vector() {constant=<rgba 0.051 0.0229 0.0078 1>}
  n148 = TwoSidedSign()
  n149 = LinearInterpolate(
   A=TextureSample(UVs=@n4) {desc=Center normal None Spiral, samplerType=SAMPLERTYPE_Normal, texture=/Game/Athena/SafeZone/Textures/StormCloudtubes_N}.RGB
   B=TextureSample!(UVs=@n16, Tex, "DDX(UVs)"=@n18, "DDY(UVs)"=@n19, Apply View MipBias) {mipValueMode=TMVM_Derivative, samplerType=SAMPLERTYPE_Normal, texture=&StormCloudtubes_N}.RGB
   Alpha=@n24
  )
  n150 = Time()
  n151 = ScalarParameter() {defaultValue=-0.0115, parameterName=DetailNormal_Speed}
  n152 = Add(A=@n13, B=Frac(None=Multiply(A=@n150, B=@n151)))
  n153 = Multiply(A=@n152, B=ScalarParameter() {defaultValue=6, parameterName=DetailNormalScale})
  n154 = ScalarParameter() {defaultValue=1, parameterName=DetailNormalFlatten}
  n155 = ScalarParameter() {defaultValue=0.4, parameterName=DetailNormalFlatten_1}
  n156 = MaterialFunctionCall(
   BaseNormal=@n149
   AdditionalNormal=MaterialFunctionCall(
    Normal=TextureSample(UVs=@n153, "DDX(UVs)"=@n18, "DDY(UVs)"=@n19) {mipValueMode=TMVM_Derivative, samplerType=SAMPLERTYPE_Normal, texture=/Game/Athena/SafeZone/Textures/StormCloudtubes_N_2}.RGB
    Flatness=LinearInterpolate(A=@n154, B=@n155, Alpha=@n20.G)
   ) {materialFunction=&FlattenNormal}.Result
  ) {materialFunction=&BlendAngleCorrectedNormals}
  n157 = QualitySwitch(Default=@n156.Result, Low=@n149, High=@n156.Result, Medium=@n149)
  n158 = ComponentMask(None=Multiply(A=Multiply(A=@n148, B=@n157), B=VectorParameter() {parameterName=Normal, defaultValue=<rgba -0.8 -0.8 -0.7128 1>}.RGB) {constB=-0.2}) {b=true, desc=Normal, g=true, r=true}
  n159 = LinearInterpolate(
   A=VectorParameter() {parameterName=FresnelColor_1, defaultValue=<rgba 0 0 0 1>}.RGB
   B=LinearInterpolate(
    A=VectorParameter() {parameterName=FresnelColor_2, defaultValue=<rgba 0.0176 0.0095 0.05 1>}.RGB
    B=Multiply(A=@n147, B=@n144.Result)
    Alpha=@n2
   )
   Alpha=Fresnel(ExponentIn=ScalarParameter() {defaultValue=2.178, parameterName=Fresnel_Exponent}, Normal=Transform(None=@n158)) {baseReflectFraction=0}
  )
  n160 = CollectionParameter() {
   collection=&FortniteMaterialParameters
   parameterName=FogDirectionalInscatteringColor
  }
  n161 = Desaturation(None=ComponentMask(None=@n160) {b=true, g=true, r=true}, Fraction=ScalarParameter() {defaultValue=-0.8836, parameterName=SunSaturation})
  n162 = ScalarParameter() {desc=2.443723, parameterName=SunBrightness}
  n163 = Power(Base=@n142, Exp=ScalarParameter() {defaultValue=4.6653, parameterName=SkyBlendPower})
  n164 = ComponentMask(None=LinearInterpolate(A=Add(A=@n146, B=@n159), B=Multiply(A=@n161, B=@n162), Alpha=Clamp(None=@n163))) {b=true, g=true, r=true}
  n165 = Multiply(A=Clamp(None=Add(A=@n140.R, B=@n141)), B=@n164)
  n166 = QualitySwitch(Default=@n165, Low=@n164, High=@n165, Medium=@n164)
  n167 = OneMinus(None=Multiply(A=@n25.R, B=@n25.R))
  n168 = VectorParameter() {parameterName=Lightning, defaultValue=<rgba 1.134 0.2734 2.5 1>}
  n169 = VectorParameter() {parameterName=Lightning_SuperStorm, defaultValue=<rgba 3 0.2014 1.574 1>}
  n170 = CollectionParameter() {collection=&MPC_SafeZoneClouds, parameterName=SuperStormActive}
  n171 = Frac(None=Multiply(A=Time(), B=ScalarParameter() {defaultValue=0.0195, parameterName=SpeedFlicker}))
  n172 = Divide(
   A=ComponentMask(None=WorldPosition() {worldPositionShaderOffset=WPT_ExcludeAllShaderOffsets}.XYZ) {g=true, r=true}
   B=ScalarParameter() {defaultValue=416374.875, parameterName=SizeLighting2}
  )
  n173 = ScalarParameter() {defaultValue=3, parameterName=DistortionTiwstScale}
  n174 = Multiply(
   A=TextureSample!(UVs=Multiply(A=@n152, B=@n173), Tex, Apply View MipBias) {samplerType=SAMPLERTYPE_LinearGrayscale, texture=/Game/Global/Textures/Default/Flowmaps/T_StormSwirlMask}.R
   B=ScalarParameter() {defaultValue=0.0476, parameterName=Distortion}
  )
  n175 = Add(A=Add(A=@n171, B=@n172), B=@n174)
  n176 = Add(A=Subtract(A=@n172, B=@n171), B=@n174)
  n177 = Multiply(
   A=@n167
   B=Multiply(
    A=TextureSample(UVs=@n175) {samplerType=SAMPLERTYPE_LinearGrayscale, texture=/Game/Global/Textures/Default/Tileable_Patterns/T_StormCloudDetail01_M}.B
    B=TextureSample(UVs=@n176) {samplerType=SAMPLERTYPE_LinearGrayscale, texture=&T_StormCloudDetail01_M}.B
   )
  )
  n178 = ScalarParameter() {defaultValue=2.1375, parameterName=ContrastBlueGlow_1}
  n179 = LinearInterpolate(
   A=VectorParameter() {parameterName=Lightning_1, defaultValue=<rgba 0.1393 0.0625 2 1>}.RGB
   B=Constant3Vector() {constant=<rgba 0.0157 0.006 0 1>}
   Alpha=@n2
  )
  n180 = LinearInterpolate(
   A=VectorParameter() {parameterName=Lightning_2_SuperStorm, defaultValue=<rgba 3 0.127 2.324 1>}.RGB
   B=Constant3Vector() {constant=<rgba 0.0157 0.006 0 1>}
   Alpha=@n2
  )
  n181 = Add(
   A=Multiply(A=Multiply(A=@n125, B=@n167), B=LinearInterpolate(A=@n168.RGB, B=@n169.RGB, Alpha=@n170))
   B=Multiply(A=Power(Base=@n177, Exp=@n178), B=LinearInterpolate(A=@n179, B=@n180, Alpha=@n170))
  )
  n182 = ScalarParameter() {defaultValue=29921.2461, parameterName=MaskRadius_Min}
  n183 = ScalarParameter() {defaultValue=338661.6562, parameterName=MaskRadius}
  n184 = Clamp(None=Divide(A=@n22, B=LinearInterpolate(A=@n182, B=@n183, Alpha=@n36)))
  n185 = MaterialFunctionCall(BaseNormal=@n74.New Tangent Space Vertex Normal, AdditionalNormal=@n158) {materialFunction=&BlendAngleCorrectedNormals}
  n186 = Divide(A=Subtract(A=@n36, B=ScalarParameter() {defaultValue=1.8621, parameterName=RemapRange})) {constB=6.5}
  n187 = Add(
   A=@n79
   B=LinearInterpolate(
    A=ScalarParameter() {parameterName=WaveOffset}
    B=ScalarParameter() {defaultValue=-927225, parameterName=WaveOffset_1}
    Alpha=Clamp(None=@n186)
   )
  )
  n188 = Multiply(A=Multiply(A=@n51, B=Add(A=@n49, B=@n37) {desc=Radius}), B=Normalize(VectorInput=@n53))
  n189 = Constant()
  n190 = Multiply(A=Ceil(None=@n51) {desc=Mask}, B=Add(A=AppendVector(A=@n188, B=@n189), B=@n62))
  n191 = Constant3Vector() {constant=<rgba 0 0 1 0>}
  n192 = Multiply(A=ScalarParameter() {defaultValue=-73655.1016, parameterName=CloudHorizonWPOffset}, B=@n71)
  n193 = VectorParameter() {parameterName=SSSColor, defaultValue=<rgba 0.1273 0 0.276 1>}
  n194 = Multiply(A=Constant3Vector() {constant=<rgba 0.2745 0.1444 0.0929 1>}, B=@n144.Result)
  n195 = MakeMaterialAttributes!(
   BaseColor=ComponentMask(None=@n166) {b=true, desc=BaseColor, g=true, r=true}
   Metallic
   Specular=ScalarParameter() {parameterName=Specular}
   Roughness=Constant() {r=1}
   Anisotropy
   EmissiveColor=ComponentMask(None=Add(A=@n181, B=@n166)) {b=true, desc=Emissive, g=true, r=true}
   Opacity=ScalarParameter() {defaultValue=1, parameterName=SSSOpacity}
   OpacityMask=ComponentMask(None=VertexInterpolator(VS=@n184).PS) {desc=OpcityMask, r=true}
   Normal=QualitySwitch(Default=@n185.Result, Low=@n158, High=@n185.Result, Medium=@n185.Result)
   Tangent
   WorldPositionOffset=Add(A=Add(A=@n187, B=@n190), B=Multiply(A=@n191, B=@n192))
   SubsurfaceColor=Multiply(A=LinearInterpolate(A=@n193.RGB, B=@n194, Alpha=@n2), B=OneMinus(None=@n25.G))
   ClearCoat
   ClearCoatRoughness
   AmbientOcclusion
   Refraction
   CustomizedUVs_0
   CustomizedUVs_1
   CustomizedUVs_2
   CustomizedUVs_3
   CustomizedUVs_4
   CustomizedUVs_5
   CustomizedUVs_6
   CustomizedUVs_7
   PixelDepthOffset=Multiply(
    A=Multiply(A=@n2) {constB=10000}
    B=MaterialFunctionCall() {materialFunction=&DitherTemporalAA}.Result
   ) {constA=10000}
   ShadingModel
   Displacement
  )
  n196 = GetMaterialAttributes(None=@n195) {attributeGetTypes [
    #1
    #2
    #3
    #4
   ]}
  n197 = MaterialFunctionCall!(Normal=@n74.New World Space Vertex Normal, Rotation=@n56) {materialFunction=&/MF_CloudRingTornado}
  n198 = SetMaterialAttributes!(
   MaterialAttributes=@n196.MaterialAttributes
   Base Color=LinearInterpolate(A=@n196.Base Color, B=@n197.Emissive, Alpha=@n197.Emissive)
   Emissive Color=LinearInterpolate(A=@n196.Emissive Color, B=@n197.Emissive, Alpha=@n197.Emissive)
   World Position Offset=Add(A=@n196.World Position Offset, B=@n197.WPO)
   Opacity Mask=@n197.OpacityMask
   Normal=LinearInterpolate(A=@n196.Normal, B=@n197.Normal, Alpha=@n197.Normal)
  ) {attributeSetTypes [
    #1
    #2
    #3
    #5
    #4
   ]}
  StaticSwitchParameter(True=@n198, False=@n195) {parameterName=IsTornade}
  TextureSample(UVs=AppendVector(A=@n113, B=Constant())) {samplerType=SAMPLERTYPE_Masks, texture=&T_LightingTimeline_M}
  TextureSample(UVs=AppendVector(A=@n123, B=Constant())) {samplerType=SAMPLERTYPE_Masks, texture=&T_LightingTimeline_M}
 ]
}
