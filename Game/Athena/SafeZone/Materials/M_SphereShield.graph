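# M_SphereShield is an additive, two-sided material for the Athena safe-zone sphere shield.
# Safe-zone collection parameters drive a radial color blend and boundary band, combined with scene-depth intersection glow and fresnel-weighted world-space lines for emissive.
# Forward and mobile shading paths add panning storm-cloud color.
# Opacity math combines intersection glow, face-dependent values, near-camera fade and a barrier-grid dissolve, animated except on forward and mobile.
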Material /Game/Athena/SafeZone/Materials/M_SphereShield
graph {
 properties {blendMode=BLEND_Additive, pixelDepthOffsetMode=PDOM_Legacy, twoSided=true, usedShadingModels=Default Lit}
 nodes [
  n1 = WorldPosition()
  Add(A=@n1.XYZ, B=Multiply(A=Time() {bOverride_Period=true, period=1}, B=Constant()))
  n2 = Multiply(A=@n1.XYZ, B=ScalarParameter() {defaultValue=0.0214, parameterName=LinesScale})
  n3 = Multiply(A=TextureSample(UVs=ComponentMask(None=@n2) {b=true}) {samplerType=SAMPLERTYPE_LinearColor, texture=/Game/Athena/SafeZone/Lines}.R) {desc=lines * distance fade}
  n4 = Add(A=ScalarParameter() {defaultValue=2.1384, parameterName=MinLineBirghtness}, B=@n3)
  n5 = WorldPosition()
  n6 = CollectionParameter() {collection=/Game/Athena/SafeZone/MPC_SafeZoneClouds, parameterName=SafeZoneLoc}
  n7 = Subtract(
   A=Distance(A=ComponentMask(None=@n5.XYZ) {g=true, r=true}, B=ComponentMask(None=@n6) {g=true, r=true})
   B=Multiply(
    A=CollectionParameter() {collection=&MPC_SafeZoneClouds, parameterName=SafeZoneScale}
    B=ScalarParameter() {defaultValue=185000, parameterName=MaxSafeZoneRadius}
   )
  )
  n8 = Saturate(None=Divide(A=@n7, B=ScalarParameter() {defaultValue=5, parameterName=SafeZoneSize_SoftBlend}))
  n9 = LinearInterpolate(
   A=VectorParameter() {parameterName=Color, defaultValue=<rgba 0.1927 0.0073 0 1>}.RGB
   B=VectorParameter() {parameterName=Color2, defaultValue=<rgba 0.184 0.5629 0.8 1>}.RGB
   Alpha=@n8
  )
  n10 = ScalarParameter() {defaultValue=1, parameterName=FresnelBrightness}
  n11 = Fresnel(
   ExponentIn=ScalarParameter() {defaultValue=3, parameterName=Fresnel}
   BaseReflectFractionIn=ScalarParameter() {defaultValue=0.0952, parameterName=BaseOpacity}
  )
  n12 = Divide(A=Abs(None=@n7), B=ScalarParameter() {defaultValue=94.0785, parameterName=SafeZoneSize_SoftBlendEdge})
  n13 = Multiply(A=OneMinus(None=Saturate(None=@n12)), B=ScalarParameter() {defaultValue=0.6476, parameterName=IntersectionBrightness})
  n14 = MaterialFunctionCall!(Distance Bias=Constant(), Distance Divisor=Constant() {r=250}, Scene Depth, Pixel Depth) {materialFunction=/Game/Effects/Fort_Effects/Materials/_Functions/MF_DistanceBehindTranslucency_01}
  n15 = Multiply(
   A=Power(Base=Add(A=@n13, B=@n14.Normalized Inverted), Exp=ScalarParameter() {defaultValue=4, parameterName=Intersection Edge Exponent})
   B=Multiply(A=@n9, B=ScalarParameter() {defaultValue=1.6779, parameterName=IntBrightness})
  )
  n16 = Add(A=Multiply(A=Multiply(A=@n4, B=@n9), B=Multiply(A=@n10, B=@n11)), B=@n15) {desc=+ intersection edge}
  Add(A=@n16, B=Multiply(A=Constant(), B=@n9)) {constB=0}
  FeatureLevelSwitch()
  LinearInterpolate(
   A=VectorParameter() {parameterName=Intersection Edge Color2, defaultValue=<rgba 0.5 3 40 1>}.RGB
   B=VectorParameter() {parameterName=Intersection Edge Color, defaultValue=<rgba 2.1447 40 0 1>}.RGB
   Alpha=@n8
  )
  LinearInterpolate(A=OneMinus(None=@n8), B=@n8, Alpha=Saturate(None=TwoSidedSign()))
  LinearInterpolate(Alpha=@n3)
  n17 = VectorParameter() {parameterName=MobileStormColor, defaultValue=<rgba 0.1719 0.022 0.1465 1>}
  n18 = Multiply(A=TextureCoordinate(), B=ScalarParameter() {defaultValue=5.4038, parameterName=TilingClouds})
  n19 = ScalarParameter() {defaultValue=0.0762, parameterName=CouldPanning}
  n20 = LinearInterpolate(
   A=ScalarParameter() {defaultValue=-0.5714, parameterName=Cm1}
   B=ScalarParameter() {defaultValue=0.8041, parameterName=Cm2}
   Alpha=TextureSample(UVs=Panner(Coordinate=@n18, Speed=@n19) {speedX=0.1, speedY=0.1}) {samplerType=SAMPLERTYPE_Masks, texture=/Game/Athena/SafeZone/Textures/StormCloudtubes_M}.B
  )
  n21 = Add(A=@n16, B=Multiply(A=@n17.RGB, B=Saturate(None=@n20)))
  MaterialFunctionCall(Emissive=Add(A=ShadingPathSwitch(Default=@n16, Deferred=@n16, Forward=@n21, Mobile=@n21), B=ScalarParameter() {parameterName=Pulse}) {constB=0}) {materialFunction=&MF_FX_EmissiveAdjustments}
  Multiply(A=@n17.RGB, B=OneMinus(None=@n3))
  n22 = TextureCoordinate() {vTiling=40}
  n23 = ScalarParameter() {defaultValue=0.5, parameterName=Progress, sliderMax=20, sliderMin=-20}
  Multiply(A=TextureSample(UVs=Add(A=@n22, B=@n23) {constB=0.5}) {samplerSource=SSM_Clamp_WorldGroupSettings, samplerType=SAMPLERTYPE_LinearColor, texture=&Lines}.RGB) {constB=2}
  n24 = LinearInterpolate(
   A=ScalarParameter() {defaultValue=0.4238, parameterName=Opacity}
   B=ScalarParameter() {defaultValue=0.5762, parameterName=Opacity2}
   Alpha=Saturate(None=TwoSidedSign())
  )
  n25 = PixelDepth()
  n26 = Constant() {r=-100}
  n27 = Saturate(None=Divide(A=Add(A=@n25.R, B=@n26), B=ScalarParameter() {defaultValue=300, parameterName=NearFade}))
  n28 = TextureSample(UVs=Multiply(A=TextureCoordinate(), B=ScalarParameter() {defaultValue=7.5607, parameterName=TilingBlocks})) {samplerType=SAMPLERTYPE_Masks, texture=/Game/Athena/Playlists/Barrier/Textures/T_BarrierGrid_M}
  n29 = Time()
  n30 = Frac(None=Add(A=@n28.R, B=Frac(None=@n29)))
  n31 = OneMinus(None=Multiply(A=Sine(None=@n30), B=@n28.G))
  n32 = ScalarParameter() {parameterName=DissolveAmount}
  n33 = Saturate(None=Multiply(A=ShadingPathSwitch(Default=@n31, Deferred=@n31, Forward=@n28.R, Mobile=@n28.R), B=@n32))
  Multiply(A=Multiply(A=Add(A=@n15, B=@n24), B=Multiply(A=@n27)), B=Multiply(A=OneMinus(None=@n33), B=OneMinus(None=@n32)))
 ]
}
