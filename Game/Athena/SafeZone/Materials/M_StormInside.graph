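# Translucent unlit material that draws the cloudy, lightning-lit interior of the safe zone storm on a mesh that can attach to the camera.
# View-projected positions, rotated by flight direction, drive three axis-blended cloud samples, one faking forward motion with crossfaded scale cycles.
# Clouds drive sun-tinted color, curve-animated lightning and opacity, with depth fade and optional sphere cutout.
# Safe-zone collection parameters fade the effect by camera position, throwing vertices away when invisible.
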Material /Game/Athena/SafeZone/Materials/M_StormInside
graph {
 properties {
  bUseLightmapDirectionality=false
  bUseTranslucencyVertexFog=false
  bUsedWithNanite=true
  blendMode=BLEND_Translucent
  pixelDepthOffsetMode=PDOM_Legacy
  shadingModel=MSM_Unlit
  usedShadingModels=Unlit
 }
 nodes [
  n1 = Normalize(VectorInput=Transform(None=Constant3Vector() {constant=<rgba 0 0 1 0>}) {transformSourceType=TRANSFORMSOURCE_Local})
  n2 = ObjectPositionWS()
  n3 = Subtract(A=DotProduct(A=CameraPositionWS(), B=@n1), B=DotProduct(A=@n2, B=@n1))
  Abs(None=@n3)
  n4 = Time()
  n5 = ScalarParameter() {defaultValue=0.3284, parameterName=Lightingspeed_2}
  Add(A=@n4, B=@n5)
  n6 = Constant3Vector() {constant=<rgba 0 0 1 0>}
  n7 = StaticSwitchParameter(
   True=ScalarParameter() {parameterName=RotationAngle}
   False=CollectionParameter() {collection=/Game/Athena/SafeZone/MPC_SafeZoneClouds, parameterName=FlightDirectionAngle}
  ) {parameterName=Debug}
  n8 = Constant3Vector()
  n9 = WorldPosition()
  n10 = CameraPositionWS()
  n11 = Subtract(A=ScalarParameter() {defaultValue=21379.3926, parameterName=ProjectionDistance}, B=Distance(A=@n9.XYZ, B=@n10))
  n12 = Normalize(VectorInput=Subtract(A=@n9.XYZ, B=@n10))
  n13 = Subtract(A=Add(A=@n9.XYZ, B=Multiply(A=@n11, B=@n12)), B=@n10)
  n14 = Add(A=RotateAboutAxis(NormalizedRotationAxis=@n6, RotationAngle=@n7, PivotPoint=@n8, Position=@n13), B=@n13) {desc=UVs}
  n15 = ComponentMask(None=@n14) {r=true}
  n16 = ComponentMask(None=@n14) {b=true}
  n17 = VectorParameter() {parameterName=PanDirection, defaultValue=<rgba -1 0 0.1143 1>}
  n18 = Multiply(A=Time() {period=5}, B=ScalarParameter() {defaultValue=0.5524, parameterName=Speed})
  n19 = Add(
   A=Divide(A=AppendVector(A=@n15, B=@n16), B=ScalarParameter() {defaultValue=-20293.4453, parameterName=Scale})
   B=ComponentMask(None=Multiply(A=@n17.RGB, B=@n18)) {b=true, r=true}
  )
  Add(A=@n19)
  n20 = Multiply(A=CameraPositionWS(), B=ScalarParameter() {defaultValue=-0.0005, parameterName=Paralax})
  n21 = ComponentMask(None=@n20) {b=true, g=true}
  n22 = ComponentMask(None=@n14) {g=true}
  n23 = AppendVector(A=@n22, B=@n16)
  n24 = ScalarParameter() {defaultValue=-4000, parameterName=Scale_2}
  n25 = ScalarParameter() {defaultValue=-33036.5312, parameterName=Scale_1}
  n26 = Multiply(A=Time() {period=5}, B=ScalarParameter() {defaultValue=0.2286, parameterName=Speed_1})
  n27 = Frac(None=@n26)
  n28 = OneMinus(None=@n27)
  n29 = Add(
   A=RotateAboutAxis(NormalizedRotationAxis=@n6, RotationAngle=@n7, PivotPoint=@n8, Position=@n12)
   B=ComponentMask(None=@n12) {b=true, g=true, r=true}
  ) {desc=Direction}
  n30 = ComponentMask(None=@n29) {r=true}
  n31 = Clamp(None=Sign(None=@n30))
  n32 = Divide(A=@n23, B=LinearInterpolate(A=@n24, B=@n25, Alpha=LinearInterpolate(A=@n28, B=@n27, Alpha=@n31)))
  Add(A=@n21, B=@n32)
  n33 = Frac(None=Add(A=@n26) {constB=0.5})
  n34 = LinearInterpolate(A=OneMinus(None=@n33), B=@n33, Alpha=@n31)
  n35 = Add(A=Divide(A=@n23, B=LinearInterpolate(A=@n24, B=@n25, Alpha=@n34)), B=Constant2Vector() {g=0.743, r=0.342})
  Add(A=@n21, B=@n35)
  n36 = VectorParameter() {parameterName=PanDirection2_1, defaultValue=<rgba 0 -0.6762 0 1>}
  n37 = Add(
   A=Divide(A=AppendVector(A=@n22, B=@n15), B=ScalarParameter() {defaultValue=-29100.7031, parameterName=Scale_5})
   B=ComponentMask(None=Multiply(A=@n18, B=@n36.RGB)) {g=true, r=true}
  )
  Add(A=@n37, B=AppendVector(A=ComponentMask(None=@n20) {g=true}, B=ComponentMask(None=@n20) {r=true}))
  n38 = Multiply(A=@n3) {constB=-1}
  n39 = ScalarParameter() {defaultValue=200, parameterName=F_Height}
  Add(A=@n38, B=@n39)
  Constant2Vector() {g=-0.0005, r=0.0005}
  n40 = ObjectPositionWS()
  n41 = Constant3Vector() {constant=<rgba 0 0 1 0>}
  Distance(A=Transform(None=Multiply(A=@n40, B=@n41)), B=CameraPositionWS())
  n42 = CurveAtlasRowParameter(CurveTime=Frac(None=Multiply(A=@n5, B=@n4))) {
   atlas=&AthenaCurveAtlas01
   curve=/Game/Athena/SafeZone/Textures/SormLightningDirection
   defaultValue=19
   parameterName=Curve1_1
  }
  n43 = Normalize(VectorInput=@n42)
  n44 = Constant3Vector()
  n45 = Normalize(VectorInput=Subtract(A=WorldPosition().XYZ, B=CameraPositionWS()))
  n46 = ComponentMask(None=Add(A=RotateAboutAxis(NormalizedRotationAxis=@n43, RotationAngle=@n42.R, PivotPoint=@n44, Position=@n45), B=@n45)) {b=true, r=true}
  n47 = ScalarParameter() {defaultValue=3.4869, parameterName=ScaleLightning}
  n48 = TextureSample!(UVs=Subtract(A=Divide(A=@n46, B=@n47)) {constB=0.5}, Tex, Level=ScalarParameter() {parameterName=Blur}, Apply View MipBias) {mipValueMode=TMVM_MipLevel, samplerType=SAMPLERTYPE_LinearColor, texture=/Game/Athena/SafeZone/Textures/T_Clouds_Lightning}
  n49 = Add(A=Time(), B=ScalarParameter() {defaultValue=1.6322, parameterName=LightingOffset_1})
  n50 = ScalarParameter() {defaultValue=0.0422, parameterName=LightingSpeed_1}
  n51 = CurveAtlasRowParameter(CurveTime=Frac(None=Multiply(A=@n49, B=@n50))) {
   atlas=&AthenaCurveAtlas01
   curve=/Game/Athena/SafeZone/Textures/StormLighting
   defaultValue=18
   parameterName=Curve1
  }
  n52 = Multiply(
   A=Power(Base=Multiply(A=@n48.G, B=@n51.R), Exp=ScalarParameter() {defaultValue=1.6213, parameterName=pow})
   B=VectorParameter() {parameterName=LightingColor1, defaultValue=<rgba 2.4688 1.3021 5 1>}.RGB
  )
  n53 = TextureSample!(UVs=@n19, Tex, Apply View MipBias) {samplerType=SAMPLERTYPE_Masks, texture=/Game/Athena/SafeZone/Textures/StormCloudtubes_M}
  n54 = Multiply(A=Constant3Vector() {constant=<rgba 0 1 0 0>}, B=@n53.B)
  n55 = Add(A=Cosine(None=@n26))
  n56 = LinearInterpolate(
   A=TextureSample(UVs=@n32) {samplerType=SAMPLERTYPE_Masks, texture=&StormCloudtubes_M}.B
   B=TextureSample(UVs=@n35) {samplerType=SAMPLERTYPE_Masks, texture=&StormCloudtubes_M}.B
   Alpha=Clamp(None=Divide(A=@n55))
  )
  n57 = Multiply(A=@n56, B=Constant3Vector() {constant=<rgba 0 0 1 0>})
  n58 = TextureSample(UVs=@n37) {samplerType=SAMPLERTYPE_Masks, texture=&StormCloudtubes_M}
  n59 = Constant3Vector() {constant=<rgba 1 0 0 0>}
  n60 = ComponentMask(None=@n29) {b=true}
  n61 = LinearInterpolate(
   A=LinearInterpolate(
    A=StaticSwitchParameter(True=@n54, False=@n53.B) {parameterName=Deubg Directions_2}
    B=StaticSwitchParameter(True=@n57, False=@n56) {parameterName=Deubg Directions_3}
    Alpha=Abs(None=@n30)
   )
   B=StaticSwitchParameter(True=Multiply(A=@n58.B, B=@n59), False=@n58.B) {parameterName=Deubg Directions_1}
   Alpha=Abs(None=@n60)
  )
  n62 = OneMinus(None=@n61)
  n63 = VectorParameter() {parameterName=Color_4, defaultValue=<rgba 0.0631 0.0247 0.1823 1>}
  n64 = Add(A=@n60)
  n65 = LinearInterpolate(A=VectorParameter() {parameterName=Color_3, defaultValue=<rgba 0.005 0.0055 0.0208 1>}.RGB, B=@n63.RGB, Alpha=@n64)
  n66 = CollectionParameter() {
   collection=&FortniteMaterialParameters
   parameterName=SunlightDirectionalVector
  }
  n67 = DotProduct(A=@n12, B=ComponentMask(None=@n66) {b=true, g=true, r=true})
  n68 = LinearInterpolate(
   A=VectorParameter() {parameterName=Color_1, defaultValue=<rgba 0.0298 0.0004 0.04 1>}.RGB
   B=LinearInterpolate(
    A=QualitySwitch(Default=@n65, Low=@n63.RGB, High=@n65, Medium=@n65)
    B=VectorParameter() {parameterName=Color_2, defaultValue=<rgba 0.2941 0.084 0.375 1>}.RGB
    Alpha=Clamp(None=@n67)
   )
   Alpha=@n62
  )
  n69 = StaticSwitchParameter(True=Add(A=Multiply(A=@n52, B=@n62) {desc=AddLightning}, B=@n68) {desc=AddLighting}, False=@n68) {defaultValue=true, parameterName=UseLightning}
  MaterialFunctionCall(Emissive=QualitySwitch(Default=@n69, Low=@n68, High=@n69, Medium=@n69)) {materialFunction=&MF_FX_EmissiveAdjustments}
  n70 = Multiply(A=Time(), B=ScalarParameter() {defaultValue=3.8777, parameterName=TimeShift})
  n71 = Frac(None=@n70)
  n72 = Multiply(A=@n71, B=OneMinus(None=@n71))
  n73 = ScalarParameter() {defaultValue=8.4141, parameterName=FadeSpeedbetweenShit}
  Multiply(A=Clamp(None=Multiply(A=@n72, B=@n73)), B=@n48.G)
  n74 = VectorParameter() {parameterName=SphereCenter, defaultValue=<rgba 0 0 500 1>}
  n75 = ScalarParameter() {defaultValue=1000, parameterName=SphereRadius}
  n76 = CameraPositionWS()
  n77 = Normalize(VectorInput=Subtract(A=WorldPosition().XYZ, B=@n76))
  Multiply(
   A=MaterialFunctionCall!(
    Sphere Position=@n74.RGB
    Sphere Radius=@n75
    RayDirection=@n77
    World position=MaterialFunctionCall() {materialFunction=&WorldPositionBehindTranslucency}.Result
   ) {materialFunction=/Engine/Functions/Engine_MaterialFunctions02/RayTracedSphere}.0-1 Sphere density
   B=ScalarParameter() {defaultValue=1, parameterName=SphereOpacity}
  )
  n78 = Subtract(A=@n3, B=@n39)
  n79 = ScalarParameter() {defaultValue=400, parameterName=F_BlendTop}
  n80 = OneMinus(None=Saturate(None=Divide(A=@n78, B=@n79)))
  n81 = Subtract(A=@n38, B=@n39)
  n82 = ScalarParameter() {defaultValue=400, parameterName=F_BlendBottom}
  n83 = OneMinus(None=Saturate(None=Divide(A=@n81, B=@n82)))
  n84 = Distance(A=Add(A=Multiply(A=@n3, B=@n1), B=@n2), B=CameraPositionWS())
  n85 = ScalarParameter() {defaultValue=200, parameterName=F_Height}
  n86 = Add(A=@n3, B=@n85)
  n87 = Multiply(A=@n85) {constB=2}
  n88 = LinearInterpolate(
   A=ScalarParameter() {defaultValue=500, parameterName=F_RadiusBottom}
   B=ScalarParameter() {defaultValue=500, parameterName=F_RadiusTop}
   Alpha=Saturate(None=Divide(A=@n86, B=@n87))
  ) {constAlpha=0}
  n89 = Saturate(None=Divide(A=Subtract(A=@n84, B=@n88), B=ScalarParameter() {defaultValue=500, parameterName=F_BlendRadius}))
  n90 = CameraPositionWS()
  n91 = CollectionParameter() {collection=&MPC_SafeZoneClouds, parameterName=SafeZoneCapHeight}
  n92 = ScalarParameter() {defaultValue=35142.25, parameterName=Heightoffset}
  n93 = Divide(
   A=Subtract(A=ComponentMask(None=@n90) {b=true}, B=StaticSwitchParameter(True=@n91, False=@n92) {parameterName=IsSkyCap})
   B=ScalarParameter() {defaultValue=5000, parameterName=Heightoffset_1}
  )
  n94 = Distance(
   A=ComponentMask(None=CollectionParameter() {collection=&MPC_SafeZoneClouds, parameterName=SafeZoneLoc}) {g=true, r=true}
   B=ComponentMask(None=CameraPositionWS()) {g=true, r=true}
  )
  n95 = LinearInterpolate(
   A=CollectionParameter() {collection=&MPC_SafeZoneClouds, parameterName=SafeZoneScale}
   Alpha=CollectionParameter() {collection=&MPC_SafeZoneClouds, parameterName=PreventStormShrinking}
  )
  n96 = Multiply(A=CollectionParameter() {collection=&MPC_SafeZoneClouds, parameterName=CloudStartingBias}) {constB=2}
  n97 = Add(
   A=LinearInterpolate(
    A=ScalarParameter() {defaultValue=16104.2061, parameterName=Ring_Radius_Min}
    B=ScalarParameter() {defaultValue=170209.0625, parameterName=Ring_Radius_Max}
    Alpha=Add(A=@n95, B=@n96)
   )
   B=ScalarParameter() {parameterName=Fade_Bias}
  )
  n98 = Clamp(None=Divide(A=Subtract(A=@n94, B=@n97), B=ScalarParameter() {defaultValue=15837.6094, parameterName=Fade_Distance}))
  n99 = OneMinus(None=@n98)
  n100 = StaticSwitchParameter(
   True=Multiply(A=Multiply(A=@n80, B=@n83), B=OneMinus(None=@n89))
   False=Multiply(A=Clamp(None=@n93), B=StaticSwitchParameter(True=@n99, False=@n98) {parameterName=IsSkyCap})
  ) {parameterName=HasConeShapeFalloff}
  n101 = LinearInterpolate(
   A=ScalarParameter() {defaultValue=0.2571, parameterName=Fade}
   B=ScalarParameter() {defaultValue=0.7429, parameterName=Fade_2}
   Alpha=Clamp(None=@n64)
  )
  n102 = ScalarParameter() {defaultValue=1.3534, parameterName=Fade_1}
  n103 = Multiply(
   A=MaterialFunctionCall(Transition Phase=@n100, Height Texture=@n61) {materialFunction=&HeightLerp}.Alpha
   B=Clamp(None=LinearInterpolate(A=@n101, B=@n102, Alpha=@n61))
  )
  n104 = SceneDepth()
  n105 = ScalarParameter() {defaultValue=3532.8796, parameterName=Depth}
  n106 = Multiply(A=@n61, B=ScalarParameter() {defaultValue=6957.7056, parameterName=Depth_1})
  n107 = Clamp(None=Divide(A=ComponentMask(None=@n104.R) {r=true}, B=Subtract(A=@n105, B=@n106)))
  Multiply(A=@n103, B=@n107)
  n108 = ComponentMask(None=ObjectPositionWS()) {b=true}
  n109 = ScalarParameter() {defaultValue=200, parameterName=F_Height}
  n110 = Add(A=@n108, B=@n109)
  n111 = ScalarParameter() {defaultValue=400, parameterName=F_BlendTop}
  n112 = ComponentMask(None=CameraPositionWS()) {b=true}
  n113 = Divide(A=Subtract(A=Add(A=@n110, B=@n111), B=@n112), B=ScalarParameter() {defaultValue=400, parameterName=F_BlendTop})
  n114 = Subtract(A=@n108, B=@n109)
  n115 = ScalarParameter() {defaultValue=400, parameterName=F_BlendBottom}
  n116 = Divide(A=Subtract(A=@n112, B=Subtract(A=@n114, B=@n115)), B=ScalarParameter() {defaultValue=400, parameterName=F_BlendBottom})
  n117 = ComponentMask(None=CameraPositionWS()) {g=true, r=true}
  n118 = ComponentMask(None=ObjectPositionWS()) {g=true, r=true}
  n119 = ScalarParameter() {defaultValue=500, parameterName=F_RadiusBottom}
  n120 = ScalarParameter() {defaultValue=500, parameterName=F_RadiusTop}
  n121 = ComponentMask(None=CameraPositionWS()) {b=true}
  n122 = ScalarParameter() {defaultValue=200, parameterName=F_Height}
  n123 = Subtract(A=ComponentMask(None=ObjectPositionWS()) {b=true}, B=@n122)
  n124 = Clamp(None=Divide(A=Subtract(A=@n121, B=@n123), B=Multiply(A=@n122) {constB=2})) {desc=lerpdistance for ConeShape top to botom}
  n125 = Divide(
   A=Subtract(A=Distance(A=@n117, B=@n118), B=LinearInterpolate(A=@n119, B=@n120, Alpha=@n124))
   B=ScalarParameter() {defaultValue=500, parameterName=F_BlendRadius}
  )
  Multiply(A=Multiply(A=Clamp(None=@n113), B=Clamp(None=@n116)), B=OneMinus(None=Clamp(None=@n125))) {constA=1}
  n126 = Ceil(None=@n70)
  n127 = ScalarParameter() {defaultValue=0.0167, parameterName=TimeShiftOffset}
  n128 = TextureSample(UVs=AppendVector(A=Multiply(A=@n126, B=@n127), B=Constant())) {samplerType=SAMPLERTYPE_Masks, texture=/Game/Gadgets/Assets/VinderTech_GliderChute/Textures/T_LightingTimeline_M}
  Normalize(VectorInput=Multiply(A=Subtract(A=@n128.RGB) {constB=-0.5}, B=ScalarParameter() {defaultValue=300, parameterName=UVOffsetLayer2}) {constB=2})
  Power(Base=@n67, Exp=ScalarParameter() {defaultValue=1, parameterName=Falloff})
  ScalarParameter() {defaultValue=8000, parameterName=Test}
  ScalarParameter() {defaultValue=8000, parameterName=Test}
  SceneTexture() {sceneTextureId=PPI_SceneDepth}
  n129 = MaterialFunctionCall!(
   Object Basis 1
   Object Basis 2
   Object Basis 3
   Camera Space Position Offset=VectorParameter() {parameterName=ScreenPosition, defaultValue=<rgba 0 0 22 1>}.RGB
  ) {materialFunction=/Engine/Functions/Engine_MaterialFunctions02/WorldPositionOffset/AttachMeshToTheCamera}
  n130 = Ceil(None=@n100)
  n131 = Add(A=Multiply(A=OneMinus(None=@n130), B=Constant3Vector() {constant=<rgba 0 0 -54654566400 0>}), B=@n129.World Position Offset)
  n132 = Multiply(A=VertexNormalWS(), B=ScalarParameter() {defaultValue=-3714.3762, parameterName=MobileOffsetPush})
  StaticSwitchParameter(
   True=StaticSwitchParameter(True=@n129.World Position Offset, False=Constant3Vector()) {defaultValue=true, parameterName=AttachToCamera}
   False=ShadingPathSwitch(Default=@n131, Mobile=Add(A=@n131, B=@n132)) {desc=Migrated from FeatureLevelSwitch}
  ) {parameterName=IsGroundLevel}
  n133 = LinearInterpolate(
   A=ScalarParameter() {defaultValue=-0.1748, parameterName=Opacity_1}
   B=ScalarParameter() {defaultValue=1.5592, parameterName=Opacity_2}
   Alpha=@n61
  )
  n134 = ScalarParameter() {defaultValue=0.3619, parameterName=Opacity_3}
  n135 = ScalarParameter() {defaultValue=2.3278, parameterName=Opacity_4}
  n136 = StaticSwitchParameter(True=@n133, False=LinearInterpolate(A=@n133, B=LinearInterpolate(A=@n134, B=@n135), Alpha=Divide(A=@n64))) {parameterName=IsGroundLevel}
  n137 = ShadingPathSwitch(Default=DepthFade(Opacity=@n136, FadeDistance=ScalarParameter() {defaultValue=2740.7466, parameterName=DepthFade}), Mobile=@n136) {desc=Migrated from FeatureLevelSwitch}
  n138 = StaticSwitchParameter(True=@n137, False=Multiply(A=VertexInterpolator(VS=@n100).PS, B=@n137)) {parameterName=IsGroundLevel}
  n139 = MaterialFunctionCall!(
   Sphere Radius=@n75
   Sphere Falloff=ScalarParameter() {defaultValue=1000, parameterName=SphereFalloff}
   Unit Ray Direction=@n77
   Sphere Center=@n74.RGB
   Ray Origin=@n76
  ) {materialFunction=/Game/Athena/Environments/MaterialFunctions/RayTraceSphereFalloffMinClamp}
  n140 = StaticSwitchParameter(True=Multiply(A=@n138, B=OneMinus(None=@n139.Clamped Falloff)), False=@n138) {parameterName=IsSphereShape}
  StaticSwitchParameter(
   True=Multiply(A=CollectionParameter() {collection=&MPC_SafeZoneClouds, parameterName=GlobalOpacity}, B=@n140)
   False=@n140
  ) {parameterName=UseGlobleOpacity}
  n141 = ComponentMask(None=SceneTexture() {sceneTextureId=PPI_PostProcessInput0}.Color) {b=true, g=true, r=true}
  n142 = Multiply(A=VectorParameter() {parameterName=Tint, defaultValue=<rgba 0.2083 0.2083 0.2083 1>}.RGB, B=@n141)
  StaticSwitchParameter(True=@n61, False=LinearInterpolate(A=@n141, B=LinearInterpolate(A=@n142, B=@n69, Alpha=@n107), Alpha=@n103)) {parameterName=Deubg Directions}
  Subtract(B=@n39)
  VertexNormalWS()
 ]
}
