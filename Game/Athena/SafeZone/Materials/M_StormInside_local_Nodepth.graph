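# Translucent unlit material for the inside of the Athena safe zone storm, a local-position variant that draws without depth testing.
# All shading comes from a single material function, MF_Storm_Inside_localpostion, through material attributes.
# Depth test and translucent vertex fog are disabled.
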Material /Game/Athena/SafeZone/Materials/M_StormInside_local_Nodepth
graph {
 properties {
  bDisableDepthTest=true
  bUseLightmapDirectionality=false
  bUseMaterialAttributes=true
  bUseTranslucencyVertexFog=false
  blendMode=BLEND_Translucent
  pixelDepthOffsetMode=PDOM_Legacy
  shadingModel=MSM_Unlit
  usedShadingModels=Unlit
 }
 nodes [MaterialFunctionCall() {materialFunction=&/MF_Storm_Inside_localpostion}]
}
