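# Additive, unlit, two-sided material for the WildEstate storm shield, the safe zone wall in Athena.
# Emissive layers world-aligned star noise, caustic warp, stepped grid snow, scanlines, panning sheen wipes and a depth-fade contact line.
# Opacity combines height, near and far fades with a camera sphere mask, while a WPO branch pushes distant vertices far down.
# Quality, platform and shading-path switches simplify noise, wipes and fades on weaker targets.
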Material /Game/Athena/SafeZone/Materials/M_WildEstate_StormShield
graph {
 properties {
  bAutomaticallySetUsageInEditor=false
  bUsedWithSplineMeshes=true
  blendMode=BLEND_Additive
  shadingModel=MSM_Unlit
  twoSided=true
  usedShadingModels=Unlit
 }
 nodes [
  n1 = NamedRerouteUsage()
  n2 = NamedRerouteUsage()
  n3 = NamedRerouteUsage()
  n4 = Add(A=Multiply(A=NamedRerouteUsage(), B=Clamp(None=@n3) {minDefault=0.985}), B=Multiply(A=NamedRerouteUsage(), B=NamedRerouteUsage()))
  n5 = NamedRerouteUsage()
  n6 = Multiply(A=@n5) {constB=0.1}
  n7 = Multiply(A=@n5) {constB=0.03}
  n8 = LinearInterpolate(
   A=Add(A=Multiply(A=@n2) {constB=0.033}, B=@n4)
   B=LinearInterpolate(A=Subtract(A=@n4, B=@n6), B=Add(A=@n7, B=@n4), Alpha=@n5)
   Alpha=NamedRerouteUsage()
  )
  n9 = MaterialFunctionCall(Base=@n8, Blend=NamedRerouteUsage()) {materialFunction=&Blend_Overlay}
  n10 = NamedRerouteUsage()
  n11 = NamedRerouteUsage()
  n12 = LinearInterpolate(A=@n11, B=Multiply(A=NamedRerouteUsage(), B=@n11), Alpha=Add(A=NamedRerouteUsage()) {constB=0.33})
  n13 = Add(
   A=Add(A=Multiply(A=@n1) {constB=0.05}, B=LinearInterpolate(A=@n8, B=@n9.Result, Alpha=@n10))
   B=Multiply(A=NamedRerouteUsage(), B=Saturate(None=@n12))
  )
  n14 = NamedRerouteUsage()
  n15 = Multiply(A=@n14) {constB=0.85}
  n16 = Multiply(A=NamedRerouteUsage(), B=NamedRerouteUsage())
  n17 = Desaturation(None=LinearInterpolate(A=@n14, B=Subtract(A=@n15, B=@n16)))
  n18 = Saturate(None=Multiply(A=@n17, B=NamedRerouteUsage()))
  n19 = NamedRerouteUsage()
  n20 = Add(
   A=MaterialFunctionCall!("Xb1,PS4"=@n13, "PC-High&&Epic,XSX,PS5"=@n13, "PC-Perf,Android"=@n8, Switch=@n13, "PC-Low(NotPerf)&&Medium"=@n8) {materialFunction=/Game/Effects/Niagara/MasterMaterials/Functions/Common/PlatformSwitch_3D}.Result
   B=Multiply(A=@n18, B=@n19)
  )
  LinearInterpolate(A=@n20, B=Multiply(A=Multiply(A=@n17) {constB=1.5}, B=@n19), Alpha=@n18)
  n21 = Add(A=PixelDepth().R, B=Constant() {r=-100})
  n22 = ScalarParameter() {defaultValue=300, parameterName=NearFade}
  n23 = Divide(A=WorldPosition().XYZ, B=ScalarParameter() {defaultValue=23000, parameterName=HeightGradientScale})
  n24 = OneMinus(None=Saturate(None=ComponentMask(None=@n23) {b=true}))
  n25 = LinearInterpolate(
   A=Constant() {r=0.4}
   B=Constant() {r=0.7}
   Alpha=MaterialFunctionCall!(Distance Bias=Constant(), Distance Divisor=Constant() {r=10000}, Scene Depth, Pixel Depth) {materialFunction=/Game/Effects/Fort_Effects/Materials/_Functions/MF_DistanceBehindTranslucency_01}.Normalized
  )
  n26 = ShadingPathSwitch(Default=@n25, Mobile=Add(A=@n25, B=Constant() {r=-0.2})) {desc=Migrated from FeatureLevelSwitch}
  n27 = Multiply(A=@n24, B=Add(A=NamedRerouteUsage(), B=@n26))
  n28 = MaterialFunctionCall(Vector 2=ComponentMask(None=WorldPosition() {worldPositionShaderOffset=WPT_CameraRelative}.XYZ) {g=true, r=true}) {materialFunction=/Engine/Functions/Engine_MaterialFunctions02/Utility/VectorLength}
  n29 = Constant() {r=-100}
  n30 = ScalarParameter() {defaultValue=102045.8516, group=Visual Adjustments, parameterName=FarFade WPO}
  n31 = Saturate(None=Divide(A=Add(A=@n28.V2 Length, B=@n29), B=@n30))
  n32 = Multiply(A=Saturate(None=Divide(A=@n21, B=@n22)), B=Multiply(A=QualitySwitch(Default=@n27, Low=@n26), B=OneMinus(None=@n31)))
  n33 = Add(A=@n32, B=NamedRerouteUsage())
  n34 = ShadingPathSwitch(Default=@n32, Deferred=@n32, Mobile=@n33) {desc=Migrated from FeatureLevelSwitch}
  n35 = QualitySwitch(Default=@n34, Low=@n33, High=@n34, Medium=@n34)
  Multiply(A=@n35, B=ScalarParameter() {defaultValue=1, parameterName=OverallOpacityMult})
  n36 = Constant() {r=0.75}
  NamedRerouteDeclaration(None=MaterialFunctionCall(Value_Morning=@n36, Value_Day=Constant() {r=1}, Value_Evening=@n36, Value_Night=Constant() {r=0.5}) {materialFunction=&Scalar_Time_of_Day}.Result) {name=TODScalar, nodeColor=<rgba 0.5938 0 1 1>}
  n37 = PixelDepth()
  n38 = ScalarParameter() {defaultValue=12500, desc=74537.734375, group=Visual Adjustments, parameterName=Max distance Thick Scanline}
  NamedRerouteDeclaration(None=Saturate(None=Divide(A=@n37.R, B=@n38))) {name=DistanceScanlineThick, nodeColor=<rgba 0 0.0391 1 1>}
  n39 = TextureCoordinate()
  n40 = Add(A=Multiply(A=Multiply(A=@n39) {constB=814}) {constB=0.0002}, B=Constant2Vector() {r=0.25})
  n41 = Time()
  n42 = Frac(None=DotProduct(A=Multiply(A=@n40, B=@n40), B=@n41))
  n43 = Multiply(A=@n42, B=@n42)
  n44 = Multiply(A=@n41) {constB=2}
  NamedRerouteDeclaration(None=Frac(None=Multiply(A=@n43, B=@n44))) {name=ProcNoise, nodeColor=<rgba 0.3828 1 0 1>}
  n45 = MaterialFunctionCall!(
   TextureObject=TextureObject() {samplerType=SAMPLERTYPE_Normal, texture=/Game/TheOutpost/Textures/TilingWeakPointNormal}
   TextureSize=VectorParameter() {desc=37.719421, parameterName=CausticWorldScale, defaultValue=<rgba 512 512 256 0>}.RGB
   WorldPosition
   Export Float 4=StaticBool()
   In
   World Space Normal
   ProjectionTransitionContrast=Constant() {r=1}
  ) {materialFunction=/Game/Athena/SafeZone/WorldAlignedTexture_ScalePan}
  n46 = Multiply(A=Time() {desc=15.015014, period=1}, B=ScalarParameter() {defaultValue=0.08, desc=0.0666, parameterName=Sine Speed A_1})
  n47 = ComponentMask(None=Frac(None=Add(A=@n45.XY Texture, B=@n46))) {g=true}
  n48 = Constant() {r=0.5}
  n49 = OneMinus(None=Multiply(A=Distance(A=@n47, B=@n48)) {constB=2})
  NamedRerouteDeclaration(None=LinearInterpolate(A=@n49, B=Round(None=@n49), Alpha=ScalarParameter() {defaultValue=0.25, parameterName=Caustic Round LERP})) {name=CasuticWarp, nodeColor=<rgba 1 0 0.2578 1>}
  n50 = ScalarParameter() {defaultValue=0.15, group=Sheen, parameterName=SheenSize1}
  n51 = WorldPosition()
  n52 = Normalize(VectorInput=VectorParameter() {group=Sheen, parameterName=SweepDirectionCamSpace, defaultValue=<rgba 0 0 1 1>}.RGB)
  n53 = ComponentMask(None=Multiply(A=DotProduct(A=@n51.XYZ, B=@n52), B=@n52)) {b=true, r=true}
  n54 = MaterialFunctionCall(UVs=@n53, "Rotation Angle (0-1)"=Constant() {r=0.5}) {materialFunction=&CustomRotator}
  n55 = ScalarParameter() {defaultValue=-550, group=Sheen, parameterName=SweepSpeed1}
  n56 = Add(A=NamedRerouteUsage(), B=Panner(Coordinate=@n54.Rotated Values, Time=Time(), Speed=@n55) {speedX=0.2, speedY=0.2})
  n57 = ScalarParameter() {defaultValue=12500, group=Sheen, parameterName=SweepFreq1}
  n58 = OneMinus(None=Divide(A=ComponentMask(None=@n56) {g=true}, B=@n57))
  NamedRerouteDeclaration(None=SmoothStep(Min=@n50, Max=Constant(), Value=Frac(None=@n58))) {name=Sheen1, nodeColor=<rgba 0 1 0.8359 1>}
  n59 = ScalarParameter() {defaultValue=10000, desc=74537.734375, group=Visual Adjustments, parameterName=Max distance Scanline}
  NamedRerouteDeclaration(None=Saturate(None=Divide(A=@n37.R, B=@n59))) {name=DistanceScanline, nodeColor=<rgba 0 0.0391 1 1>}
  n60 = NamedRerouteUsage()
  n61 = Constant()
  n62 = ScalarParameter() {defaultValue=1250, group=Sheen, parameterName=SweepSpacing}
  n63 = Add(A=@n54.Rotated Values, B=@n62) {constB=1000}
  n64 = Time()
  n65 = ComponentMask(None=Add(A=NamedRerouteUsage(), B=Panner(Coordinate=@n63, Time=@n64, Speed=@n55) {speedX=0.2, speedY=0.2})) {g=true}
  n66 = Frac(None=OneMinus(None=Divide(A=@n65, B=@n57)))
  NamedRerouteDeclaration(None=Multiply(A=Clamp(None=@n60) {minDefault=0.85}, B=SmoothStep(Min=@n50, Max=@n61, Value=@n66))) {name=Sheen2, nodeColor=<rgba 0 1 0.8359 1>}
  n67 = Constant()
  NamedRerouteDeclaration(None=MaterialFunctionCall(
    Input=CollectionParameter() {collection=/Game/Packages/Fortress_SharedMaterials/GlobalMaterialParameters/DirectSafeZoneMPC, parameterName=SafeRadius}
    Input Low=@n67
    Input High=Constant() {r=185185}
    Target Low=@n67
    Target High=Constant() {r=1}
   ) {materialFunction=&RemapValueRange}.Result) {name=Scale Storm Shield, nodeColor=<rgba 0 0.25 1 1>}
  n68 = Saturate(None=NamedRerouteUsage())
  n69 = MaterialFunctionCall(
   TextureObject=TextureObject() {samplerType=SAMPLERTYPE_Masks, texture=/Game/Athena/SafeZone/Textures/T_WildEstate_StormShield_M}
   TextureSize=ScalarParameter() {defaultValue=128, parameterName=GridWorldScale}
   Export Float 4=StaticBool()
   ProjectionTransitionContrast=Constant() {r=5}
  ) {materialFunction=&WorldAlignedTexture_ScalePan}
  n70 = ComponentMask(None=@n69.XY Texture) {r=true}
  n71 = Frac(None=Multiply(A=Time(), B=ScalarParameter() {defaultValue=0.8, parameterName=Sine Speed Grid}))
  n72 = Divide(A=Ceil(None=Multiply(A=@n71) {constB=24}), B=Constant() {r=12})
  n73 = Multiply(A=Sine(None=Add(A=@n70, B=@n72)), B=ComponentMask(None=@n69.XY Texture) {g=true})
  n74 = Multiply(A=Multiply(A=Saturate(None=@n73)) {constB=0.5}, B=OneMinus(None=NamedRerouteUsage()))
  NamedRerouteDeclaration(None=Saturate(None=Multiply(A=@n68, B=@n74))) {name=SnowNoise, nodeColor=<rgba 1 0.7031 0 1>}
  n75 = Add(A=NamedRerouteUsage(), B=Panner(Coordinate=Add(A=@n63, B=@n62) {constB=1000}, Time=Time(), Speed=@n55) {speedX=0.2, speedY=0.2})
  n76 = OneMinus(None=Divide(A=ComponentMask(None=@n75) {g=true}, B=@n57))
  NamedRerouteDeclaration(None=SmoothStep(Min=@n50, Max=Constant(), Value=Frac(None=@n76))) {name=Sheen3, nodeColor=<rgba 0 1 0.8359 1>}
  n77 = TextureObject() {texture=/Game/Athena/SafeZone/PulseLancerAssets/T_StarSky_SafeZone}
  n78 = VectorParameter() {desc=37.719421, parameterName=CausticWorldScale, defaultValue=<rgba 512 512 256 0>}
  n79 = Divide(A=Time(), B=ScalarParameter() {defaultValue=500, group=Noises, parameterName=Wave Inensity Sped}) {constB=20}
  n80 = Divide(
   A=Add(A=@n78.RGB, B=NamedRerouteDeclaration(None=@n79) {name=Time, nodeColor=<rgba 1 0 0.0769 1>})
   B=ScalarParameter() {defaultValue=1.6, group=Noises, parameterName=Noise 1 Size}
  )
  n81 = StaticBool()
  n82 = Constant() {r=1}
  n83 = MaterialFunctionCall(TextureObject=@n77, TextureSize=@n80, Export Float 4=@n81, ProjectionTransitionContrast=@n82) {materialFunction=&WorldAlignedTexture_ScalePan}
  n84 = ComponentMask(None=@n83.XYZ Texture) {r=true}
  n85 = Multiply(
   A=Add(A=NamedRerouteDeclaration(None=@n84) {name=Noise 2, nodeColor=<rgba 0 0.5078 1 1>}) {constB=0.2}
   B=ComponentMask(None=@n83.XYZ Texture) {b=true}
  )
  n86 = Divide(
   A=Add(A=Multiply(A=@n78.RGB) {constB=0.672}, B=NamedRerouteUsage())
   B=ScalarParameter() {defaultValue=0.8, group=Noises, parameterName=Noise 2 Size}
  )
  n87 = ComponentMask(None=MaterialFunctionCall(TextureObject=@n77, TextureSize=@n86, Export Float 4=@n81, ProjectionTransitionContrast=@n82) {materialFunction=&WorldAlignedTexture_ScalePan}.XYZ Texture) {r=true}
  n88 = Multiply(A=@n85) {constB=0.6}
  NamedRerouteDeclaration(None=QualitySwitch(Default=Multiply(A=@n85, B=@n87), Low=@n88, Medium=@n88)) {name=Moving Noise, nodeColor=<rgba 1 0.3984 0 1>}
  n89 = Multiply(A=Time(), B=ScalarParameter() {defaultValue=0.1, group=Noises, parameterName=JitterSpeed, sortPriority=103})
  n90 = Add(A=Cosine(None=Multiply(A=@n89) {constB=4.5}), B=@n89)
  n91 = Multiply(A=Sine(None=Multiply(A=@n90) {constB=3})) {constB=0.66}
  n92 = Multiply(A=@n89) {constB=4}
  n93 = Add(A=@n91, B=Multiply(A=Sine(None=@n92)) {constB=0.5})
  NamedRerouteDeclaration(None=Saturate(None=ConstantBiasScale(None=@n93))) {name=JitterSpeed2, nodeColor=<rgba 0.875 0 1 1>}
  n94 = Time()
  n95 = Multiply(A=@n94, B=Constant() {r=0.2})
  n96 = Divide(
   A=ComponentMask(None=WorldPosition().XYZ) {b=true, g=true}
   B=LinearInterpolate(
    A=ScalarParameter() {defaultValue=5, group=Lines, parameterName=Tlling Lines}
    B=ScalarParameter() {defaultValue=2048, group=Lines, parameterName=Tilling Lines Max Scale}
    Alpha=NamedRerouteUsage()
   )
  )
  n97 = TextureSample(UVs=VertexInterpolator(VS=Add(A=@n95, B=@n96)).PS) {samplerType=SAMPLERTYPE_LinearGrayscale, texture=/Game/Athena/SafeZone/PulseLancerAssets/T_GradientLine_HQ16_SafeZone}
  n98 = ScalarParameter() {defaultValue=1.5, group=Lines, parameterName=Scanline_Intensity}
  n99 = ScalarParameter() {defaultValue=2, group=Lines, parameterName=Scanline_Contrast}
  n100 = Multiply(A=@n99) {constB=2}
  n101 = NamedRerouteUsage()
  n102 = Saturate(None=Power(Base=Multiply(A=@n97.G, B=@n98) {constB=1.5}, Exp=LinearInterpolate(A=@n99, B=@n100, Alpha=@n101) {constA=0.2, constB=4}))
  n103 = Saturate(None=OneMinus(None=NamedRerouteUsage()))
  n104 = Power(Base=Add(A=MaterialFunctionCall() {materialFunction=&BoundingBoxBased_0-1_UVW}.B) {constB=0.25}) {constExponent=6}
  NamedRerouteDeclaration(None=Subtract(A=Multiply(A=@n102, B=@n103), B=@n104)) {name=Lines, nodeColor=<rgba 0.8984 1 0 1>}
  n105 = DepthFade(
   Opacity=Constant() {r=1}
   FadeDistance=ScalarParameter() {defaultValue=197.4958, group=Visual Adjustments, parameterName=Depth fade Distance}
  )
  n106 = Multiply(
   A=Saturate(None=OneMinus(None=@n105))
   B=ScalarParameter() {defaultValue=2.5, group=Visual Adjustments, parameterName=Contact line Intensity Depth}
  ) {constB=2}
  NamedRerouteDeclaration(None=Multiply(A=@n106, B=@n106)) {name=Depth, nodeColor=<rgba 0.4297 0 1 1>}
  n107 = LocalPosition()
  n108 = ScalarParameter() {defaultValue=133.5986, group=Pos Mask BG Color, parameterName=Pos Mask}
  n109 = OneMinus(None=Divide(A=Add(A=@n107.Z, B=@n108), B=ScalarParameter() {defaultValue=190.8016, group=Pos Mask BG Color, parameterName=Pos Mask Fade}))
  n110 = Saturate(None=@n109)
  n111 = NamedRerouteUsage()
  n112 = Multiply(
   A=Multiply(A=Add(A=@n111) {constB=0.2}, B=ScalarParameter() {defaultValue=3, group=Pos Mask BG Color, parameterName=Pos Mask Noise})
   B=LinearInterpolate(A=Constant() {r=2}, Alpha=NamedRerouteUsage()) {constA=2, constB=1.2}
  )
  NamedRerouteDeclaration(None=Clamp(None=Multiply(A=@n110, B=@n112)) {minDefault=0.75}) {name=Pos Mask Noise, nodeColor=<rgba 0.4062 1 0 1>}
  NamedRerouteDeclaration(None=VectorParameter() {group=Color, parameterName=Color, defaultValue=<rgba 0 0.1524 1 1>}.RGB) {name=Shield Color, nodeColor=<rgba 0 0.4924 1 1>}
  n113 = Multiply(A=@n89) {constB=3.6}
  n114 = Multiply(A=@n93, B=Add(A=Cosine(None=@n113), B=Sine(None=@n89)))
  NamedRerouteDeclaration(None=Saturate(None=ConstantBiasScale(None=@n114))) {name=JitterSPeed, nodeColor=<rgba 0 0.1094 1 1>}
  NamedRerouteDeclaration(None=@n109) {name=Pos Mask, nodeColor=<rgba 0 1 0.9297 1>}
  NamedRerouteDeclaration(None=Saturate(None=ConstantBiasScale(None=@n91))) {name=JitterSpeed3, nodeColor=<rgba 0.875 0 1 1>}
  NamedRerouteDeclaration(None=Divide(A=@n94, B=ScalarParameter() {defaultValue=242.494, group=Lines, parameterName=Line Movement}) {constB=40}) {name=Time Slow, nodeColor=<rgba 0 1 0.2969 1>}
  n115 = ScalarParameter() {defaultValue=0.05, group=Sheen, parameterName=SheenSize2}
  n116 = Constant()
  n117 = Time()
  n118 = ScalarParameter() {defaultValue=-1500, group=Sheen, parameterName=SweepSpeed2}
  n119 = ComponentMask(None=Add(A=NamedRerouteUsage(), B=Panner(Coordinate=@n54.Rotated Values, Time=@n117, Speed=@n118) {speedX=0.2, speedY=0.2})) {g=true}
  n120 = ScalarParameter() {defaultValue=13500, group=Sheen, parameterName=SweepFreq2}
  n121 = Frac(None=OneMinus(None=Divide(A=@n119, B=@n120)))
  NamedRerouteDeclaration(None=Multiply(A=SmoothStep(Min=@n115, Max=@n116, Value=@n121), B=ScalarParameter() {defaultValue=0.33, group=Sheen, parameterName=Sheen2Brightness}) {constB=0.18}) {name=SheenSmall, nodeColor=<rgba 0 1 0.8359 1>}
  n122 = MaterialFunctionCall(
   TextureObject=TextureObject() {samplerType=SAMPLERTYPE_Masks, texture=&T_WildEstate_StormShield_M}
   TextureSize=ScalarParameter() {defaultValue=3000, parameterName=GridWorldScale2}
   Export Float 4=StaticBool()
   ProjectionTransitionContrast=Constant() {r=5}
  ) {materialFunction=&WorldAlignedTexture_ScalePan}
  n123 = Multiply(A=Time(), B=ScalarParameter() {defaultValue=0.2, parameterName=Sine Speed Grid2})
  n124 = Ceil(None=Multiply(A=Frac(None=@n123)) {constB=24})
  n125 = Constant() {r=12}
  n126 = Sine(None=Add(A=ComponentMask(None=@n122.XY Texture) {r=true}, B=Divide(A=@n124, B=@n125)))
  n127 = ComponentMask(None=@n122.XY Texture) {g=true}
  NamedRerouteDeclaration(None=Saturate(None=Multiply(A=@n126, B=@n127))) {name=DistanceSnow, nodeColor=<rgba 0 1 0.2734 1>}
  n128 = Divide(A=@n37.R, B=ScalarParameter() {defaultValue=5000, group=Visual Adjustments, parameterName=Distance fade})
  n129 = Constant() {r=0.7}
  NamedRerouteDeclaration(None=Saturate(None=MaterialFunctionCall(In=@n128, Contrast=@n129) {materialFunction=&CheapContrast}.Result)) {name=Distance Mask, nodeColor=<rgba 0 1 0.8594 1>}
  n130 = NamedRerouteUsage()
  n131 = NamedRerouteUsage()
  n132 = NamedRerouteUsage()
  n133 = Multiply(A=NamedRerouteUsage()) {constB=1.7}
  n134 = Multiply(
   A=Add(A=Add(A=@n132, B=@n133), B=NamedRerouteUsage())
   B=ScalarParameter() {defaultValue=0.5, group=TVStatic, parameterName=SweepSecBrightness}
  ) {constB=0.33}
  n135 = NamedRerouteUsage()
  n136 = QualitySwitch(Default=Multiply(A=Add(A=@n131, B=@n134), B=Clamp(None=@n135) {minDefault=0.925}), Low=Constant())
  n137 = Multiply(
   A=Multiply(A=Add(A=@n130, B=@n136), B=ScalarParameter() {defaultValue=0.5, group=TVStatic, parameterName=SweepBrightness}) {constB=0.5}
   B=Clamp(None=NamedRerouteUsage()) {minDefault=0.925}
  )
  n138 = Saturate(None=Multiply(A=NamedRerouteUsage(), B=Saturate(None=@n137)))
  NamedRerouteDeclaration(None=@n138) {name=Wipes, nodeColor=<rgba 1 0.2812 0 1>}
  n139 = SphereMask(A=CameraPositionWS(), B=WorldPosition().XYZ, Radius=ScalarParameter() {defaultValue=1500, parameterName=CameraRadius}) {hardnessPercent=0}
  NamedRerouteDeclaration(None=LinearInterpolate(A=@n35, Alpha=NamedRerouteDeclaration(None=@n139) {name=CameraProtection, nodeColor=<rgba 1 0.6094 0 1>}) {constB=0}) {name=Opacity, nodeColor=<rgba 0 0.9531 1 1>}
  n140 = Add(
   A=Multiply(A=Time(), B=ScalarParameter() {defaultValue=0.001, group=TVStatic, parameterName=ScanlineSpeed})
   B=Multiply(A=TextureCoordinate()) {constB=35}
  )
  n141 = ViewSize()
  n142 = Floor(None=Multiply(A=ComponentMask(None=@n140) {g=true}, B=ComponentMask(None=@n141) {g=true}))
  n143 = ScalarParameter() {defaultValue=2, group=TVStatic, parameterName=ScanlineScale}
  n144 = Multiply(
   A=Multiply(A=ScalarParameter() {defaultValue=2, group=TVStatic, parameterName=ScanlineChromatic Offset}, B=Fmod(A=@n142, B=@n143))
   B=ScalarParameter() {defaultValue=0.33, group=TVStatic, parameterName=SecondaryScanlineBrightness}
  )
  n145 = NamedRerouteUsage()
  n146 = NamedRerouteUsage()
  NamedRerouteDeclaration(None=Multiply(A=Multiply(A=@n144, B=@n145), B=OneMinus(None=@n146))) {name=Scanline, nodeColor=<rgba 0.7109 1 0 1>}
  n147 = ScalarParameter() {defaultValue=1.1, group=Visual Adjustments, parameterName=Buffer}
  n148 = Constant()
  NamedRerouteDeclaration(None=If(
    A=@n28.V2 Length
    B=Multiply(A=@n30, B=@n147) {constB=1.36, desc=add some buffer}
    "A > B"=Constant3Vector() {constant=<rgba 0 0 -999999 0>}
    "A == B"=@n148
    "A < B"=@n148
   )) {name=WPO, nodeColor=<rgba 1 0.7266 0 1>}
  NamedRerouteDeclaration(None=Saturate(None=@n24)) {name=HeightGradScale, nodeColor=<rgba 0 1 0.1562 1>}
  n149 = Divide(A=@n37.R, B=ScalarParameter() {defaultValue=50000, desc=74537.734375, group=Visual Adjustments, parameterName=Max distance Fade})
  n150 = Multiply(A=@n149, B=ScalarParameter() {defaultValue=1.5, group=Stars, parameterName=Star Fade Distance}) {constB=0.5}
  NamedRerouteDeclaration(None=OneMinus(None=Saturate(None=@n150))) {name=star distance fade, nodeColor=<rgba 1 0 0.1641 1>}
  NamedRerouteDeclaration(None=@n104) {name=BOundingMask, nodeColor=<rgba 0 1 0.9531 1>}
  NamedRerouteDeclaration(None=@n74) {name=Snow, nodeColor=<rgba 1 0 0.6797 1>}
  n151 = Constant()
  n152 = Constant() {r=-27928.2363}
  n153 = Subtract(
   A=WorldPosition().XYZ
   B=Add(
    A=ObjectPositionWS()
    B=MaterialFunctionCall(R=@n151, G=@n151, B=@n152) {materialFunction=&AppendMany}.RGB
   )
  )
  n154 = Arctangent2Fast(Y=ComponentMask(None=@n153) {g=true}, X=ComponentMask(None=@n153) {r=true})
  n155 = ScalarParameter() {defaultValue=0.2, group=UVs, parameterName=UV Squash Min}
  n156 = ScalarParameter() {defaultValue=5, group=UVs, parameterName=UV Squash Max}
  n157 = NamedRerouteUsage()
  n158 = Divide(
   A=ComponentMask(None=@n153) {b=true}
   B=Multiply(A=ObjectRadius(), B=LinearInterpolate(A=@n155, B=@n156, Alpha=@n157) {constA=0.2, constB=4}) {constB=2}
  )
  n159 = Abs(None=AppendVector(A=Divide(A=@n154) {constB=3.1416}, B=Add(A=@n158) {constB=0.5}))
  n160 = ScalarParameter() {defaultValue=0.01, group=UVs, parameterName=Cylinder UV Min Size}
  n161 = ScalarParameter() {defaultValue=30, group=UVs, parameterName=Cylinder UV Max Size}
  n162 = NamedRerouteUsage()
  NamedRerouteDeclaration(None=Multiply(A=VertexInterpolator(VS=@n159).PS, B=LinearInterpolate(A=@n160, B=@n161, Alpha=@n162) {constA=1, constB=4})) {name=Cylindrical UV, nodeColor=<rgba 0 1 0.2734 1>}
  n163 = DotProduct(A=CameraVectorWS(), B=PixelNormalWS())
  n164 = Power(Base=OneMinus(None=Abs(None=@n163))) {constExponent=10}
  n165 = Saturate(None=Divide(A=PixelDepth().R) {constB=200})
  n166 = NamedRerouteUsage()
  n167 = Add(A=Multiply(A=NamedRerouteUsage(), B=@n138), B=@n20)
  n168 = QualitySwitch(Default=@n167, Low=Multiply(A=@n167, B=Constant() {r=1.8}))
  n169 = NamedRerouteUsage()
  n170 = EyeAdaptationInverse(LightValueInput=Multiply(A=Multiply(A=@n166, B=@n168), B=OneMinus(None=@n169)), AlphaInput=Constant() {r=0.4})
  NamedRerouteDeclaration(None=Multiply(A=OneMinus(None=@n164), B=Multiply(A=@n165, B=@n170.EyeAdaptationInverse))) {name=Emissive, nodeColor=<rgba 1 0 0.6328 1>}
  n171 = CollectionParameter() {
   collection=&FortniteMaterialParameters
   parameterName=FortniteActiveTimeOfDay
  }
  NamedRerouteDeclaration(None=Saturate(None=MaterialFunctionCall(Float4=@n171) {materialFunction=&BreakOutFloat4Components}.A)) {name=NightTime, nodeColor=<rgba 0 1 0.6016 1>}
  n172 = Add(
   A=Multiply(A=NamedRerouteUsage())
   B=Panner(Coordinate=@n53, Time=Time(), Speed=ScalarParameter() {defaultValue=-400, group=Sheen, parameterName=SweepSpeed}) {speedX=0.2, speedY=0.2}
  )
  n173 = OneMinus(None=Divide(A=ComponentMask(None=@n172) {g=true}, B=ScalarParameter() {defaultValue=15000, group=Sheen, parameterName=SweepFreq}))
  NamedRerouteDeclaration(None=SmoothStep(Min=ScalarParameter() {defaultValue=0.4, group=Sheen, parameterName=Sheen Size}, Max=Constant(), Value=Frac(None=@n173))) {name=Sheen, nodeColor=<rgba 0 1 0.8359 1>}
  NamedRerouteDeclaration(None=Saturate(None=@n149)) {name=Distance Mask Far, nodeColor=<rgba 0 1 0.8594 1>}
  n174 = OneMinus(None=NamedRerouteUsage())
  NamedRerouteDeclaration(None=Multiply(A=NamedRerouteUsage(), B=Multiply(A=@n174, B=@n174))) {name=FinalSnowNoise, nodeColor=<rgba 1 0 0.1641 1>}
  NamedRerouteUsage()
  NamedRerouteUsage()
  NamedRerouteUsage()
  NamedRerouteUsage()
  NamedRerouteUsage()
  NamedRerouteUsage()
  n175 = NamedRerouteUsage()
  n176 = ScalarParameter() {defaultValue=-0.1447, group=Pos Mask BG Color, parameterName=Top Fade Pos}
  Saturate(None=Multiply(A=Add(A=@n175, B=@n176), B=ScalarParameter() {defaultValue=31.8189, group=Pos Mask BG Color, parameterName=Top fade Size}))
  n177 = Time()
  n178 = Frac(None=Multiply(A=@n177) {constB=10})
  n179 = Abs(None=Sine(None=Multiply(A=@n177) {constB=1.5}))
  n180 = Constant() {r=1.4}
  Saturate(None=Add(A=OneMinus(None=@n178), B=Subtract(A=@n179, B=@n180)))
  TextureObject() {samplerType=SAMPLERTYPE_LinearColor, texture=/Game/Athena/HUD/reticles/NoiseThermalVision}
  TextureSample(UVs=@n80) {texture=&T_StarSky_SafeZone}
  TextureSample(UVs=@n86) {texture=&T_StarSky_SafeZone}
  n181 = ViewSize()
  n182 = Multiply(
   A=Multiply(A=NamedRerouteUsage(), B=ScalarParameter() {defaultValue=0.2, group=TVStatic, parameterName=NoiseScaleMultiplierFar, sortPriority=200}) {constB=2}
   B=@n181
  )
  n183 = Multiply(A=Time(), B=ScalarParameter() {defaultValue=6, group=TVStatic, parameterName=Noise frame rate, sortPriority=200}) {constB=24, desc=Frame rate}
  VectorNoise(World Position=AppendVector(A=Floor(None=@n182), B=@n183))
  n184 = Multiply(
   A=Multiply(A=NamedRerouteUsage(), B=ScalarParameter() {defaultValue=15, group=TVStatic, parameterName=NoiseScaleMultiplier, sortPriority=200}) {constB=2}
   B=@n181
  )
  VectorNoise(World Position=AppendVector(A=Floor(None=@n184), B=@n183))
  VertexInterpolator()
  n185 = ComponentMask(None=@n95) {g=true, r=true}
  VertexInterpolator(VS=@n185)
  VertexInterpolator(VS=@n185)
 ]
}
