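; Blueprint interface for the storm day sequence in the Athena safe zone system.
; It declares a single function, UpdateStormMeshVisibility, which takes a Boolean Visible flag.
; Implementers receive this call to show or hide the storm mesh; the interface itself holds no logic.
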(fn UpdateStormMeshVisibility
  (params (Visible :type Boolean)))
